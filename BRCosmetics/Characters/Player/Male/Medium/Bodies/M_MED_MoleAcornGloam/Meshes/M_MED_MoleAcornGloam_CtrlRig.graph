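# Control Rig for the MoleAcornGloam male medium body that drives its corrective deform bones and garment bones.
# A shared DeformRig function runs first, then shoulder and thigh additions (animation attributes, spherical pose readers), middle-finger constraints and foot-yaw-driven toe correctives.
# Skirt correctives run only with cloth disabled, blending Leg ROM poses, with RBAN switching the left-side set.
# Skirt and jacket bones follow the body through weighted multi-parent constraints.
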ControlRigBlueprint /BRCosmetics/Characters/Player/Male/Medium/Bodies/M_MED_MoleAcornGloam/Meshes/M_MED_MoleAcornGloam_CtrlRig
graphs [
 New Function_ContainedGraph {nodes [
   RigVMFunctionEntryNode()
   > RigVMFunction_Sequence = RigVMAggregateNode()
   RigVMFunctionReturnNode(ExecuteContext=@RigVMFunction_Sequence.C)
   Get_Animation_Attribute = RigVMDispatchNode(BoneName=root, Default=0, Name=deformRig_l_upperarm_up_alpha) {template="DISPATCH_RigDispatch_GetAnimAttribute(in Name,in BoneName,in Default,out Value,out Found)"}
   Get_Animation_Attribute_1 = RigVMDispatchNode(BoneName=root, Default=0, Name=deformRig_r_upperarm_up_alpha) {template="DISPATCH_RigDispatch_GetAnimAttribute(in Name,in BoneName,in Default,out Value,out Found)"}
   Modify Transforms with Mirror_1 = RigVMFunctionReferenceNode(
    Base Modify Data=((Item=(Type=Bone,Name="deform_upperarm_fr_l"),Transform=(Rotation=(X=0,Y=0,Z=0,W=1),Translation=(X=0,Y=5,Z=10),Scale3D=(X=1,Y=1,Z=1))))
    Mirror Modify Data=((Item=(Type=Bone,Name="deform_upperarm_fr_r"),Transform=(Rotation=(X=0,Y=0,Z=0,W=1),Translation=(X=0,Y=0,Z=0),Scale3D=(X=1,Y=1,Z=1))))
    Weight=@Get_Animation_Attribute.Value
    "Mirror Weight "=@Get_Animation_Attribute_1.Value
    ExecuteContext=@RigVMFunction_Sequence.A
   )
   Get_Animation_Attribute_2 = RigVMDispatchNode(BoneName=root, Default=0, Name=deformRig_l_upperarm_fr_alpha) {template="DISPATCH_RigDispatch_GetAnimAttribute(in Name,in BoneName,in Default,out Value,out Found)"}
   Get_Animation_Attribute_3 = RigVMDispatchNode(BoneName=root, Default=0, Name=deformRig_r_upperarm_fr_alpha) {template="DISPATCH_RigDispatch_GetAnimAttribute(in Name,in BoneName,in Default,out Value,out Found)"}
   > Modify Transforms with Mirror = RigVMFunctionReferenceNode(
    Base Modify Data=((Item=(Type=Bone,Name="deform_upperarm_up_l"),Transform=(Rotation=(X=0,Y=0,Z=0,W=1),Translation=(X=0,Y=2,Z=1),Scale3D=(X=1.5,Y=1,Z=1))))
    Mirror Modify Data=((Item=(Type=Bone,Name="deform_upperarm_up_r"),Transform=(Rotation=(X=0,Y=0,Z=0,W=1),Translation=(X=0,Y=0,Z=0),Scale3D=(X=1,Y=1,Z=1))))
    Weight=@Get_Animation_Attribute_2.Value
    "Mirror Weight "=@Get_Animation_Attribute_3.Value
   )
   > n1 = &RigUnit_SphericalPoseReader(
    ActiveRegionScaleFactors=(PositiveWidth=3,NegativeWidth=1,PositiveHeight=3,NegativeHeight=2)
    ActiveRegionSize=0.1
    Debug=(bDrawDebug=True,bDraw2D=False,bDrawLocalAxes=False,DebugScale=25,DebugSegments=20,DebugThickness=0.25)
    DriverAxis=(X=-1,Y=0,Z=0)
    DriverItem=(Type=Bone,Name="upperarm_r")
    FalloffRegionScaleFactors=(PositiveWidth=1,NegativeWidth=1,PositiveHeight=1,NegativeHeight=1)
    FalloffSize=0.35
    FlipHeightScaling=False
    FlipWidthScaling=False
    OptionalParentItem=(Type=Bone,Name="clavicle_r")
    RotationOffset=(X=-145,Y=-20,Z=70)
   )
   > &RigUnit_ModifyTransforms(
    ItemToModify="((Item=(Type=Bone,Name=\"deform_upperarm_fr_r\"),Transform=(Rotation=(X=0,Y=0,Z=-0.0105,W=-0.9999),Translation=(X=9.2449,Y=0.0574,Z=0),Scale3D=(X=1,Y=1.25,Z=1))),(Item=(Type=Bone,Name=\"deform_upperarm_up_r\"),Transform=(Rotation=(X=0,Y=0,Z=-0.0037,W=-1),Trans ...(truncated)"
    WeightMinimum=0
    Weight=@n1.OutputParam
   )
   &RigUnit_ParentConstraint(
    AdvancedSettings=(InterpolationType=Average,RotationOrderForFilter=XZY)
    Child=(Type=Bone,Name="deform_thigh_inner_base_l")
    Parents=((Item=(Type=Bone,Name="pelvis"),Weight=1),(Item=(Type=Bone,Name="thigh_twist_01_l"),Weight=1))
    Weight=1
    ExecutePin=@RigVMFunction_Sequence.B
   )
   > &RigUnit_ParentConstraint(
    AdvancedSettings=(InterpolationType=Average,RotationOrderForFilter=XZY)
    Child=(Type=Bone,Name="deform_thigh_inner_base_r")
    Parents=((Item=(Type=Bone,Name="pelvis"),Weight=1),(Item=(Type=Bone,Name="thigh_twist_01_r"),Weight=1))
    Weight=1
   )
   > &RigUnit_ParentConstraint(
    AdvancedSettings=(InterpolationType=Average,RotationOrderForFilter=XZY)
    Child=(Type=Bone,Name="deform_hip_base_l")
    Parents=((Item=(Type=Bone,Name="pelvis"),Weight=1),(Item=(Type=Bone,Name="thigh_twist_01_l"),Weight=1))
    Weight=1
   )
   > &RigUnit_ParentConstraint(
    AdvancedSettings=(InterpolationType=Average,RotationOrderForFilter=XZY)
    Child=(Type=Bone,Name="deform_hip_base_r")
    Parents=((Item=(Type=Bone,Name="pelvis"),Weight=1),(Item=(Type=Bone,Name="thigh_twist_01_r"),Weight=1))
    Weight=1
   )
   > n2 = &RigUnit_SphericalPoseReader(
    ActiveRegionScaleFactors=(PositiveWidth=3,NegativeWidth=1,PositiveHeight=1,NegativeHeight=1)
    ActiveRegionSize=0.1
    Debug=(bDrawDebug=false,bDraw2D=False,bDrawLocalAxes=False,DebugScale=25,DebugSegments=20,DebugThickness=0.25)
    DriverAxis=(X=-1,Y=0,Z=0)
    DriverItem=(Type=Bone,Name="thigh_l")
    FalloffRegionScaleFactors=(PositiveWidth=1,NegativeWidth=1,PositiveHeight=1,NegativeHeight=1)
    FalloffSize=0.5
    FlipHeightScaling=False
    FlipWidthScaling=False
    OptionalParentItem=(Type=Bone,Name="None")
    RotationOffset=(X=180,Y=0,Z=0)
   )
   > n3 = &RigUnit_SphericalPoseReader(
    ActiveRegionScaleFactors=(PositiveWidth=3,NegativeWidth=1,PositiveHeight=1,NegativeHeight=1)
    ActiveRegionSize=0.1
    Debug=(bDrawDebug=false,bDraw2D=False,bDrawLocalAxes=False,DebugScale=25,DebugSegments=20,DebugThickness=0.25)
    DriverAxis=(X=1,Y=0,Z=0)
    DriverItem=(Type=Bone,Name="thigh_r")
    FalloffRegionScaleFactors=(PositiveWidth=1,NegativeWidth=1,PositiveHeight=1,NegativeHeight=1)
    FalloffSize=0.5
    FlipHeightScaling=true
    FlipWidthScaling=true
    OptionalParentItem=(Type=Bone,Name="None")
    RotationOffset=(X=0,Y=0,Z=0)
   )
   > Modify Transforms with Mirror_2 = RigVMFunctionReferenceNode(
    Base Modify Data="((Item=(Type=Bone,Name=\"deform_thigh_inner_l\"),Transform=(Rotation=(X=0,Y=0,Z=0.3827,W=0.9239),Translation=(X=0,Y=0,Z=-3),Scale3D=(X=1,Y=1,Z=1))),(Item=(Type=Bone,Name=\"deform_hip_l\"),Transform=(Rotation=(X=0,Y=0,Z=0,W=1),Translation=(X=-7,Y=0,Z=-10),Scale ...(truncated)"
    Mirror Modify Data="((Item=(Type=Bone,Name=\"deform_thigh_inner_r\"),Transform=(Rotation=(X=0,Y=0,Z=0,W=1),Translation=(X=0,Y=0,Z=0),Scale3D=(X=1,Y=1,Z=1))),(Item=(Type=Bone,Name=\"deform_hip_r\"),Transform=(Rotation=(X=0,Y=0,Z=0,W=1),Translation=(X=0,Y=0,Z=0),Scale3D=(X=1,Y=1,Z= ...(truncated)"
    Weight=@n2.OutputParam
    "Mirror Weight "=@n3.OutputParam
   )
  ]}
 New Function_ContainedGraph {nodes [
   RigVMFunctionEntryNode()
   > For_Each = RigVMDispatchNode(Array=RigVMVariableNode() {variable=Source Middle Fingers}.Value) {template="DISPATCH_RigVMDispatch_ArrayIterator(in Array,out Element,out Index,out Count,out Ratio)"}
   > &RigUnit_ParentConstraint(
    AdvancedSettings=(InterpolationType=Average,RotationOrderForFilter=XZY)
    Parents=((Item=(Type=Bone,Name="None"),Weight=1))
    Weight=1
    Child=RigVMDispatchNode(Index=@For_Each.Index, Array=RigVMVariableNode() {variable=Target Middle Fingers}.Value) {template="DISPATCH_RigVMDispatch_ArrayGetAtIndex(in Array,in Index,out Element)"}.Element
    Parents.0.Item=@For_Each.Element
   )
   RigVMFunctionReturnNode(ExecuteContext=@For_Each.Completed)
  ]}
 New Function_ContainedGraph {nodes [
   RigVMFunctionEntryNode()
   > &RigUnit_AimConstraintLocalSpaceOffset(
    Child=(Type=Bone,Name="deform_toe_l_outer")
    Filter=(bX=True,bY=True,bZ=True)
    Parents=((Item=(Type=Control,Name="ball_l_ctrl"),Weight=1))
    Weight=1
    WorldUp=(Target=(X=0,Y=0,Z=1),Kind=Location,Space=(Type=Control,Name="ball_l_ctrl"))
   )
   > &RigUnit_AimConstraintLocalSpaceOffset(
    Child=(Type=Bone,Name="deform_toe_l_mid")
    Filter=(bX=True,bY=True,bZ=True)
    Parents=((Item=(Type=Control,Name="ball_l_ctrl"),Weight=1))
    Weight=1
    WorldUp=(Target=(X=0,Y=0,Z=1),Kind=Location,Space=(Type=Control,Name="ball_l_ctrl"))
   )
   > &RigUnit_AimConstraintLocalSpaceOffset(
    Child=(Type=Bone,Name="deform_toe_l_inner")
    Filter=(bX=True,bY=True,bZ=True)
    Parents=((Item=(Type=Control,Name="ball_l_ctrl"),Weight=1))
    Weight=1
    WorldUp=(Target=(X=0,Y=0,Z=1),Kind=Location,Space=(Type=Control,Name="ball_l_ctrl"))
   )
   > &RigUnit_AimConstraintLocalSpaceOffset(
    Child=(Type=Bone,Name="deform_toe_r_outer")
    Filter=(bX=True,bY=True,bZ=True)
    Parents=((Item=(Type=Control,Name="ball_r_ctrl"),Weight=1))
    Weight=1
    WorldUp=(Target=(X=0,Y=0,Z=1),Kind=Location,Space=(Type=Control,Name="ball_r_ctrl"))
   )
   > &RigUnit_AimConstraintLocalSpaceOffset(
    Child=(Type=Bone,Name="deform_toe_r_mid")
    Filter=(bX=True,bY=True,bZ=True)
    Parents=((Item=(Type=Control,Name="ball_r_ctrl"),Weight=1))
    Weight=1
    WorldUp=(Target=(X=0,Y=0,Z=1),Kind=Location,Space=(Type=Control,Name="ball_r_ctrl"))
   )
   > &RigUnit_AimConstraintLocalSpaceOffset(
    Child=(Type=Bone,Name="deform_toe_r_inner")
    Filter=(bX=True,bY=True,bZ=True)
    Parents=((Item=(Type=Control,Name="ball_r_ctrl"),Weight=1))
    Weight=1
    WorldUp=(Target=(X=0,Y=0,Z=1),Kind=Location,Space=(Type=Control,Name="ball_r_ctrl"))
   )
   > Get Delta Ref = RigVMFunctionReferenceNode(Item=(Type=Bone,Name="foot_l"))
   > Get Delta Ref_1 = RigVMFunctionReferenceNode(Item=(Type=Bone,Name="foot_r"))
   Remap = &RigVMFunction_MathDoubleRemap(SourceMaximum=50, SourceMinimum=0, TargetMaximum=1, TargetMinimum=0, Value=@"Get Delta Ref".Result.Rotation.Yaw)
   > Modify Transforms with Mirror = RigVMFunctionReferenceNode(
    Base Modify Data="((Item=(Type=Bone,Name=\"deform_toe_l_inner\"),Transform=(Rotation=(X=0,Y=0.2971,Z=0,W=0.9549),Translation=(X=-0.6923,Y=0,Z=-1.3044),Scale3D=(X=1,Y=1,Z=1))),(Item=(Type=Bone,Name=\"deform_toe_l_outer\"),Transform=(Rotation=(X=0,Y=-0.2039,Z=0,W=0.979),Translati ...(truncated)"
    Mirror Modify Data="((Item=(Type=Bone,Name=\"deform_toe_r_inner\"),Transform=(Rotation=(X=0,Y=0,Z=0,W=1),Translation=(X=0,Y=0,Z=0),Scale3D=(X=1,Y=1,Z=1))),(Item=(Type=Bone,Name=\"deform_toe_r_outer\"),Transform=(Rotation=(X=0,Y=0,Z=0,W=1),Translation=(X=0,Y=0,Z=0),Scale3D=(X=1,Y= ...(truncated)"
    Weight=@Remap.Result
    "Mirror Weight "=&RigVMFunction_MathDoubleRemap(SourceMaximum=50, SourceMinimum=0, TargetMaximum=1, TargetMinimum=0, Value=@"Get Delta Ref_1".Result.Rotation.Yaw).Result
   )
   > Get Delta Ref_2 = RigVMFunctionReferenceNode(Item=(Type=Bone,Name="foot_l"))
   > Get Delta Ref_3 = RigVMFunctionReferenceNode(Item=(Type=Bone,Name="foot_r"))
   > Modify Transforms with Mirror_1 = RigVMFunctionReferenceNode(
    Base Modify Data="((Item=(Type=Bone,Name=\"deform_toe_l_inner\"),Transform=(Rotation=(X=0,Y=-0.2164,Z=0,W=0.9763),Translation=(X=0,Y=0,Z=0),Scale3D=(X=1,Y=1,Z=1))),(Item=(Type=Bone,Name=\"deform_toe_l_outer\"),Transform=(Rotation=(X=0,Y=0.2164,Z=0,W=0.9763),Translation=(X=0,Y=0 ...(truncated)"
    Mirror Modify Data="((Item=(Type=Bone,Name=\"deform_toe_r_inner\"),Transform=(Rotation=(X=0,Y=0,Z=0,W=1),Translation=(X=0,Y=0,Z=0),Scale3D=(X=1,Y=1,Z=1))),(Item=(Type=Bone,Name=\"deform_toe_r_outer\"),Transform=(Rotation=(X=0,Y=0,Z=0,W=1),Translation=(X=0,Y=0,Z=0),Scale3D=(X=1,Y= ...(truncated)"
    Weight=&RigVMFunction_MathDoubleRemap(SourceMaximum=-20, SourceMinimum=0, TargetMaximum=1, TargetMinimum=0, Value=@"Get Delta Ref_2".Result.Rotation.Yaw).Result
    "Mirror Weight "=&RigVMFunction_MathDoubleRemap(SourceMaximum=-20, SourceMinimum=0, TargetMaximum=1, TargetMinimum=0, Value=@"Get Delta Ref_3".Result.Rotation.Yaw).Result
   )
   > RigVMFunctionReturnNode()
  ]}
 New Function_ContainedGraph {nodes [
   n1 = RigVMFunctionEntryNode()
   RigVMFunction_ControlFlowBranch = &RigVMFunction_ControlFlowBranch(Condition=@n1.Cloth Enabled, ExecuteContext=@n1.ExecuteContext)
   RigVMFunctionReturnNode(ExecuteContext=@RigVMFunction_ControlFlowBranch.Completed)
   RigVMFunction_Sequence = RigVMAggregateNode(ExecuteContext=@RigVMFunction_ControlFlowBranch.False)
   &RigUnit_AimConstraintLocalSpaceOffset(
    Child=(Type=Bone,Name="dyn_skirt_front_l_MSTR")
    Filter=(bX=True,bY=True,bZ=True)
    Parents=((Item=(Type=Control,Name="dyn_skirt_front_l_04_ctrl"),Weight=1))
    Weight=1
    WorldUp=(Target=(X=0,Y=0,Z=1),Kind=Location,Space=(Type=Control,Name="dyn_skirt_front_l_04_ctrl"))
    ExecutePin=@RigVMFunction_Sequence.B
   )
   > &RigUnit_AimConstraintLocalSpaceOffset(
    Child=(Type=Bone,Name="dyn_skirt_side_l_MSTR")
    Filter=(bX=True,bY=True,bZ=True)
    Parents=((Item=(Type=Control,Name="dyn_skirt_side_l_04_ctrl"),Weight=1))
    Weight=1
    WorldUp=(Target=(X=0,Y=0,Z=1),Kind=Location,Space=(Type=Control,Name="dyn_skirt_side_l_04_ctrl"))
   )
   > &RigUnit_AimConstraintLocalSpaceOffset(
    Child=(Type=Bone,Name="dyn_skirt_side_r_MSTR")
    Filter=(bX=True,bY=True,bZ=True)
    Parents=((Item=(Type=Control,Name="dyn_skirt_side_r_03_ctrl"),Weight=1))
    Weight=1
    WorldUp=(Target=(X=0,Y=0,Z=1),Kind=Location,Space=(Type=Control,Name="dyn_skirt_side_r_03_ctrl"))
   )
   > Leg ROM = RigVMFunctionReferenceNode(
    Include Ankles=false
    Include Distance Checks=false
    Include Knees=false
    Interp Output Speed=5
    Interp Results=true
    Result Alpha=1
    SPR Optional Parent Item=(Type=Bone,Name="spine_02")
   )
   n2 = &RigVMFunction_ControlFlowBranch(ExecuteContext=@"Leg ROM".ExecuteContext, Condition=RigVMVariableNode() {variable=RBAN Enabled}.Value)
   &RigUnit_ModifyTransforms(
    ItemToModify="((Item=(Type=Bone,Name=\"dyn_skirt_front_l_MSTR\"),Transform=(Rotation=(X=0.1463,Y=-0.1401,Z=0.3995,W=0.8941),Translation=(X=-0.1675,Y=-0.4686,Z=0.2708),Scale3D=(X=1,Y=1,Z=1))),(Item=(Type=Bone,Name=\"dyn_skirt_front_l_02\"),Transform=(Rotation=(X=0,Y=0,Z=0,W= ...(truncated)"
    WeightMinimum=0
    ExecutePin=@n2.False
    Weight=@"Leg ROM".L Thigh Raised Fwd
   )
   > &RigUnit_ModifyTransforms(
    ItemToModify="((Item=(Type=Bone,Name=\"dyn_skirt_front_l_MSTR\"),Transform=(Rotation=(X=0.0241,Y=0.0004,Z=-0.079,W=0.9966),Translation=(X=1.5434,Y=3.4173,Z=-2.0677),Scale3D=(X=1,Y=1,Z=1))),(Item=(Type=Bone,Name=\"dyn_skirt_front_l_02\"),Transform=(Rotation=(X=0,Y=0,Z=-0.097 ...(truncated)"
    WeightMinimum=0
    Weight=&RigVMFunction_MathDoubleRemap(SourceMaximum=1, SourceMinimum=0.5, TargetMaximum=1, TargetMinimum=0, Value=@"Leg ROM".L Thigh Raised Fwd).Result
   )
   > &RigUnit_ModifyTransforms(
    ItemToModify="((Item=(Type=Bone,Name=\"dyn_skirt_front_l_MSTR\"),Transform=(Rotation=(X=0.0695,Y=-0.2246,Z=-0.0128,W=0.9719),Translation=(X=0,Y=0,Z=0),Scale3D=(X=1,Y=1,Z=1))),(Item=(Type=Bone,Name=\"dyn_skirt_front_l_02\"),Transform=(Rotation=(X=0,Y=0,Z=0,W=1),Translation=( ...(truncated)"
    WeightMinimum=0
    Weight=@"Leg ROM".L Thigh Raised Fwd Side Inbetween
   )
   > &RigUnit_ModifyTransforms(
    ItemToModify="((Item=(Type=Bone,Name=\"dyn_skirt_front_l_MSTR\"),Transform=(Rotation=(X=0.2584,Y=-0.3337,Z=0.3539,W=0.8346),Translation=(X=0,Y=0,Z=0),Scale3D=(X=1,Y=1,Z=1))),(Item=(Type=Bone,Name=\"dyn_skirt_side_l_MSTR\"),Transform=(Rotation=(X=0.0011,Y=-0.0018,Z=0.5293,W= ...(truncated)"
    WeightMinimum=0
    Weight=@"Leg ROM".L Thigh Raised Side
   )
   &RigUnit_ModifyTransforms(
    ItemToModify=((Item=(Type=Bone,Name="dyn_skirt_front_l_MSTR"),Transform=(Rotation=(X=0.0092,Y=-0.3209,Z=0.3341,W=0.8862),Translation=(X=-1.8761,Y=0.2192,Z=0.459),Scale3D=(X=1,Y=1,Z=1))))
    WeightMinimum=0
    Weight=@"Leg ROM".L Thigh Raised Fwd
    ExecutePin=@n2.True
   )
   > &RigUnit_ModifyTransforms(
    ItemToModify=((Item=(Type=Bone,Name="dyn_skirt_front_l_MSTR"),Transform=(Rotation=(X=0,Y=0,Z=0,W=1),Translation=(X=0,Y=0,Z=0),Scale3D=(X=1,Y=1,Z=1))))
    WeightMinimum=0
    Weight=&RigVMFunction_MathDoubleRemap(SourceMaximum=1, SourceMinimum=0.5, TargetMaximum=1, TargetMinimum=0, Value=@"Leg ROM".L Thigh Raised Fwd).Result
   )
   > &RigUnit_ModifyTransforms(
    ItemToModify="((Item=(Type=Bone,Name=\"dyn_skirt_front_l_MSTR\"),Transform=(Rotation=(X=0.1783,Y=-0.0122,Z=0.1783,W=0.9676),Translation=(X=2.8,Y=2,Z=0),Scale3D=(X=1,Y=1,Z=1))),(Item=(Type=Bone,Name=\"dyn_skirt_side_l_MSTR\"),Transform=(Rotation=(X=0,Y=0,Z=0.0191,W=0.9998),T ...(truncated)"
    WeightMinimum=0
    Weight=@"Leg ROM".L Thigh Raised Fwd Side Inbetween
   )
   > &RigUnit_ModifyTransforms(
    ItemToModify="((Item=(Type=Bone,Name=\"dyn_skirt_front_l_MSTR\"),Transform=(Rotation=(X=-0.0755,Y=-0.4819,Z=0.2495,W=0.8365),Translation=(X=0.6654,Y=1.0925,Z=1.3093),Scale3D=(X=1,Y=1,Z=1))),(Item=(Type=Bone,Name=\"dyn_skirt_side_l_MSTR\"),Transform=(Rotation=(X=0.001,Y=-0.0 ...(truncated)"
    WeightMinimum=0
    Weight=@"Leg ROM".L Thigh Raised Side
   )
   &RigUnit_ModifyTransforms(
    ItemToModify="((Item=(Type=Bone,Name=\"dyn_skirt_front_r_MSTR\"),Transform=(Rotation=(X=0.0031,Y=-0.0502,Z=0.0613,W=0.9969),Translation=(X=0,Y=0,Z=0),Scale3D=(X=1,Y=1,Z=1))),(Item=(Type=Bone,Name=\"dyn_skirt_side_r_MSTR\"),Transform=(Rotation=(X=0.1251,Y=0.0052,Z=0.0288,W=0 ...(truncated)"
    WeightMinimum=0
    Weight=@"Leg ROM".R Thigh Raised Fwd
    ExecutePin=@n2.Completed
   )
   > &RigUnit_ModifyTransforms(
    ItemToModify="((Item=(Type=Bone,Name=\"dyn_skirt_front_r_MSTR\"),Transform=(Rotation=(X=0.0053,Y=-0.1945,Z=0.0265,W=0.9805),Translation=(X=-0.2048,Y=-0.0111,Z=-1.5206),Scale3D=(X=1,Y=1,Z=1))),(Item=(Type=Bone,Name=\"dyn_skirt_side_r_MSTR\"),Transform=(Rotation=(X=0.0364,Y=- ...(truncated)"
    WeightMinimum=0
    Weight=@"Leg ROM".R Thigh Raised Fwd Side Inbetween
   )
   > &RigUnit_ModifyTransforms(
    ItemToModify="((Item=(Type=Bone,Name=\"dyn_skirt_front_r_MSTR\"),Transform=(Rotation=(X=0,Y=0,Z=0,W=1),Translation=(X=0,Y=0,Z=0),Scale3D=(X=1,Y=1,Z=1))),(Item=(Type=Bone,Name=\"dyn_skirt_side_r_MSTR\"),Transform=(Rotation=(X=0.0097,Y=-0.1349,Z=0.4373,W=0.8891),Translation=( ...(truncated)"
    WeightMinimum=0
    Weight=@"Leg ROM".R Thigh Raised Side
   )
   &RigUnit_ModifyTransforms(
    ItemToModify=((Item=(Type=Bone,Name="dyn_skirt_front_l_MSTR"),Transform=(Rotation=(X=0,Y=0,Z=0,W=1),Translation=(X=0,Y=2.5,Z=0),Scale3D=(X=1,Y=1,Z=1))))
    Weight=1
    WeightMinimum=0
    ExecutePin=@RigVMFunction_Sequence.C
   )
   &RigUnit_ParentConstraint(
    AdvancedSettings=(RotationOrderForFilter=XZY)
    Child=(Type=Bone,Name="dyn_skirt_front_l_MSTR")
    Parents=((Item=(Type=Bone,Name="spine_04"),Weight=0.0329),(Item=(Type=Bone,Name="pelvis"),Weight=0.1608),(Item=(Type=Bone,Name="spine_01"),Weight=0.2079),(Item=(Type=Bone,Name="spine_03"),Weight=0.2572),(Item=(Type=Bone,Name="spine_02"),Weight=0.3412))
    Weight=1
    ExecutePin=@RigVMFunction_Sequence.A
   )
   > &RigUnit_ParentConstraint(
    AdvancedSettings=(RotationOrderForFilter=XZY)
    Child=(Type=Bone,Name="dyn_skirt_side_l_MSTR")
    Parents="((Item=(Type=Bone,Name=\"clavicle_l\"),Weight=0.0047),(Item=(Type=Bone,Name=\"spine_04\"),Weight=0.0226),(Item=(Type=Bone,Name=\"pelvis\"),Weight=0.2069),(Item=(Type=Bone,Name=\"spine_03\"),Weight=0.2175),(Item=(Type=Bone,Name=\"spine_01\"),Weight=0.219),(Item=(Type ...(truncated)"
    Weight=1
   )
   > &RigUnit_ParentConstraint(
    AdvancedSettings=(RotationOrderForFilter=XZY)
    Child=(Type=Bone,Name="dyn_skirt_rear_l_MSTR")
    Parents=((Item=(Type=Bone,Name="spine_04"),Weight=0.0493),(Item=(Type=Bone,Name="pelvis"),Weight=0.1878),(Item=(Type=Bone,Name="spine_01"),Weight=0.1889),(Item=(Type=Bone,Name="spine_03"),Weight=0.2605),(Item=(Type=Bone,Name="spine_02"),Weight=0.3134))
    Weight=1
   )
   > &RigUnit_ParentConstraint(
    AdvancedSettings=(RotationOrderForFilter=XZY)
    Child=(Type=Bone,Name="dyn_skirt_rear_MSTR")
    Parents=((Item=(Type=Bone,Name="spine_04"),Weight=0.0537),(Item=(Type=Bone,Name="spine_01"),Weight=0.1815),(Item=(Type=Bone,Name="pelvis"),Weight=0.1838),(Item=(Type=Bone,Name="spine_03"),Weight=0.2835),(Item=(Type=Bone,Name="spine_02"),Weight=0.2974))
    Weight=1
   )
   > &RigUnit_ParentConstraint(
    AdvancedSettings=(RotationOrderForFilter=XZY)
    Child=(Type=Bone,Name="dyn_skirt_front_MSTR")
    Parents=((Item=(Type=Bone,Name="spine_04"),Weight=0.037),(Item=(Type=Bone,Name="pelvis"),Weight=0.1922),(Item=(Type=Bone,Name="spine_01"),Weight=0.2068),(Item=(Type=Bone,Name="spine_03"),Weight=0.2385),(Item=(Type=Bone,Name="spine_02"),Weight=0.3254))
    Weight=1
   )
   > &RigUnit_ParentConstraint(
    AdvancedSettings=(RotationOrderForFilter=XZY)
    Child=(Type=Bone,Name="dyn_skirt_front_r_MSTR")
    Parents="((Item=(Type=Bone,Name=\"clavicle_r\"),Weight=0.0003),(Item=(Type=Bone,Name=\"spine_04\"),Weight=0.031),(Item=(Type=Bone,Name=\"pelvis\"),Weight=0.1579),(Item=(Type=Bone,Name=\"spine_01\"),Weight=0.2104),(Item=(Type=Bone,Name=\"spine_03\"),Weight=0.2605),(Item=(Type ...(truncated)"
    Weight=1
   )
   > &RigUnit_ParentConstraint(
    AdvancedSettings=(RotationOrderForFilter=XZY)
    Child=(Type=Bone,Name="dyn_skirt_side_r_MSTR")
    Parents="((Item=(Type=Bone,Name=\"clavicle_r\"),Weight=0.0047),(Item=(Type=Bone,Name=\"spine_04\"),Weight=0.0229),(Item=(Type=Bone,Name=\"pelvis\"),Weight=0.2059),(Item=(Type=Bone,Name=\"spine_03\"),Weight=0.2184),(Item=(Type=Bone,Name=\"spine_01\"),Weight=0.2188),(Item=(Typ ...(truncated)"
    Weight=1
   )
   > &RigUnit_ParentConstraint(
    AdvancedSettings=(RotationOrderForFilter=XZY)
    Child=(Type=Bone,Name="dyn_skirt_rear_r_MSTR")
    Parents=((Item=(Type=Bone,Name="spine_04"),Weight=0.0494),(Item=(Type=Bone,Name="pelvis"),Weight=0.187),(Item=(Type=Bone,Name="spine_01"),Weight=0.1888),(Item=(Type=Bone,Name="spine_03"),Weight=0.2612),(Item=(Type=Bone,Name="spine_02"),Weight=0.3136))
    Weight=1
   )
  ]}
 New Function_ContainedGraph {nodes [
   n1 = RigVMFunctionEntryNode()
   RigVMFunction_Sequence = &RigVMFunction_Sequence(ExecuteContext=@n1.ExecuteContext)
   n2 = &RigVMFunction_Sequence(ExecuteContext=@RigVMFunction_Sequence.A)
   &RigUnit_ParentConstraint(
    AdvancedSettings=(RotationOrderForFilter=XZY)
    Child=(Type=Bone,Name="dyn_jacket_front_l_01")
    Parents="((Item=(Type=Bone,Name=\"upperarm_twist_01_l\"),Weight=0.0196),(Item=(Type=Bone,Name=\"deform_upperarm_fr_l\"),Weight=0.0243),(Item=(Type=Bone,Name=\"spine_02\"),Weight=0.0274),(Item=(Type=Bone,Name=\"spine_03\"),Weight=0.0774),(Item=(Type=Bone,Name=\"clavicle_l\"), ...(truncated)"
    Weight=1
    ExecutePin=@n2.A
   )
   > &RigUnit_ParentConstraint(
    AdvancedSettings=(RotationOrderForFilter=XZY)
    Child=(Type=Bone,Name="dyn_jacket_front_l_02")
    Parents="((Item=(Type=Bone,Name=\"deform_upperarm_fr_l\"),Weight=0.0031),(Item=(Type=Bone,Name=\"spine_01\"),Weight=0.0056),(Item=(Type=Bone,Name=\"clavicle_l\"),Weight=0.068),(Item=(Type=Bone,Name=\"spine_02\"),Weight=0.0699),(Item=(Type=Bone,Name=\"spine_03\"),Weight=0.152 ...(truncated)"
    Weight=1
   )
   > &RigUnit_ParentConstraint(
    AdvancedSettings=(RotationOrderForFilter=XZY)
    Child=(Type=Bone,Name="dyn_jacket_front_side_l_01")
    Parents="((Item=(Type=Bone,Name=\"spine_05\"),Weight=0.0112),(Item=(Type=Bone,Name=\"spine_02\"),Weight=0.0427),(Item=(Type=Bone,Name=\"spine_03\"),Weight=0.1141),(Item=(Type=Bone,Name=\"spine_04\"),Weight=0.1183),(Item=(Type=Bone,Name=\"deform_upperarm_fr_l\"),Weight=0.1324 ...(truncated)"
    Weight=1
   )
   > &RigUnit_ParentConstraint(
    AdvancedSettings=(RotationOrderForFilter=XZY)
    Child=(Type=Bone,Name="dyn_jacket_front_side_l_02")
    Parents="((Item=(Type=Bone,Name=\"deform_upperarm_fr_l\"),Weight=0.0133),(Item=(Type=Bone,Name=\"spine_01\"),Weight=0.0484),(Item=(Type=Bone,Name=\"clavicle_l\"),Weight=0.1319),(Item=(Type=Bone,Name=\"spine_02\"),Weight=0.1415),(Item=(Type=Bone,Name=\"spine_04\"),Weight=0.19 ...(truncated)"
    Weight=1
   )
   > &RigUnit_ParentConstraint(
    AdvancedSettings=(RotationOrderForFilter=XZY)
    Child=(Type=Bone,Name="dyn_jacket_side_l_01")
    Parents="((Item=(Type=Bone,Name=\"deform_pec_l\"),Weight=0.0072),(Item=(Type=Bone,Name=\"deform_upperarm_bk_l\"),Weight=0.0484),(Item=(Type=Bone,Name=\"deform_upperarm_fr_l\"),Weight=0.0932),(Item=(Type=Bone,Name=\"spine_02\"),Weight=0.112),(Item=(Type=Bone,Name=\"spine_04\" ...(truncated)"
    Weight=1
   )
   > &RigUnit_ParentConstraint(
    AdvancedSettings=(RotationOrderForFilter=XZY)
    Child=(Type=Bone,Name="dyn_jacket_side_l_02")
    Parents="((Item=(Type=Bone,Name=\"deform_pec_l\"),Weight=0.0223),(Item=(Type=Bone,Name=\"pelvis\"),Weight=0.0285),(Item=(Type=Bone,Name=\"spine_01\"),Weight=0.0889),(Item=(Type=Bone,Name=\"clavicle_l\"),Weight=0.1512),(Item=(Type=Bone,Name=\"spine_04\"),Weight=0.1736),(Item= ...(truncated)"
    Weight=1
   )
   > &RigUnit_ParentConstraint(
    AdvancedSettings=(RotationOrderForFilter=XZY)
    Child=(Type=Bone,Name="dyn_jacket_rear_l_01")
    Parents="((Item=(Type=Bone,Name=\"deform_upperarm_up_l\"),Weight=0.0004),(Item=(Type=Bone,Name=\"spine_02\"),Weight=0.0173),(Item=(Type=Bone,Name=\"upperarm_twist_01_l\"),Weight=0.0348),(Item=(Type=Bone,Name=\"deform_upperarm_bk_l\"),Weight=0.0487),(Item=(Type=Bone,Name=\"s ...(truncated)"
    Weight=1
   )
   > &RigUnit_ParentConstraint(
    AdvancedSettings=(RotationOrderForFilter=XZY)
    Child=(Type=Bone,Name="dyn_jacket_rear_l_02")
    Parents="((Item=(Type=Bone,Name=\"spine_01\"),Weight=0.026),(Item=(Type=Bone,Name=\"spine_05\"),Weight=0.035),(Item=(Type=Bone,Name=\"deform_upperarm_bk_l\"),Weight=0.0474),(Item=(Type=Bone,Name=\"spine_02\"),Weight=0.0792),(Item=(Type=Bone,Name=\"spine_04\"),Weight=0.241),( ...(truncated)"
    Weight=1
   )
   > &RigUnit_ParentConstraint(
    AdvancedSettings=(RotationOrderForFilter=XZY)
    Child=(Type=Bone,Name="dyn_jacket_rear_01")
    Parents=((Item=(Type=Bone,Name="spine_02"),Weight=0.0199),(Item=(Type=Bone,Name="spine_04"),Weight=0.1853),(Item=(Type=Bone,Name="spine_05"),Weight=0.3784),(Item=(Type=Bone,Name="spine_03"),Weight=0.4164))
    Weight=1
   )
   > &RigUnit_ParentConstraint(
    AdvancedSettings=(RotationOrderForFilter=XZY)
    Child=(Type=Bone,Name="dyn_jacket_rear_02")
    Parents=((Item=(Type=Bone,Name="spine_01"),Weight=0.0243),(Item=(Type=Bone,Name="spine_02"),Weight=0.0489),(Item=(Type=Bone,Name="spine_05"),Weight=0.1911),(Item=(Type=Bone,Name="spine_04"),Weight=0.3349),(Item=(Type=Bone,Name="spine_03"),Weight=0.4007))
    Weight=1
   )
   > &RigUnit_ParentConstraint(
    AdvancedSettings=(RotationOrderForFilter=XZY)
    Child=(Type=Bone,Name="dyn_jacket_rear_r_01")
    Parents="((Item=(Type=Bone,Name=\"deform_upperarm_up_r\"),Weight=0.0004),(Item=(Type=Bone,Name=\"spine_02\"),Weight=0.0182),(Item=(Type=Bone,Name=\"upperarm_twist_01_r\"),Weight=0.0326),(Item=(Type=Bone,Name=\"deform_upperarm_bk_r\"),Weight=0.0477),(Item=(Type=Bone,Name=\"s ...(truncated)"
    Weight=1
   )
   > &RigUnit_ParentConstraint(
    AdvancedSettings=(RotationOrderForFilter=XZY)
    Child=(Type=Bone,Name="dyn_jacket_rear_r_02")
    Parents="((Item=(Type=Bone,Name=\"spine_01\"),Weight=0.0252),(Item=(Type=Bone,Name=\"spine_05\"),Weight=0.0351),(Item=(Type=Bone,Name=\"deform_upperarm_bk_r\"),Weight=0.05),(Item=(Type=Bone,Name=\"spine_02\"),Weight=0.078),(Item=(Type=Bone,Name=\"spine_04\"),Weight=0.2396),( ...(truncated)"
    Weight=1
   )
   > &RigUnit_ParentConstraint(
    AdvancedSettings=(RotationOrderForFilter=XZY)
    Child=(Type=Bone,Name="dyn_jacket_side_r_01")
    Parents="((Item=(Type=Bone,Name=\"upperarm_twist_01_r\"),Weight=0.0003),(Item=(Type=Bone,Name=\"deform_pec_r\"),Weight=0.0071),(Item=(Type=Bone,Name=\"deform_upperarm_bk_r\"),Weight=0.0486),(Item=(Type=Bone,Name=\"deform_upperarm_fr_r\"),Weight=0.0934),(Item=(Type=Bone,Nam ...(truncated)"
    Weight=1
   )
   > &RigUnit_ParentConstraint(
    AdvancedSettings=(RotationOrderForFilter=XZY)
    Child=(Type=Bone,Name="dyn_jacket_side_r_02")
    Parents="((Item=(Type=Bone,Name=\"deform_upperarm_bk_r\"),Weight=0.0008),(Item=(Type=Bone,Name=\"deform_pec_r\"),Weight=0.0226),(Item=(Type=Bone,Name=\"pelvis\"),Weight=0.0249),(Item=(Type=Bone,Name=\"spine_01\"),Weight=0.0873),(Item=(Type=Bone,Name=\"clavicle_r\"),Weight=0. ...(truncated)"
    Weight=1
   )
   > &RigUnit_ParentConstraint(
    AdvancedSettings=(RotationOrderForFilter=XZY)
    Child=(Type=Bone,Name="dyn_jacket_front_side_r_01")
    Parents="((Item=(Type=Bone,Name=\"spine_05\"),Weight=0.007),(Item=(Type=Bone,Name=\"spine_02\"),Weight=0.0541),(Item=(Type=Bone,Name=\"spine_04\"),Weight=0.1284),(Item=(Type=Bone,Name=\"spine_03\"),Weight=0.1301),(Item=(Type=Bone,Name=\"deform_upperarm_fr_r\"),Weight=0.1345) ...(truncated)"
    Weight=1
   )
   > &RigUnit_ParentConstraint(
    AdvancedSettings=(RotationOrderForFilter=XZY)
    Child=(Type=Bone,Name="dyn_jacket_front_side_r_02")
    Parents="((Item=(Type=Bone,Name=\"spine_05\"),Weight=0.0001),(Item=(Type=Bone,Name=\"deform_upperarm_fr_r\"),Weight=0.011),(Item=(Type=Bone,Name=\"spine_01\"),Weight=0.0504),(Item=(Type=Bone,Name=\"clavicle_r\"),Weight=0.1299),(Item=(Type=Bone,Name=\"spine_02\"),Weight=0.145 ...(truncated)"
    Weight=1
   )
   > &RigUnit_ParentConstraint(
    AdvancedSettings=(RotationOrderForFilter=XZY)
    Child=(Type=Bone,Name="dyn_jacket_front_r_01")
    Parents="((Item=(Type=Bone,Name=\"spine_02\"),Weight=0.0114),(Item=(Type=Bone,Name=\"deform_upperarm_fr_r\"),Weight=0.0164),(Item=(Type=Bone,Name=\"upperarm_twist_01_r\"),Weight=0.0167),(Item=(Type=Bone,Name=\"spine_03\"),Weight=0.0549),(Item=(Type=Bone,Name=\"spine_04\"),We ...(truncated)"
    Weight=1
   )
   > &RigUnit_ParentConstraint(
    AdvancedSettings=(RotationOrderForFilter=XZY)
    Child=(Type=Bone,Name="dyn_jacket_front_r_02")
    Parents="((Item=(Type=Bone,Name=\"upperarm_twist_01_r\"),Weight=0.0034),(Item=(Type=Bone,Name=\"deform_upperarm_fr_r\"),Weight=0.0042),(Item=(Type=Bone,Name=\"spine_02\"),Weight=0.034),(Item=(Type=Bone,Name=\"spine_03\"),Weight=0.0946),(Item=(Type=Bone,Name=\"clavicle_r\"),W ...(truncated)"
    Weight=1
   )
   &RigUnit_ModifyTransforms(
    ItemToModify="((Item=(Type=Bone,Name=\"dyn_jacket_front_l_01\"),Transform=(Rotation=(X=0,Y=0,Z=0,W=1),Translation=(X=0,Y=0.5,Z=0),Scale3D=(X=1,Y=1,Z=1))),(Item=(Type=Bone,Name=\"dyn_jacket_front_r_01\"),Transform=(Rotation=(X=0,Y=0,Z=0,W=1),Translation=(X=0,Y=-0.5,Z=0),Scal ...(truncated)"
    Weight=1
    WeightMinimum=0
    ExecutePin=@n2.B
   )
   RigVMFunctionReturnNode(ExecuteContext=@RigVMFunction_Sequence.B)
  ]}
 RigVMFunctionLibrary {nodes [
   Deform Jacket = RigVMCollapseNode(ExecuteContext=())
   Deform Middle Finger = RigVMCollapseNode(ExecuteContext=())
   Deform Rig Additions = RigVMCollapseNode(ExecuteContext=())
   Deform Skirt = RigVMCollapseNode(Cloth Enabled=False, ExecuteContext=(), RBAN Enabled=False)
   Deform Toes = RigVMCollapseNode(ExecuteContext=())
  ]}
 RigVMFunction_Sequence_1_ContainedGraph {nodes [
   n1 = RigVMFunctionEntryNode()
   RigVMFunction_Sequence = &RigVMFunction_Sequence(ExecuteContext=@n1.ExecuteContext)
   n2 = &RigVMFunction_Sequence(ExecuteContext=@RigVMFunction_Sequence.B)
   RigVMFunctionReturnNode(A=@RigVMFunction_Sequence.A, B=@n2.A, C=@n2.B)
  ]}
 RigVMFunction_Sequence_1_ContainedGraph {nodes [
   n1 = RigVMFunctionEntryNode()
   RigVMFunction_Sequence = &RigVMFunction_Sequence(ExecuteContext=@n1.ExecuteContext)
   n2 = &RigVMFunction_Sequence(ExecuteContext=@RigVMFunction_Sequence.B)
   RigVMFunctionReturnNode(A=@RigVMFunction_Sequence.A, B=@n2.A, C=@n2.B)
  ]}
 RigVMModel {nodes [
   RigUnit_BeginExecution = &RigUnit_BeginExecution()
   FN_DeformRig_FNC = RigVMFunctionReferenceNode(
    DeformRig_ Curve="(deformRig_l_upperarm_up_alpha_72_#1=True,deformRig_r_upperarm_up_alpha_73_#2=True,deformRig_l_upperarm_fr_alpha_74_#3=True,deformRig_r_upperarm_fr_alpha_75_CB51B80C4 ...(truncated)"
    DeformRig_Bool="(upperarm_1_#4=True,elbow_3_#5=True,wrist_7_#6=True,groin_8_#7=True,glute_17_#8=False,knee_9_40C408324CA26 ...(truncated)"
    DeformRig_Debug=(upperarm_debug_18_#4=False,wrist_debug_20_#6=False,lat_debug_26_#9=False,groinGlute_debug_27_#7=False)
    DeformRig_Value="(upperarm_up_18_#10=(Rotation=(X=0.0,Y=-0.0,Z=0.0,W=1.0),Translation=(X=0.0,Y=0.0,Z=6),Scale3D=(X=0.70,Y=1.0,Z=1.0)),upperarm_fwd_fr_19_#11=(Rotation=(X=0,Y=0,Z=0,W=1),Translation=(X=0,Y=10,Z=0),Sca ...(truncated)"
    ExecuteContext=@RigUnit_BeginExecution.ExecutePin
   )
   > Deform Rig Additions = RigVMFunctionReferenceNode()
   > Deform Middle Finger = RigVMFunctionReferenceNode()
   > Deform Toes = RigVMFunctionReferenceNode()
   > Deform Skirt = RigVMFunctionReferenceNode(Cloth Enabled=RigVMVariableNode() {variable=Cloth Enabled}.Value, RBAN Enabled=RigVMVariableNode() {variable=RBAN Enabled}.Value)
   > Deform Jacket = RigVMFunctionReferenceNode()
  ]}
]
hierarchy [
 root {loc=<0 0 0>, rot=<0 0 0 1>}
 pelvis {parent=root, loc=<0 -0.7071 93.1191>, rot=<0.0127 0.707 -0.0127 -0.707>}
 spine_01 {parent=pelvis, loc=<6.7494 -0.6388 0>, rot=<0 0 0.034 -0.9994>}
 spine_02 {parent=spine_01, loc=<7.2518 -0.035 0>, rot=<0 0 0.0238 -0.9997>}
 spine_03 {parent=spine_02, loc=<6.0104 0 0>, rot=<0 0 0.0149 -0.9999>}
 spine_04 {parent=spine_03, loc=<8.1024 0 0>, rot=<0 0 -0.0132 -0.9999>}
 spine_05 {parent=spine_04, loc=<10.0146 0 0>, rot=<0 0 -0.0015 -1>}
 neck_01 {parent=spine_05, loc=<11.3138 0.1529 0>, rot=<0 0 -0.2003 -0.9797>}
 neck_02 {parent=neck_01, loc=<4.5627 0 0>, rot=<0 0 0.0504 -0.9987>}
 head {parent=neck_02, loc=<5.3914 0 0>, rot=<0 0 0.115 -0.9934>}
 attach_cape {parent=spine_05, loc=<9.017 -11.714 0>, rot=<0.0616 -0.7044 -0.0616 -0.7044>}
 attach_backpack {parent=spine_05, loc=<-0.123 -13.473 0>, rot=<-0.0093 -0.707 0.0093 -0.707>}
 clavicle_l {parent=spine_05, loc=<4.3575 -0.1466 -2.3223>, rot=<0.0465 -0.7289 -0.009 -0.6829>}
 upperarm_l {parent=clavicle_l, loc=<15.5111 0 0>, rot=<0.0226 -0.3234 -0.0088 -0.9459>}
 lowerarm_l {parent=upperarm_l, loc=<26.6514 0 0>, rot=<0.0391 -0.0099 -0.2453 -0.9686>}
 hand_l {parent=lowerarm_l, loc=<22.3491 0 0>, rot=<-0.7081 -0.0507 0.0975 -0.6975>}
 thumb_01_l {parent=hand_l, loc=<1.3118 -0.3191 -2.459>, rot=<0.6546 -0.4042 0.0664 -0.6354>}
 thumb_02_l {parent=thumb_01_l, loc=<5.9336 -0.0448 0.0401>, rot=<-0.0037 0.0213 0.298 -0.9543>}
 thumb_03_l {parent=thumb_02_l, loc=<2.2223 0.3682 -0.0907>, rot=<0.0118 -0.013 0.0066 -0.9998>}
 index_metacarpal_l {parent=hand_l, loc=<1.6535 0.0183 -2.96>, rot=<0.0158 -0.0124 -0.0173 -0.9996>}
 index_01_l {parent=index_metacarpal_l, loc=<8.3008 0.0357 -0.0524>, rot=<-0.0044 0.0485 0.0924 -0.9945>}
 index_02_l {parent=index_01_l, loc=<3.3587 -0.3004 0.0452>, rot=<-0.002 0.0027 0.3225 -0.9466>}
 index_03_l {parent=index_02_l, loc=<2.5578 -0.123 -0.0252>, rot=<-0.002 0.0067 0.2558 -0.9667>}
 middle_metacarpal_l {parent=hand_l, loc=<2.1856 -0.1725 -0.5984>, rot=<0.006 0.0168 -0.0329 -0.9993>}
 middle_01_l {parent=middle_metacarpal_l, loc=<7.3275 -0.2447 0.0129>, rot=<-0.0208 0.0042 0.0858 -0.9961>}
 middle_02_l {parent=middle_01_l, loc=<4.0123 0.1401 0.1175>, rot=<-0.0218 0.0258 0.3512 -0.9357>}
 middle_03_l {parent=middle_02_l, loc=<2.6077 0.0534 0.1373>, rot=<0.0082 0.0188 0.1708 -0.9851>}
 deform_middle_01_l {parent=middle_metacarpal_l, loc=<7.3571 -0.0483 0.9391>, rot=<-0.021 0.0029 0.1452 -0.9892>}
 deform_middle_02_l {parent=deform_middle_01_l, loc=<4.0123 0.1401 0.1175>, rot=<-0.0215 0.026 0.341 -0.9394>}
 deform_middle_03_l {parent=deform_middle_02_l, loc=<2.8714 -0.0186 0.148>, rot=<0.0089 0.0185 0.1329 -0.9909>}
 ring_metacarpal_l {parent=hand_l, loc=<2.2074 -0.3625 1.3638>, rot=<-0.0284 0.0722 -0.0458 -0.9959>}
 ring_01_l {parent=ring_metacarpal_l, loc=<7.0295 -0.6197 -0.1446>, rot=<-0.0489 -0.0252 0.1892 -0.9804>}
 ring_02_l {parent=ring_01_l, loc=<3.6174 0.4647 -0.1027>, rot=<0 0 0.2806 -0.9598>}
 ring_03_l {parent=ring_02_l, loc=<2.4152 0.114 0.0199>, rot=<0 0 0.1189 -0.9929>}
 pinky_metacarpal_l {parent=hand_l, loc=<1.7616 -0.9753 3.3814>, rot=<-0.0397 0.0814 -0.069 -0.9935>}
 pinky_01_l {parent=pinky_metacarpal_l, loc=<7.1709 -0.4831 -0.1094>, rot=<-0.0654 -0.0035 0.2242 -0.9723>}
 pinky_02_l {parent=pinky_01_l, loc=<2.8274 0.0893 -0.1043>, rot=<0.0037 0.0011 0.2854 -0.9584>}
 pinky_03_l {parent=pinky_02_l, loc=<2.2046 0.0307 -0.0078>, rot=<-0.0002 0.0098 0.1445 -0.9895>}
 deform_wrist_dn_base_l {parent=hand_l, loc=<0 -2.2342 0>, rot=<0 0 0 -1>}
 deform_wrist_dn_l {parent=deform_wrist_dn_base_l, loc=<0 0 0>, rot=<0 0 0 1>}
 deform_wrist_up_base_l {parent=hand_l, loc=<0 2.2682 0>, rot=<0 0 0 -1>}
 deform_wrist_up_l {parent=deform_wrist_up_base_l, loc=<0 0 0>, rot=<0 0 0 1>}
 lowerarm_twist_02_l {parent=lowerarm_l, loc=<9.6199 0 0>, rot=<0 0 0 -1>}
 lowerarm_twist_01_l {parent=lowerarm_l, loc=<17.8549 0 0>, rot=<0 0 0 -1>}
 deform_elbow_in_base_l {parent=lowerarm_l, loc=<0.2112 5.259 0.8274>, rot=<0 0 0 -1>}
 deform_elbow_in_l {parent=deform_elbow_in_base_l, loc=<0 0 0>, rot=<0 0 0 1>}
 deform_elbow_out_base_l {parent=lowerarm_l, loc=<0 -4 0>, rot=<0 0 0 -1>}
 deform_elbow_out_l {parent=deform_elbow_out_base_l, loc=<0 0 0>, rot=<0 0 0 1>}
 upperarm_twist_01_l {parent=upperarm_l, loc=<0.0041 0 0>, rot=<0 0 0 -1>}
 deform_upperarm_bk_base_l {parent=upperarm_twist_01_l, loc=<1.0813 -7.2193 -1.9453>, rot=<0 0 0 1>}
 deform_upperarm_bk_l {parent=deform_upperarm_bk_base_l, loc=<0 0 0>, rot=<0 0 0 1>}
 deform_upperarm_fr_base_l {parent=upperarm_twist_01_l, loc=<-0.3691 7.0506 -2.765>, rot=<0 0 0 1>}
 deform_upperarm_fr_l {parent=deform_upperarm_fr_base_l, loc=<0 0 0>, rot=<0 0 0 1>}
 deform_upperarm_up_base_l {parent=upperarm_twist_01_l, loc=<-3.2216 -1.1391 3.3311>, rot=<0 0 0 1>}
 deform_upperarm_up_l {parent=deform_upperarm_up_base_l, loc=<0 0 0>, rot=<0 0 0 1>}
 upperarm_twist_02_l {parent=upperarm_l, loc=<13.7941 0 0>, rot=<0 0 0 -1>}
 clavicle_r {parent=spine_05, loc=<4.3575 -0.1466 2.3223>, rot=<0.7289 0.0465 -0.6829 0.009>}
 upperarm_r {parent=clavicle_r, loc=<-15.5111 0 0>, rot=<0.0226 -0.3234 -0.0088 -0.9459>}
 lowerarm_r {parent=upperarm_r, loc=<-26.6514 0 0>, rot=<0.0391 -0.0099 -0.2453 -0.9686>}
 hand_r {parent=lowerarm_r, loc=<-22.3491 0 0>, rot=<-0.7081 -0.0507 0.0975 -0.6975>}
 thumb_01_r {parent=hand_r, loc=<-1.3118 0.3191 2.459>, rot=<0.6546 -0.4042 0.0664 -0.6354>}
 thumb_02_r {parent=thumb_01_r, loc=<-5.9336 0.0448 -0.0401>, rot=<-0.0037 0.0213 0.298 -0.9543>}
 thumb_03_r {parent=thumb_02_r, loc=<-2.2223 -0.3682 0.0907>, rot=<0.0118 -0.013 0.0066 -0.9998>}
 index_metacarpal_r {parent=hand_r, loc=<-1.6535 -0.0183 2.96>, rot=<0.0158 -0.0124 -0.0173 -0.9996>}
 index_01_r {parent=index_metacarpal_r, loc=<-8.3008 -0.0357 0.0524>, rot=<-0.0044 0.0485 0.0924 -0.9945>}
 index_02_r {parent=index_01_r, loc=<-3.3587 0.3004 -0.0452>, rot=<-0.002 0.0027 0.3225 -0.9466>}
 index_03_r {parent=index_02_r, loc=<-2.5578 0.123 0.0252>, rot=<-0.002 0.0067 0.2558 -0.9667>}
 middle_metacarpal_r {parent=hand_r, loc=<-2.1856 0.1725 0.5984>, rot=<0.006 0.0168 -0.0329 -0.9993>}
 middle_01_r {parent=middle_metacarpal_r, loc=<-7.3275 0.2447 -0.0129>, rot=<-0.0208 0.0042 0.0858 -0.9961>}
 middle_02_r {parent=middle_01_r, loc=<-4.0123 -0.1401 -0.1175>, rot=<-0.0218 0.0258 0.3512 -0.9357>}
 middle_03_r {parent=middle_02_r, loc=<-2.6077 -0.0534 -0.1373>, rot=<0.0082 0.0188 0.1708 -0.9851>}
 deform_middle_01_r {parent=middle_metacarpal_r, loc=<-7.3572 0.0484 -0.9391>, rot=<-0.021 0.0029 0.1452 -0.9892>}
 deform_middle_02_r {parent=deform_middle_01_r, loc=<-4.0123 -0.1401 -0.1175>, rot=<-0.0215 0.026 0.341 -0.9394>}
 deform_middle_03_r {parent=deform_middle_02_r, loc=<-2.8713 0.0186 -0.148>, rot=<0.0089 0.0185 0.1329 -0.9909>}
 ring_metacarpal_r {parent=hand_r, loc=<-2.2074 0.3625 -1.3638>, rot=<-0.0284 0.0722 -0.0458 -0.9959>}
 ring_01_r {parent=ring_metacarpal_r, loc=<-7.0295 0.6197 0.1446>, rot=<-0.0489 -0.0252 0.1892 -0.9804>}
 ring_02_r {parent=ring_01_r, loc=<-3.6174 -0.4647 0.1027>, rot=<0 0 0.2806 -0.9598>}
 ring_03_r {parent=ring_02_r, loc=<-2.4152 -0.114 -0.0199>, rot=<0 0 0.1189 -0.9929>}
 pinky_metacarpal_r {parent=hand_r, loc=<-1.7616 0.9753 -3.3814>, rot=<-0.0397 0.0814 -0.069 -0.9935>}
 pinky_01_r {parent=pinky_metacarpal_r, loc=<-7.1709 0.4831 0.1094>, rot=<-0.0654 -0.0035 0.2242 -0.9723>}
 pinky_02_r {parent=pinky_01_r, loc=<-2.8274 -0.0893 0.1043>, rot=<0.0037 0.0011 0.2854 -0.9584>}
 pinky_03_r {parent=pinky_02_r, loc=<-2.2046 -0.0307 0.0078>, rot=<-0.0002 0.0098 0.1445 -0.9895>}
 deform_wrist_dn_base_r {parent=hand_r, loc=<-0.0004 2.2345 -0.0001>, rot=<0 0 0 -1>}
 deform_wrist_dn_r {parent=deform_wrist_dn_base_r, loc=<0 0 0>, rot=<0 0 0 1>}
 deform_wrist_up_base_r {parent=hand_r, loc=<0.0003 -2.2684 0.0001>, rot=<0 0 0 -1>}
 deform_wrist_up_r {parent=deform_wrist_up_base_r, loc=<0 0 0>, rot=<0 0 0 1>}
 lowerarm_twist_02_r {parent=lowerarm_r, loc=<-9.6199 0 0>, rot=<0 0 0 -1>}
 lowerarm_twist_01_r {parent=lowerarm_r, loc=<-17.8549 0 0>, rot=<0 0 0 -1>}
 deform_elbow_in_base_r {parent=lowerarm_r, loc=<-0.2111 -5.2591 -0.8275>, rot=<0 0 0 -1>}
 deform_elbow_in_r {parent=deform_elbow_in_base_r, loc=<-0.0001 0 0.0002>, rot=<0 0 0 1>}
 deform_elbow_out_base_r {parent=lowerarm_r, loc=<0 4 -0.0001>, rot=<0 0 0 -1>}
 deform_elbow_out_r {parent=deform_elbow_out_base_r, loc=<0 0 0>, rot=<0 0 0 1>}
 upperarm_twist_01_r {parent=upperarm_r, loc=<-0.0041 0 0>, rot=<0 0 0 -1>}
 deform_upperarm_bk_base_r {parent=upperarm_twist_01_r, loc=<-1.0814 7.2192 1.9455>, rot=<0 0 0 1>}
 deform_upperarm_bk_r {parent=deform_upperarm_bk_base_r, loc=<0 0 0>, rot=<0 0 0 1>}
 deform_upperarm_fr_base_r {parent=upperarm_twist_01_r, loc=<0.3689 -7.0506 2.7652>, rot=<0 0 0 1>}
 deform_upperarm_fr_r {parent=deform_upperarm_fr_base_r, loc=<0 0 0>, rot=<0 0 0 1>}
 deform_upperarm_up_base_r {parent=upperarm_twist_01_r, loc=<3.2214 1.1391 -3.3308>, rot=<0 0 0 1>}
 deform_upperarm_up_r {parent=deform_upperarm_up_base_r, loc=<0 0 0>, rot=<0 0 0 -1>}
 upperarm_twist_02_r {parent=upperarm_r, loc=<-13.7941 0 0>, rot=<0 0 0 -1>}
 dyn_jacket_front_l_MSTR {parent=spine_05, loc=<0.4507 9.5457 -7.9611>, rot=<-0.2107 -0.9247 0.1123 0.2966>}
 dyn_jacket_front_l_01 {parent=dyn_jacket_front_l_MSTR, loc=<0 0 0>, rot=<0 0 0 1>}
 dyn_jacket_front_l_02 {parent=dyn_jacket_front_l_01, loc=<5.2894 0 0>, rot=<0.0076 0.1165 0.1363 -0.9838>}
 dyn_jacket_front_side_l_MSTR {parent=spine_05, loc=<-3.7241 5.384 -12.7459>, rot=<-0.1313 -0.8515 0.478 0.1706>}
 dyn_jacket_front_side_l_01 {parent=dyn_jacket_front_side_l_MSTR, loc=<0 0 0>, rot=<0 0 0 1>}
 dyn_jacket_front_side_l_02 {parent=dyn_jacket_front_side_l_01, loc=<6.0742 0 -0.0003>, rot=<0 -0.0001 -0.0088 -1>}
 dyn_jacket_side_l_MSTR {parent=spine_05, loc=<-5.7571 -0.7146 -13.1474>, rot=<-0.0768 -0.7331 0.6724 -0.0679>}
 dyn_jacket_side_l_01 {parent=dyn_jacket_side_l_MSTR, loc=<0 0 0>, rot=<0 0 0 1>}
 dyn_jacket_side_l_02 {parent=dyn_jacket_side_l_01, loc=<6.0742 0 -0.0003>, rot=<0 -0.0001 -0.0088 -1>}
 dyn_jacket_rear_l_MSTR {parent=spine_05, loc=<0.3369 -10.7837 -8.0912>, rot=<-0.0162 -0.0438 0.9982 -0.0378>}
 dyn_jacket_rear_l_01 {parent=dyn_jacket_rear_l_MSTR, loc=<0 0 0>, rot=<0 0 0 1>}
 dyn_jacket_rear_l_02 {parent=dyn_jacket_rear_l_01, loc=<7 0 0>, rot=<0 0 0 1>}
 dyn_jacket_rear_MSTR {parent=spine_05, loc=<0.0207 -10.3236 0.1169>, rot=<0.0017 -0.043 0.9983 -0.0401>}
 dyn_jacket_rear_01 {parent=dyn_jacket_rear_MSTR, loc=<0 0 0>, rot=<0 0 0 1>}
 dyn_jacket_rear_02 {parent=dyn_jacket_rear_01, loc=<7 0 0>, rot=<0 0 0 1>}
 dyn_jacket_rear_r_MSTR {parent=spine_05, loc=<0.3372 -10.7837 8.0912>, rot=<0.0438 -0.0162 -0.0378 -0.9982>}
 dyn_jacket_rear_r_01 {parent=dyn_jacket_rear_r_MSTR, loc=<0 0 0>, rot=<0 0 0 1>}
 dyn_jacket_rear_r_02 {parent=dyn_jacket_rear_r_01, loc=<-7.0006 0 0>, rot=<0 0 0 1>}
 dyn_jacket_side_r_MSTR {parent=spine_05, loc=<-5.7573 -0.7146 13.1474>, rot=<0.7331 -0.0768 -0.0679 -0.6724>}
 dyn_jacket_side_r_01 {parent=dyn_jacket_side_r_MSTR, loc=<0 0 0>, rot=<0 0 0 -1>}
 dyn_jacket_side_r_02 {parent=dyn_jacket_side_r_01, loc=<-6.0741 0 0.0003>, rot=<0 -0.0001 -0.0088 -1>}
 dyn_jacket_front_side_r_MSTR {parent=spine_05, loc=<-3.7244 5.384 12.7459>, rot=<0.8515 -0.1313 0.1706 -0.478>}
 dyn_jacket_front_side_r_01 {parent=dyn_jacket_front_side_r_MSTR, loc=<0 0 0>, rot=<0 0 0 1>}
 dyn_jacket_front_side_r_02 {parent=dyn_jacket_front_side_r_01, loc=<-6.0743 0 0.0002>, rot=<0 -0.0001 -0.0088 -1>}
 dyn_jacket_front_r_MSTR {parent=spine_05, loc=<4.0654 7.1338 8.1421>, rot=<0.8729 -0.2881 0.3748 -0.1209>}
 dyn_jacket_front_r_01 {parent=dyn_jacket_front_r_MSTR, loc=<0 0 0>, rot=<0 0 0 -1>}
 dyn_jacket_front_r_02 {parent=dyn_jacket_front_r_01, loc=<-5.2893 0 0.0001>, rot=<0.018 0.1153 0.0482 -0.992>}
 deform_pec_aim_base_l {parent=spine_04, loc=<7.8732 12.6388 -2.4123>, rot=<0.0102 -0.4602 0.3979 -0.7936>}
 deform_pec_aim_l {parent=deform_pec_aim_base_l, loc=<0 0 0>, rot=<0 0 0 1>}
 deform_pec_base_l {parent=deform_pec_aim_l, loc=<10.766 0 0>, rot=<0 0 0 1>}
 deform_pec_l {parent=deform_pec_base_l, loc=<0.055 0 0>, rot=<0 0 0 1>}
 deform_pec_aim_base_r {parent=spine_04, loc=<7.8727 12.6388 2.4123>, rot=<-0.4602 -0.0102 0.7936 0.3979>}
 deform_pec_aim_r {parent=deform_pec_aim_base_r, loc=<0 0 0>, rot=<0 0 0 1>}
 deform_pec_base_r {parent=deform_pec_aim_r, loc=<-10.7662 -0.0006 -0.0007>, rot=<0 0 0 1>}
 deform_pec_r {parent=deform_pec_base_r, loc=<0 0 0>, rot=<0 0 0 1>}
 dyn_skirt_front_l_MSTR {parent=spine_02, loc=<1.2342 11.5813 -7.0047>, rot=<-0.0179 -0.9955 0.0797 -0.0483>}
 dyn_skirt_front_l_01 {parent=dyn_skirt_front_l_MSTR, loc=<0 0 0>, rot=<0 0 0 1>}
 dyn_skirt_front_l_02 {parent=dyn_skirt_front_l_01, loc=<10 0 0>, rot=<0 -0.0001 0.058 -0.9983>}
 dyn_skirt_front_l_03 {parent=dyn_skirt_front_l_02, loc=<8 0 0>, rot=<0 0 0.0233 -0.9997>}
 dyn_skirt_front_l_04 {parent=dyn_skirt_front_l_03, loc=<8 0 0>, rot=<0 0 0 1>}
 dyn_skirt_side_l_MSTR {parent=spine_02, loc=<3.1247 2.082 -14.0746>, rot=<-0.0295 -0.7503 0.6546 -0.0883>}
 dyn_skirt_side_l_01 {parent=dyn_skirt_side_l_MSTR, loc=<0 0 0>, rot=<0 0 0 1>}
 dyn_skirt_side_l_02 {parent=dyn_skirt_side_l_01, loc=<10 0 0>, rot=<0 -0.0001 0.0258 -0.9997>}
 dyn_skirt_side_l_03 {parent=dyn_skirt_side_l_02, loc=<10 0 0>, rot=<0 0 -0.0059 -1>}
 dyn_skirt_side_l_04 {parent=dyn_skirt_side_l_03, loc=<10 0 0>, rot=<0 0 0 1>}
 dyn_skirt_rear_l_MSTR {parent=spine_02, loc=<4.5746 -9.0862 -8.3748>, rot=<-0.0469 -0.0699 0.9814 -0.1723>}
 dyn_skirt_rear_l_01 {parent=dyn_skirt_rear_l_MSTR, loc=<0 0 0>, rot=<0 0 0 1>}
 dyn_skirt_rear_l_02 {parent=dyn_skirt_rear_l_01, loc=<10 0 0>, rot=<0 -0.0001 0.1598 -0.9872>}
 dyn_skirt_rear_l_03 {parent=dyn_skirt_rear_l_02, loc=<10 0 0>, rot=<0 0 -0.0059 -1>}
 dyn_skirt_rear_l_04 {parent=dyn_skirt_rear_l_03, loc=<10 0 0>, rot=<0 0 0 1>}
 dyn_skirt_rear_MSTR {parent=spine_02, loc=<4.6696 -10.3925 -0.0299>, rot=<0.0038 0.0362 0.9825 -0.1828>}
 dyn_skirt_rear_01 {parent=dyn_skirt_rear_MSTR, loc=<0 0 0>, rot=<0 0 0 1>}
 dyn_skirt_rear_02 {parent=dyn_skirt_rear_01, loc=<10 0 0>, rot=<0 -0.0001 0.1598 -0.9872>}
 dyn_skirt_rear_03 {parent=dyn_skirt_rear_02, loc=<10 0 0>, rot=<0 0 -0.0059 -1>}
 dyn_skirt_rear_04 {parent=dyn_skirt_rear_03, loc=<10 0 0>, rot=<0 0 0 1>}
 dyn_skirt_front_MSTR {parent=spine_02, loc=<0.5686 13.1915 -0.0188>, rot=<0.0398 -0.9985 0.0364 -0.0015>}
 dyn_skirt_front_01 {parent=dyn_skirt_front_MSTR, loc=<0 0 0>, rot=<0 0 0 1>}
 dyn_skirt_front_r_MSTR {parent=spine_02, loc=<1.4782 9.7426 10.5517>, rot=<0.9955 -0.0179 -0.0483 -0.0797>}
 dyn_skirt_front_r_01 {parent=dyn_skirt_front_r_MSTR, loc=<0 0 0>, rot=<0 0 0 1>}
 dyn_skirt_side_r_MSTR {parent=spine_02, loc=<3.1251 2.082 14.0746>, rot=<0.7503 -0.0295 -0.0883 -0.6546>}
 dyn_skirt_side_r_01 {parent=dyn_skirt_side_r_MSTR, loc=<0 0 0>, rot=<0 0 0 1>}
 dyn_skirt_side_r_02 {parent=dyn_skirt_side_r_01, loc=<-10.0007 0.0001 0>, rot=<0 -0.0001 0.0258 -0.9997>}
 dyn_skirt_side_r_03 {parent=dyn_skirt_side_r_02, loc=<-9.9996 0 0>, rot=<0 0 -0.0059 -1>}
 dyn_skirt_rear_r_MSTR {parent=spine_02, loc=<4.5747 -9.0862 8.3748>, rot=<0.0699 -0.0469 -0.1723 -0.9814>}
 dyn_skirt_rear_r_01 {parent=dyn_skirt_rear_r_MSTR, loc=<0 0 0>, rot=<0 0 0 1>}
 dyn_skirt_rear_r_02 {parent=dyn_skirt_rear_r_01, loc=<-10.0001 0 0>, rot=<0 -0.0001 0.1598 -0.9872>}
 dyn_skirt_rear_r_03 {parent=dyn_skirt_rear_r_02, loc=<-10.0001 0 0>, rot=<0 0 -0.0059 -1>}
 dyn_skirt_rear_r_04 {parent=dyn_skirt_rear_r_03, loc=<-10 0 0>, rot=<0 0 0 -1>}
 thigh_l {parent=pelvis, loc=<-2.5693 -0.0109 -7.7501>, rot=<0.0012 0.0572 0.0155 -0.9982>}
 calf_l {parent=thigh_l, loc=<-39.0484 0 0>, rot=<0.0006 -0.0161 -0.0391 -0.9991>}
 foot_l {parent=calf_l, loc=<-42.561 0.1077 0.4292>, rot=<0.0418 -0.0455 0.0343 -0.9975>}
 ball_l {parent=foot_l, loc=<-6.1599 13.8264 0.6806>, rot=<0.0001 -0.0001 -0.7063 -0.7079>}
 ball_l_ctrl {controlType=EULER_TRANSFORM, parent=ball_l, type=Control, loc=<0 0 0>, rot=<0 0 0 1>}
 deform_toe_l_outer {parent=foot_l, loc=<-7.0967 7.4935 -4.9834>, rot=<0.1492 -0.1495 -0.6904 -0.692>}
 deform_toe_l_mid {parent=foot_l, loc=<-7.0963 9.0184 0.3426>, rot=<-0.0248 0.0249 -0.7058 -0.7075>}
 deform_toe_l_inner {parent=foot_l, loc=<-7.0962 6.2076 6.101>, rot=<-0.2049 0.2053 -0.6759 -0.6775>}
 calf_twist_02_l {parent=calf_l, loc=<-18.8064 -0.0967 0.0397>, rot=<0.0237 0.0095 -0.0035 -0.9997>}
 calf_twist_01_l {parent=calf_l, loc=<-35.813 -0.0967 0.0397>, rot=<0.0237 0.0095 -0.0035 -0.9997>}
 deform_knee_in_base_l {parent=calf_l, loc=<1.298 -5.2957 -0.0496>, rot=<0 0 0 -1>}
 deform_knee_in_l {parent=deform_knee_in_base_l, loc=<0 0 0>, rot=<0 0 0 1>}
 deform_knee_out_base_l {parent=calf_l, loc=<0 4 0>, rot=<0 0 0 -1>}
 deform_knee_out_l {parent=deform_knee_out_base_l, loc=<0 0 0>, rot=<0 0 0 1>}
 thigh_twist_01_l {parent=thigh_l, loc=<-21.6303 0 0>, rot=<0.0013 0 0 -1>}
 deform_groin_base_l {parent=thigh_l, loc=<-2.1908 8.1142 -1.7442>, rot=<0 0 0 -1>}
 deform_groin_l {parent=deform_groin_base_l, loc=<0 0 0>, rot=<0 0 0 1>}
 deform_thigh_inner_base_l {parent=thigh_l, loc=<-8.5733 0.2866 6.1491>, rot=<0 0 0 -1>}
 deform_thigh_inner_l {parent=deform_thigh_inner_base_l, loc=<0 0 0>, rot=<0 0 0 1>}
 deform_hip_base_l {parent=thigh_l, loc=<3.3614 0.2865 -10.12>, rot=<0 0 0 -1>}
 deform_hip_l {parent=deform_hip_base_l, loc=<0 0 0>, rot=<0 0 0 1>}
 thigh_r {parent=pelvis, loc=<-2.5693 -0.0109 7.7501>, rot=<0.0572 -0.0012 0.9982 0.0155>}
 calf_r {parent=thigh_r, loc=<39.0484 0 0>, rot=<0.0006 -0.0161 -0.0391 -0.9991>}
 foot_r {parent=calf_r, loc=<42.561 -0.1077 -0.4292>, rot=<0.0418 -0.0455 0.0343 -0.9975>}
 ball_r {parent=foot_r, loc=<6.1599 -13.8264 -0.6806>, rot=<0.0001 -0.0001 -0.7063 -0.7079>}
 ball_r_ctrl {controlType=EULER_TRANSFORM, parent=ball_r, type=Control, loc=<0 0 0>, rot=<0 0 0 1>}
 deform_toe_r_outer {parent=foot_r, loc=<7.0967 -7.4935 4.9834>, rot=<0.1492 -0.1495 -0.6904 -0.692>}
 deform_toe_r_mid {parent=foot_r, loc=<7.0963 -9.0184 -0.3427>, rot=<-0.0248 0.0249 -0.7058 -0.7075>}
 deform_toe_r_inner {parent=foot_r, loc=<7.0962 -6.2076 -6.101>, rot=<-0.2049 0.2053 -0.6759 -0.6775>}
 calf_twist_02_r {parent=calf_r, loc=<18.8064 0.0967 -0.0397>, rot=<0.0237 0.0095 -0.0035 -0.9997>}
 calf_twist_01_r {parent=calf_r, loc=<35.813 0.0967 -0.0397>, rot=<0.0237 0.0095 -0.0035 -0.9997>}
 deform_knee_in_base_r {parent=calf_r, loc=<-1.298 5.2957 0.0496>, rot=<0 0 0 -1>}
 deform_knee_in_r {parent=deform_knee_in_base_r, loc=<0 0 0>, rot=<0 0 0 1>}
 deform_knee_out_base_r {parent=calf_r, loc=<0 -4 0>, rot=<0 0 0 -1>}
 deform_knee_out_r {parent=deform_knee_out_base_r, loc=<0 0 0>, rot=<0 0 0 1>}
 thigh_twist_01_r {parent=thigh_r, loc=<21.6303 0 0>, rot=<0.0013 0 0 -1>}
 deform_groin_base_r {parent=thigh_r, loc=<2.1908 -8.1142 1.7442>, rot=<0 0 0 -1>}
 deform_groin_r {parent=deform_groin_base_r, loc=<0 0 0>, rot=<0 0 0 1>}
 deform_thigh_inner_base_r {parent=thigh_r, loc=<8.5733 -0.2865 -6.1491>, rot=<0 0 0 -1>}
 deform_thigh_inner_r {parent=deform_thigh_inner_base_r, loc=<0 0 0>, rot=<0 0 0 1>}
 deform_hip_base_r {parent=thigh_r, loc=<-3.3613 -0.2865 10.12>, rot=<0 0 0 -1>}
 deform_hip_r {parent=deform_hip_base_r, loc=<0 0 0>, rot=<0 0 0 1>}
 deform_glute_base_l {parent=pelvis, loc=<-8.8064 -8.8222 -8.4287>, rot=<0.0275 0.0501 -0.4472 -0.8926>}
 deform_glute_l {parent=deform_glute_base_l, loc=<0 0 0>, rot=<0 0 0 1>}
 deform_glute_base_r {parent=pelvis, loc=<-8.8062 -8.8222 8.4287>, rot=<0.0501 -0.0275 0.8926 -0.4472>}
 deform_glute_r {parent=deform_glute_base_r, loc=<0 0 0>, rot=<0 0 0 1>}
 dyn_skirt_front_l_04_ctrl {controlType=EULER_TRANSFORM, parent=pelvis, type=Control, loc=<0 0 0>, rot=<0 0 0 1>}
 dyn_skirt_side_l_04_ctrl {controlType=EULER_TRANSFORM, parent=pelvis, type=Control, loc=<0 0 0>, rot=<0 0 0 1>}
 dyn_skirt_rear_l_04_ctrl {controlType=EULER_TRANSFORM, parent=pelvis, type=Control, loc=<0 0 0>, rot=<0 0 0 1>}
 dyn_skirt_rear_04_ctrl {controlType=EULER_TRANSFORM, parent=pelvis, type=Control, loc=<0 0 0>, rot=<0 0 0 1>}
 dyn_skirt_side_r_03_ctrl {controlType=EULER_TRANSFORM, parent=pelvis, type=Control, loc=<0 0 0>, rot=<0 0 0 1>}
 dyn_skirt_rear_r_04_ctrl {controlType=EULER_TRANSFORM, parent=pelvis, type=Control, loc=<0 0 0>, rot=<0 0 0 1>}
]
variables [
 Cloth Enabled {type=bool, default=False}
 RBAN Enabled {type=bool, default=True}
 Source Middle Fingers {
  type="TArray<FRigElementKey>"
  default=((Type=Bone,Name="middle_01_l"),(Type=Bone,Name="middle_02_l"),(Type=Bone,Name="middle_03_l"),(Type=Bone,Name="middle_01_r"),(Type=Bone,Name="middle_02_r"),(Type=Bone,Name="middle_03_r"))
 }
 Target Middle Fingers {
  type="TArray<FRigElementKey>"
  default=((Type=Bone,Name="deform_middle_01_l"),(Type=Bone,Name="deform_middle_02_l"),(Type=Bone,Name="deform_middle_03_l"),(Type=Bone,Name="deform_middle_01_r"),(Type=Bone,Name="deform_middle_02_r"),(Type=Bone,Name="deform_middle_03_r"))
 }
]
