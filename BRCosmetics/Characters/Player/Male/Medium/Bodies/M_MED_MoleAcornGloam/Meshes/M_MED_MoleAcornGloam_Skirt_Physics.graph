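# Physics asset that sets up the dynamic skirt of the MoleAcornGloam medium male player body.
# Capsules on spine, pelvis and legs accompany front, side and rear skirt bone chains of uneven length, the front center and right having one segment each.
# Chains hang from spine_02 via angularly free master bodies, with tighter swing and twist limits below.
# Angularly free cross-links tie neighboring rear chains at each segment, plus the left front and side tips.
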PhysicsAsset /BRCosmetics/Characters/Player/Male/Medium/Bodies/M_MED_MoleAcornGloam/Meshes/M_MED_MoleAcornGloam_Skirt_Physics
bodies {
 calf_l [{shapeType=Capsule, length=22.6016, radius=9.6105, center=<-21.0051 -1.1404 0.5245>, rotation=<rot 0 89.3693 -89.7307>}]
 calf_r [{shapeType=Capsule, length=22.2455, radius=9.5089, center=<21.3453 1.2542 -0.3894>, rotation=<rot 0 89.2914 -89.9795>}]
 dyn_skirt_front_01 [{shapeType=Capsule, length=8, radius=5, shapeName=dyn_skirt_front_01_capsule}]
 dyn_skirt_front_MSTR [{shapeType=Capsule, length=0.505, radius=0.505, shapeName=dyn_skirt_front_MSTR_capsule, rotation=<rot 0 -90 0>}]
 dyn_skirt_front_l_01 [{shapeType=Capsule, length=8.5625, radius=4.4375, shapeName=dyn_skirt_front_l_01_capsule, center=<4 6.5 0>}]
 dyn_skirt_front_l_02 [{shapeType=Capsule, length=12.4375, radius=6.0625, shapeName=dyn_skirt_front_l_02_capsule, center=<4 7 0>, rotation=<rot 0 0.0001 -1.7383>}]
 dyn_skirt_front_l_03 [{shapeType=Capsule, length=12, radius=6.625, shapeName=dyn_skirt_front_l_03_capsule, center=<4 6 0>}]
 dyn_skirt_front_l_04 [
  {shapeType=Capsule, length=11.3125, radius=7.3, shapeName=dyn_skirt_front_l_04_capsule, center=<4 5 0>, rotation=<rot -23.1731 0 0>}
  {
   shapeType=Capsule
   length=16.5
   radius=7.25
   shapeName=dyn_skirt_front_l_04_capsule
   center=<10.1872 1.7684 10.3989>
   rotation=<rot -3.9427 2.4717 -32.1198>
  }
 ]
 dyn_skirt_front_l_MSTR [{shapeType=Capsule, length=0.505, radius=0.505, shapeName=dyn_skirt_front_l_MSTR_capsule, rotation=<rot 0 -90 0>}]
 dyn_skirt_front_r_01 [{shapeType=Capsule, length=8, radius=5, shapeName=dyn_skirt_front_r_01_capsule}]
 dyn_skirt_front_r_MSTR [{shapeType=Capsule, length=0.505, radius=0.505, shapeName=dyn_skirt_front_r_MSTR_capsule, rotation=<rot 0 -90 0>}]
 dyn_skirt_rear_01 [{shapeType=Capsule, length=16.625, radius=5, shapeName=dyn_skirt_rear_01_capsule, center=<4 5.5 0>}]
 dyn_skirt_rear_02 [{shapeType=Capsule, length=16.625, radius=5, shapeName=dyn_skirt_rear_02_capsule, center=<4 5 0>}]
 dyn_skirt_rear_03 [{shapeType=Capsule, length=16.625, radius=5, shapeName=dyn_skirt_rear_03_capsule, center=<4 5 0>}]
 dyn_skirt_rear_04 [{shapeType=Capsule, length=16.625, radius=5, shapeName=dyn_skirt_rear_04_capsule, center=<4 5 0>}]
 dyn_skirt_rear_MSTR [{shapeType=Capsule, length=0.505, radius=0.505, shapeName=dyn_skirt_rear_MSTR_capsule, rotation=<rot 0 -90 0>}]
 dyn_skirt_rear_l_01 [{shapeType=Capsule, length=19.0625, radius=5, shapeName=dyn_skirt_rear_l_01_capsule, center=<4 5.8209 0>}]
 dyn_skirt_rear_l_02 [{shapeType=Capsule, length=19.0625, radius=5, shapeName=dyn_skirt_rear_l_02_capsule, center=<4 5 0>}]
 dyn_skirt_rear_l_03 [
  {shapeType=Capsule, length=19.0625, radius=5, shapeName=dyn_skirt_rear_l_03_capsule, center=<4 5 0>}
  {
   shapeType=Capsule
   length=16.8125
   radius=5.0625
   shapeName=dyn_skirt_rear_l_03_capsule
   center=<5.363 1.3087 -9.3351>
   rotation=<rot 0 0 41.9662>
  }
 ]
 dyn_skirt_rear_l_04 [{shapeType=Capsule, length=19.0625, radius=5, shapeName=dyn_skirt_rear_l_04_capsule, center=<4 5.112 0>}]
 dyn_skirt_rear_l_MSTR [{shapeType=Capsule, length=0.505, radius=0.505, shapeName=dyn_skirt_rear_l_MSTR_capsule, rotation=<rot 0 -90 0>}]
 dyn_skirt_rear_r_01 [{shapeType=Capsule, length=18.25, radius=5, shapeName=dyn_skirt_rear_r_01_capsule, center=<-4 -5.5 0>}]
 dyn_skirt_rear_r_02 [{shapeType=Capsule, length=18.25, radius=5, shapeName=dyn_skirt_rear_r_02_capsule, center=<-4 -5 0>}]
 dyn_skirt_rear_r_03 [
  {shapeType=Capsule, length=18.25, radius=5, shapeName=dyn_skirt_rear_r_03_capsule, center=<-4 -5 0>}
  {
   shapeType=Capsule
   length=17.1875
   radius=5.3125
   shapeName=dyn_skirt_rear_r_03_capsule
   center=<-3.0952 -1.7837 7.0756>
   rotation=<rot 0 -0.0001 41.616>
  }
 ]
 dyn_skirt_rear_r_04 [{shapeType=Capsule, length=18.25, radius=5, shapeName=dyn_skirt_rear_r_04_capsule, center=<-4 -5 0>}]
 dyn_skirt_rear_r_MSTR [{shapeType=Capsule, length=0.505, radius=0.505, shapeName=dyn_skirt_rear_r_MSTR_capsule, rotation=<rot 0 -90 0>}]
 dyn_skirt_side_l_01 [{shapeType=Capsule, length=20.5, radius=5, shapeName=dyn_skirt_side_l_01_capsule, center=<4 6 0>}]
 dyn_skirt_side_l_02 [{shapeType=Capsule, length=20.5, radius=5, shapeName=dyn_skirt_side_l_02_capsule, center=<4 5 0>}]
 dyn_skirt_side_l_03 [{shapeType=Capsule, length=20.5, radius=5, shapeName=dyn_skirt_side_l_03_capsule, center=<4 5 0>}]
 dyn_skirt_side_l_04 [{shapeType=Capsule, length=18.3125, radius=7.1875, shapeName=dyn_skirt_side_l_04_capsule, center=<6.5049 5 0>}]
 dyn_skirt_side_l_MSTR [{shapeType=Capsule, length=0.505, radius=0.505, shapeName=dyn_skirt_side_l_MSTR_capsule, rotation=<rot 0 -90 0>}]
 dyn_skirt_side_r_01 [{shapeType=Capsule, length=18.5625, radius=5, shapeName=dyn_skirt_side_r_01_capsule, center=<-4 -5.5 0>}]
 dyn_skirt_side_r_02 [{shapeType=Capsule, length=18.5625, radius=5, shapeName=dyn_skirt_side_r_02_capsule, center=<-4 -6 0>}]
 dyn_skirt_side_r_03 [{shapeType=Capsule, length=17.5625, radius=6, shapeName=dyn_skirt_side_r_03_capsule, center=<-4 -5 0>}]
 dyn_skirt_side_r_MSTR [{shapeType=Capsule, length=0.505, radius=0.505, shapeName=dyn_skirt_side_r_MSTR_capsule, rotation=<rot 0 -90 0>}]
 foot_l [{shapeType=Capsule, length=12.1082, radius=9.9, center=<-1.9628 4.1293 0.3057>, rotation=<rot 0 8.1409 -87.3289>}]
 foot_r [{shapeType=Capsule, length=12.1048, radius=9.9038, center=<2.0131 -4.1152 -0.3065>, rotation=<rot 0 8.0001 -87.2777>}]
 pelvis [{shapeType=Capsule, length=8.0269, radius=13.3723, center=<2.8275 0 0>, rotation=<rot 0 -90 0>}]
 spine_02 [{shapeType=Capsule, length=5.5266, radius=13.0201, center=<0 0.7769 0>, rotation=<rot 0 -90 -1.6628>}]
 thigh_l [
  {shapeType=Capsule, length=22, radius=10.8777, center=<-5.1213 -0.0069 0.4112>, rotation=<rot 0.3363 -87.1657 -89.8628>}
  {
   shapeType=Capsule
   length=20.4375
   radius=9.5625
   shapeName=thigh_l_capsule
   center=<-26.7388 0.0268 0.8814>
   rotation=<rot 84.6031 3.2079 2.0803>
  }
 ]
 thigh_r [{shapeType=Capsule, length=35.5625, radius=10.7527, center=<18.825 0.9956 -0.4459>, rotation=<rot 0.336 -87.0419 90.1379>}]
}
constraints [
 {bone1Name=spine_02, bone2Name=pelvis, swing1LimitDegrees=45, swing2LimitDegrees=45, twistLimitDegrees=45}
 {bone1Name=thigh_l, bone2Name=pelvis, swing1LimitDegrees=45, swing2LimitDegrees=45, twistLimitDegrees=45}
 {bone1Name=calf_l, bone2Name=thigh_l, swing1LimitDegrees=45, swing2LimitDegrees=45, twistLimitDegrees=45}
 {bone1Name=foot_l, bone2Name=calf_l, swing1LimitDegrees=45, swing2LimitDegrees=45, twistLimitDegrees=45}
 {bone1Name=thigh_r, bone2Name=pelvis, swing1LimitDegrees=45, swing2LimitDegrees=45, twistLimitDegrees=45}
 {bone1Name=calf_r, bone2Name=thigh_r, swing1LimitDegrees=45, swing2LimitDegrees=45, twistLimitDegrees=45}
 {bone1Name=foot_r, bone2Name=calf_r, swing1LimitDegrees=45, swing2LimitDegrees=45, twistLimitDegrees=45}
 {
  bone1Name=dyn_skirt_front_MSTR
  bone2Name=spine_02
  swing1LimitDegrees=45
  swing1Motion=Free
  swing2LimitDegrees=45
  swing2Motion=Free
  twistLimitDegrees=45
  twistMotion=Free
 }
 {bone1Name=dyn_skirt_front_01, bone2Name=dyn_skirt_front_MSTR, swing1LimitDegrees=24, swing2LimitDegrees=10, twistLimitDegrees=10}
 {
  bone1Name=dyn_skirt_front_l_MSTR
  bone2Name=spine_02
  swing1LimitDegrees=45
  swing1Motion=Free
  swing2LimitDegrees=45
  swing2Motion=Free
  twistLimitDegrees=45
  twistMotion=Free
 }
 {bone1Name=dyn_skirt_front_l_01, bone2Name=dyn_skirt_front_l_MSTR, swing1LimitDegrees=14, swing2LimitDegrees=8, twistLimitDegrees=8}
 {bone1Name=dyn_skirt_front_l_02, bone2Name=dyn_skirt_front_l_01, swing1LimitDegrees=20, swing2LimitDegrees=16, twistLimitDegrees=8}
 {bone1Name=dyn_skirt_front_l_03, bone2Name=dyn_skirt_front_l_02, swing1LimitDegrees=20, swing2LimitDegrees=20, twistLimitDegrees=12}
 {bone1Name=dyn_skirt_front_l_04, bone2Name=dyn_skirt_front_l_03, swing1LimitDegrees=20, swing2LimitDegrees=20, twistLimitDegrees=12}
 {
  bone1Name=dyn_skirt_front_r_MSTR
  bone2Name=spine_02
  swing1LimitDegrees=45
  swing1Motion=Free
  swing2LimitDegrees=45
  swing2Motion=Free
  twistLimitDegrees=45
  twistMotion=Free
 }
 {bone1Name=dyn_skirt_front_r_01, bone2Name=dyn_skirt_front_r_MSTR, swing1LimitDegrees=14, swing2LimitDegrees=8, twistLimitDegrees=8}
 {
  bone1Name=dyn_skirt_rear_MSTR
  bone2Name=spine_02
  swing1LimitDegrees=45
  swing1Motion=Free
  swing2LimitDegrees=45
  swing2Motion=Free
  twistLimitDegrees=45
  twistMotion=Free
 }
 {bone1Name=dyn_skirt_rear_01, bone2Name=dyn_skirt_rear_MSTR, swing1LimitDegrees=14, swing2LimitDegrees=8, twistLimitDegrees=8}
 {bone1Name=dyn_skirt_rear_02, bone2Name=dyn_skirt_rear_01, swing1LimitDegrees=16, swing2LimitDegrees=8, twistLimitDegrees=8}
 {bone1Name=dyn_skirt_rear_03, bone2Name=dyn_skirt_rear_02, swing1LimitDegrees=24, swing2LimitDegrees=8, twistLimitDegrees=8}
 {bone1Name=dyn_skirt_rear_04, bone2Name=dyn_skirt_rear_03, swing1LimitDegrees=24, swing2LimitDegrees=8, twistLimitDegrees=8}
 {
  bone1Name=dyn_skirt_rear_l_MSTR
  bone2Name=spine_02
  swing1LimitDegrees=45
  swing1Motion=Free
  swing2LimitDegrees=45
  swing2Motion=Free
  twistLimitDegrees=45
  twistMotion=Free
 }
 {bone1Name=dyn_skirt_rear_l_01, bone2Name=dyn_skirt_rear_l_MSTR, swing1LimitDegrees=14, swing2LimitDegrees=8, twistLimitDegrees=8}
 {bone1Name=dyn_skirt_rear_l_02, bone2Name=dyn_skirt_rear_l_01, swing1LimitDegrees=14, swing2LimitDegrees=8, twistLimitDegrees=8}
 {bone1Name=dyn_skirt_rear_l_03, bone2Name=dyn_skirt_rear_l_02, swing1LimitDegrees=16, swing2LimitDegrees=8, twistLimitDegrees=8}
 {bone1Name=dyn_skirt_rear_l_04, bone2Name=dyn_skirt_rear_l_03, swing1LimitDegrees=16, swing2LimitDegrees=8, twistLimitDegrees=8}
 {
  bone1Name=dyn_skirt_rear_r_MSTR
  bone2Name=spine_02
  swing1LimitDegrees=45
  swing1Motion=Free
  swing2LimitDegrees=45
  swing2Motion=Free
  twistLimitDegrees=45
  twistMotion=Free
 }
 {bone1Name=dyn_skirt_rear_r_01, bone2Name=dyn_skirt_rear_r_MSTR, swing1LimitDegrees=14, swing2LimitDegrees=8, twistLimitDegrees=8}
 {bone1Name=dyn_skirt_rear_r_02, bone2Name=dyn_skirt_rear_r_01, swing1LimitDegrees=14, swing2LimitDegrees=8, twistLimitDegrees=8}
 {bone1Name=dyn_skirt_rear_r_03, bone2Name=dyn_skirt_rear_r_02, swing1LimitDegrees=16, swing2LimitDegrees=8, twistLimitDegrees=8}
 {bone1Name=dyn_skirt_rear_r_04, bone2Name=dyn_skirt_rear_r_03, swing1LimitDegrees=16, swing2LimitDegrees=8, twistLimitDegrees=8}
 {
  bone1Name=dyn_skirt_side_l_MSTR
  bone2Name=spine_02
  swing1LimitDegrees=45
  swing1Motion=Free
  swing2LimitDegrees=45
  swing2Motion=Free
  twistLimitDegrees=45
  twistMotion=Free
 }
 {bone1Name=dyn_skirt_side_l_01, bone2Name=dyn_skirt_side_l_MSTR, swing1LimitDegrees=14, swing2LimitDegrees=8, twistLimitDegrees=8}
 {bone1Name=dyn_skirt_side_l_02, bone2Name=dyn_skirt_side_l_01, swing1LimitDegrees=14, swing2LimitDegrees=8, twistLimitDegrees=8}
 {bone1Name=dyn_skirt_side_l_03, bone2Name=dyn_skirt_side_l_02, swing1LimitDegrees=28, swing2LimitDegrees=28, twistLimitDegrees=8}
 {bone1Name=dyn_skirt_side_l_04, bone2Name=dyn_skirt_side_l_03, swing1LimitDegrees=28, swing2LimitDegrees=28, twistLimitDegrees=8}
 {
  bone1Name=dyn_skirt_side_r_MSTR
  bone2Name=spine_02
  swing1LimitDegrees=45
  swing1Motion=Free
  swing2LimitDegrees=45
  swing2Motion=Free
  twistLimitDegrees=45
  twistMotion=Free
 }
 {bone1Name=dyn_skirt_side_r_01, bone2Name=dyn_skirt_side_r_MSTR, swing1LimitDegrees=14, swing2LimitDegrees=8, twistLimitDegrees=8}
 {bone1Name=dyn_skirt_side_r_02, bone2Name=dyn_skirt_side_r_01, swing1LimitDegrees=14, swing2LimitDegrees=8, twistLimitDegrees=8}
 {bone1Name=dyn_skirt_side_r_03, bone2Name=dyn_skirt_side_r_02, swing1LimitDegrees=14, swing2LimitDegrees=8, twistLimitDegrees=8}
 {
  bone1Name=dyn_skirt_rear_01
  bone2Name=dyn_skirt_rear_l_01
  swing1LimitDegrees=45
  swing1Motion=Free
  swing2LimitDegrees=45
  swing2Motion=Free
  twistLimitDegrees=45
  twistMotion=Free
 }
 {
  bone1Name=dyn_skirt_rear_02
  bone2Name=dyn_skirt_rear_l_02
  swing1LimitDegrees=45
  swing1Motion=Free
  swing2LimitDegrees=45
  swing2Motion=Free
  twistLimitDegrees=45
  twistMotion=Free
 }
 {
  bone1Name=dyn_skirt_rear_03
  bone2Name=dyn_skirt_rear_l_03
  swing1LimitDegrees=45
  swing1Motion=Free
  swing2LimitDegrees=45
  swing2Motion=Free
  twistLimitDegrees=45
  twistMotion=Free
 }
 {
  bone1Name=dyn_skirt_rear_04
  bone2Name=dyn_skirt_rear_l_04
  swing1LimitDegrees=45
  swing1Motion=Free
  swing2LimitDegrees=45
  swing2Motion=Free
  twistLimitDegrees=45
  twistMotion=Free
 }
 {
  bone1Name=dyn_skirt_rear_01
  bone2Name=dyn_skirt_rear_r_01
  swing1LimitDegrees=45
  swing1Motion=Free
  swing2LimitDegrees=45
  swing2Motion=Free
  twistLimitDegrees=45
  twistMotion=Free
 }
 {
  bone1Name=dyn_skirt_rear_02
  bone2Name=dyn_skirt_rear_r_02
  swing1LimitDegrees=45
  swing1Motion=Free
  swing2LimitDegrees=45
  swing2Motion=Free
  twistLimitDegrees=45
  twistMotion=Free
 }
 {
  bone1Name=dyn_skirt_rear_03
  bone2Name=dyn_skirt_rear_r_03
  swing1LimitDegrees=45
  swing1Motion=Free
  swing2LimitDegrees=45
  swing2Motion=Free
  twistLimitDegrees=45
  twistMotion=Free
 }
 {
  bone1Name=dyn_skirt_rear_04
  bone2Name=dyn_skirt_rear_r_04
  swing1LimitDegrees=45
  swing1Motion=Free
  swing2LimitDegrees=45
  swing2Motion=Free
  twistLimitDegrees=45
  twistMotion=Free
 }
 {
  bone1Name=dyn_skirt_front_l_04
  bone2Name=dyn_skirt_side_l_04
  swing1LimitDegrees=45
  swing1Motion=Free
  swing2LimitDegrees=45
  swing2Motion=Free
  twistLimitDegrees=45
  twistMotion=Free
 }
]
