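# Physics asset of cloth colliders for the MoleAcornGloam medium male player body cosmetic, named for use in cloth simulation.
# Three capsules on pelvis, spine_01 and spine_02 cover the lower torso.
# Thigh, calf and foot bodies on both sides list no collision shapes.
# Constraints link the spine, left leg and right calf-to-foot with all angular motion free; right thigh-to-pelvis and calf-to-thigh links are absent.
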PhysicsAsset /BRCosmetics/Characters/Player/Male/Medium/Bodies/M_MED_MoleAcornGloam/Meshes/M_MED_MoleAcornGloam_ClothColliders
bodies {
 calf_l []
 calf_r []
 foot_l []
 foot_r []
 pelvis [{shapeType=Capsule, length=8.2081, radius=13.2263, center=<-0.4695 -0.6224 0>}]
 spine_01 [{shapeType=Capsule, length=5.505, radius=13.0675, shapeName=spine_01_capsule, rotation=<rot 0 -90 0>}]
 spine_02 [{
   shapeType=Capsule
   length=5.912
   radius=11.3983
   shapeName=spine_02_capsule
   center=<3.7364 1.3792 -0.0882>
   rotation=<rot 10.5776 -90 -0.8982>
  }]
 thigh_l []
 thigh_r []
}
constraints [
 {
  bone1Name=thigh_l
  bone2Name=pelvis
  swing1LimitDegrees=45
  swing1Motion=Free
  swing2LimitDegrees=45
  swing2Motion=Free
  twistLimitDegrees=45
  twistMotion=Free
 }
 {
  bone1Name=foot_l
  bone2Name=calf_l
  swing1LimitDegrees=45
  swing1Motion=Free
  swing2LimitDegrees=45
  swing2Motion=Free
  twistLimitDegrees=45
  twistMotion=Free
 }
 {
  bone1Name=foot_r
  bone2Name=calf_r
  swing1LimitDegrees=45
  swing1Motion=Free
  swing2LimitDegrees=45
  swing2Motion=Free
  twistLimitDegrees=45
  twistMotion=Free
 }
 {
  bone1Name=calf_l
  bone2Name=thigh_l
  swing1LimitDegrees=45
  swing1Motion=Free
  swing2LimitDegrees=45
  swing2Motion=Free
  twistLimitDegrees=45
  twistMotion=Free
 }
 {
  bone1Name=spine_02
  bone2Name=spine_01
  swing1LimitDegrees=45
  swing1Motion=Free
  swing2LimitDegrees=45
  swing2Motion=Free
  twistLimitDegrees=45
  twistMotion=Free
 }
 {
  bone1Name=spine_01
  bone2Name=pelvis
  swing1LimitDegrees=45
  swing1Motion=Free
  swing2LimitDegrees=45
  swing2Motion=Free
  twistLimitDegrees=45
  twistMotion=Free
 }
]
