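# Control Rig for the long ears of the MoleAcornGloam character's face accessory, with UseSplineEars choosing between two methods.
# Spline mode (default) weight-constrains guide bones to the body, adds head-pose and IsSkydiving offsets, then runs Spline IK on the dyn_ear chains through child controls.
# The alternative aims the ear roots at spine_04, applies head pose-reader rotations and adds harmonic waving driven by EarHarmonicsSpeed.
# Setup snaps controls to initial ear bone transforms.
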ControlRigBlueprint /BRCosmetics/Characters/Player/Male/Medium/Bodies/M_MED_MoleAcornGloam/Meshes/Parts/M_MED_MoleAcornGloam_FaceAcc_CtrlRig
graphs [
 New Function_1_ContainedGraph {nodes [
   n1 = RigVMFunctionEntryNode()
   RigVMFunction_Sequence = &RigVMFunction_Sequence(ExecuteContext=@n1.ExecuteContext)
   RigVMFunction_Sequence_1 = RigVMAggregateNode(ExecuteContext=@RigVMFunction_Sequence.A)
   SplineIK = RigVMFunctionReferenceNode(
    Primary Axis=(X=1,Y=0,Z=0)
    Samples Per Segment=4
    Secondary Spline Direction=(X=0,Y=-10,Z=0)
    Spline Mode=BSpline
    Stretch=false
    Up Axis=(X=0,Y=0,Z=0)
    debug=true
    Controls=RigVMVariableNode() {variable=EarControls_L}.Value
    Bones=RigVMVariableNode() {variable=EarBones_L}.Value
    ExecuteContext=@RigVMFunction_Sequence_1.C
   )
   SplineIK_1 = RigVMFunctionReferenceNode(
    Primary Axis=(X=-1,Y=0,Z=0)
    Samples Per Segment=4
    Secondary Spline Direction=(X=0,Y=10,Z=0)
    Spline Mode=BSpline
    Stretch=false
    Up Axis=(X=0,Y=0,Z=0)
    debug=true
    Bones=RigVMVariableNode() {variable=EarBones_R}.Value
    Controls=RigVMVariableNode() {variable=EarControls_R}.Value
    ExecuteContext=@RigVMFunction_Sequence_1.D
   )
   Modify Transforms with Mirror = RigVMFunctionReferenceNode(
    Base Modify Data="((Item=(Type=Bone,Name=\"spline_ear_l_04\"),Transform=(Rotation=(X=0,Y=0,Z=0,W=-1),Translation=(X=0,Y=-2,Z=-1.5),Scale3D=(X=1,Y=1,Z=1))),(Item=(Type=Bone,Name=\"spline_ear_l_05\"),Transform=(Rotation=(X=0,Y=0,Z=-0.0667,W=0.9978),Translation=(X=0,Y=0,Z=-1.5),Sc ...(truncated)"
    Mirror Modify Data="((Item=(Type=Bone,Name=\"spline_ear_r_04\"),Transform=(Rotation=(X=0,Y=0,Z=0,W=1),Translation=(X=0,Y=0,Z=0),Scale3D=(X=1,Y=1,Z=1))),(Item=(Type=Bone,Name=\"spline_ear_r_05\"),Transform=(Rotation=(X=0,Y=0,Z=0,W=1),Translation=(X=0,Y=0,Z=0),Scale3D=(X=1,Y=1,Z=1) ...(truncated)"
    "Mirror Weight "=1
    Weight=1
    ExecuteContext=@RigVMFunction_Sequence_1.B
   )
   > n2 = &RigUnit_SphericalPoseReader(
    ActiveRegionScaleFactors=(PositiveWidth=3,NegativeWidth=1,PositiveHeight=2,NegativeHeight=2)
    ActiveRegionSize=0.1
    Debug=(bDrawDebug=true,bDraw2D=False,bDrawLocalAxes=False,DebugScale=25,DebugSegments=20,DebugThickness=0.25)
    DriverAxis=(X=1,Y=0,Z=0)
    DriverItem=(Type=Bone,Name="head")
    FalloffRegionScaleFactors=(PositiveWidth=1,NegativeWidth=1,PositiveHeight=1,NegativeHeight=1)
    FalloffSize=0.5
    FlipHeightScaling=False
    FlipWidthScaling=False
    OptionalParentItem=(Type=Bone,Name="spine_05")
    RotationOffset=(X=70,Y=0,Z=180)
   )
   > &RigUnit_ModifyTransforms(
    ItemToModify="((Item=(Type=Bone,Name=\"spline_ear_l_03\"),Transform=(Rotation=(X=0,Y=0,Z=0,W=1),Translation=(X=20,Y=3.5067,Z=-2),Scale3D=(X=1,Y=1,Z=1))),(Item=(Type=Bone,Name=\"spline_ear_r_03\"),Transform=(Rotation=(X=0,Y=0,Z=0,W=-1),Translation=(X=-20,Y=-2.8183,Z=2),Scale ...(truncated)"
    WeightMinimum=0
    Weight=@n2.OutputParam
   )
   RigVMFunction_MathBoolToFloat = &RigVMFunction_MathBoolToFloat(Value=RigVMVariableNode() {variable=IsSkydiving}.Value)
   > &RigUnit_ModifyTransforms(
    ItemToModify="((Item=(Type=Bone,Name=\"spline_ear_l_03\"),Transform=(Rotation=(X=0,Y=0,Z=0,W=1),Translation=(X=0,Y=0,Z=0),Scale3D=(X=1,Y=1,Z=1))),(Item=(Type=Bone,Name=\"spline_ear_l_04\"),Transform=(Rotation=(X=0,Y=0,Z=0,W=1),Translation=(X=0,Y=0,Z=0),Scale3D=(X=1,Y=1,Z=1) ...(truncated)"
    WeightMinimum=0
    Weight=@RigVMFunction_MathBoolToFloat.Result
   )
   &RigUnit_AimConstraintLocalSpaceOffset(
    Child=(Type=Bone,Name="dyn_ear_l_MSTR")
    Filter=(bX=True,bY=True,bZ=True)
    Parents=((Item=(Type=Bone,Name="spine_04"),Weight=1))
    Weight=1
    WorldUp=(Target=(X=0,Y=0,Z=1),Kind=Location,Space=(Type=Bone,Name="spine_04"))
    ExecutePin=@RigVMFunction_Sequence_1.A
   )
   > &RigUnit_AimConstraintLocalSpaceOffset(
    Child=(Type=Bone,Name="dyn_ear_r_MSTR")
    Filter=(bX=True,bY=True,bZ=True)
    Parents=((Item=(Type=Bone,Name="spine_04"),Weight=1))
    Weight=1
    WorldUp=(Target=(X=0,Y=0,Z=1),Kind=Location,Space=(Type=Bone,Name="spine_04"))
   )
   > &RigUnit_AimConstraintLocalSpaceOffset(
    Child=(Type=Bone,Name="spline_dyn_ear_l_MSTR")
    Filter=(bX=True,bY=True,bZ=True)
    Parents=((Item=(Type=Bone,Name="spine_04"),Weight=1))
    Weight=1
    WorldUp=(Target=(X=0,Y=0,Z=1),Kind=Location,Space=(Type=Bone,Name="spine_04"))
   )
   > &RigUnit_AimConstraintLocalSpaceOffset(
    Child=(Type=Bone,Name="spline_ear_r_MSTR")
    Filter=(bX=True,bY=True,bZ=True)
    Parents=((Item=(Type=Bone,Name="spine_04"),Weight=1))
    Weight=1
    WorldUp=(Target=(X=0,Y=0,Z=1),Kind=Location,Space=(Type=Bone,Name="spine_04"))
   )
   > &RigUnit_ParentConstraint(
    AdvancedSettings=(RotationOrderForFilter=XZY)
    Child=(Type=Bone,Name="spline_ear_l_03")
    Parents=((Item=(Type=Bone,Name="neck_01"),Weight=0.0578),(Item=(Type=Bone,Name="clavicle_l"),Weight=0.1022),(Item=(Type=Bone,Name="neck_02"),Weight=0.1309),(Item=(Type=Bone,Name="spine_05"),Weight=0.7091))
    Weight=1
   )
   > &RigUnit_ParentConstraint(
    AdvancedSettings=(RotationOrderForFilter=XZY)
    Child=(Type=Bone,Name="spline_ear_l_04")
    Parents="((Item=(Type=Bone,Name=\"deform_upperarm_bk_l\"),Weight=0.0052),(Item=(Type=Bone,Name=\"spine_04\"),Weight=0.0305),(Item=(Type=Bone,Name=\"upperarm_twist_01_l\"),Weight=0.0767),(Item=(Type=Bone,Name=\"deform_upperarm_up_l\"),Weight=0.0853),(Item=(Type=Bone,Name=\"s ...(truncated)"
    Weight=1
   )
   > &RigUnit_ParentConstraint(
    AdvancedSettings=(RotationOrderForFilter=XZY)
    Child=(Type=Bone,Name="spline_ear_l_05")
    Parents="((Item=(Type=Bone,Name=\"spine_02\"),Weight=0.0219),(Item=(Type=Bone,Name=\"spine_05\"),Weight=0.0463),(Item=(Type=Bone,Name=\"upperarm_twist_01_l\"),Weight=0.0767),(Item=(Type=Bone,Name=\"spine_04\"),Weight=0.1286),(Item=(Type=Bone,Name=\"deform_upperarm_bk_l\"),We ...(truncated)"
    Weight=1
   )
   > &RigUnit_ParentConstraint(
    AdvancedSettings=(RotationOrderForFilter=XZY)
    Child=(Type=Bone,Name="spline_ear_l_06")
    Parents="((Item=(Type=Bone,Name=\"spine_05\"),Weight=0.0148),(Item=(Type=Bone,Name=\"upperarm_twist_01_l\"),Weight=0.0337),(Item=(Type=Bone,Name=\"spine_01\"),Weight=0.0433),(Item=(Type=Bone,Name=\"deform_upperarm_bk_l\"),Weight=0.0713),(Item=(Type=Bone,Name=\"spine_02\"),We ...(truncated)"
    Weight=1
   )
   > &RigUnit_ParentConstraint(
    AdvancedSettings=(RotationOrderForFilter=XZY)
    Child=(Type=Bone,Name="spline_ear_l_07")
    Parents="((Item=(Type=Bone,Name=\"deform_upperarm_bk_l\"),Weight=0.0178),(Item=(Type=Bone,Name=\"pelvis\"),Weight=0.0593),(Item=(Type=Bone,Name=\"clavicle_l\"),Weight=0.0913),(Item=(Type=Bone,Name=\"spine_01\"),Weight=0.1075),(Item=(Type=Bone,Name=\"spine_04\"),Weight=0.1383 ...(truncated)"
    Weight=1
   )
   > &RigUnit_ParentConstraint(
    AdvancedSettings=(RotationOrderForFilter=XZY)
    Child=(Type=Bone,Name="spline_ear_l_08")
    Parents=((Item=(Type=Bone,Name="spine_03"),Weight=0.0072),(Item=(Type=Bone,Name="spine_01"),Weight=0.0107),(Item=(Type=Bone,Name="spine_02"),Weight=0.0119),(Item=(Type=Bone,Name="pelvis"),Weight=0.8834))
    Weight=1
   )
   > &RigUnit_ParentConstraint(
    AdvancedSettings=(RotationOrderForFilter=XZY)
    Child=(Type=Bone,Name="spline_ear_r_03")
    Parents=((Item=(Type=Bone,Name="neck_01"),Weight=0.0581),(Item=(Type=Bone,Name="clavicle_r"),Weight=0.1014),(Item=(Type=Bone,Name="neck_02"),Weight=0.1316),(Item=(Type=Bone,Name="spine_05"),Weight=0.7089))
    Weight=1
   )
   > &RigUnit_ParentConstraint(
    AdvancedSettings=(RotationOrderForFilter=XZY)
    Child=(Type=Bone,Name="spline_ear_r_04")
    Parents="((Item=(Type=Bone,Name=\"deform_upperarm_bk_r\"),Weight=0.0053),(Item=(Type=Bone,Name=\"spine_04\"),Weight=0.0305),(Item=(Type=Bone,Name=\"upperarm_twist_01_r\"),Weight=0.0768),(Item=(Type=Bone,Name=\"deform_upperarm_up_r\"),Weight=0.0854),(Item=(Type=Bone,Name=\"s ...(truncated)"
    Weight=1
   )
   > &RigUnit_ParentConstraint(
    AdvancedSettings=(RotationOrderForFilter=XZY)
    Child=(Type=Bone,Name="spline_ear_r_05")
    Parents="((Item=(Type=Bone,Name=\"spine_02\"),Weight=0.0219),(Item=(Type=Bone,Name=\"spine_05\"),Weight=0.0463),(Item=(Type=Bone,Name=\"upperarm_twist_01_r\"),Weight=0.0767),(Item=(Type=Bone,Name=\"spine_04\"),Weight=0.1287),(Item=(Type=Bone,Name=\"deform_upperarm_bk_r\"),We ...(truncated)"
    Weight=1
   )
   > &RigUnit_ParentConstraint(
    AdvancedSettings=(RotationOrderForFilter=XZY)
    Child=(Type=Bone,Name="spline_ear_r_06")
    Parents="((Item=(Type=Bone,Name=\"spine_05\"),Weight=0.0148),(Item=(Type=Bone,Name=\"upperarm_twist_01_r\"),Weight=0.0337),(Item=(Type=Bone,Name=\"spine_01\"),Weight=0.0433),(Item=(Type=Bone,Name=\"deform_upperarm_bk_r\"),Weight=0.0713),(Item=(Type=Bone,Name=\"spine_02\"),We ...(truncated)"
    Weight=1
   )
   > &RigUnit_ParentConstraint(
    AdvancedSettings=(RotationOrderForFilter=XZY)
    Child=(Type=Bone,Name="spline_ear_r_07")
    Parents="((Item=(Type=Bone,Name=\"deform_upperarm_bk_r\"),Weight=0.0178),(Item=(Type=Bone,Name=\"pelvis\"),Weight=0.0593),(Item=(Type=Bone,Name=\"clavicle_r\"),Weight=0.0913),(Item=(Type=Bone,Name=\"spine_01\"),Weight=0.1075),(Item=(Type=Bone,Name=\"spine_04\"),Weight=0.1383 ...(truncated)"
    Weight=1
   )
   > &RigUnit_ParentConstraint(
    AdvancedSettings=(RotationOrderForFilter=XZY)
    Child=(Type=Bone,Name="spline_ear_r_08")
    Parents=((Item=(Type=Bone,Name="spine_03"),Weight=0.0072),(Item=(Type=Bone,Name="spine_01"),Weight=0.0107),(Item=(Type=Bone,Name="spine_02"),Weight=0.0119),(Item=(Type=Bone,Name="pelvis"),Weight=0.8834))
    Weight=1
   )
   RigVMFunctionReturnNode(ExecuteContext=@RigVMFunction_Sequence.B)
   &RigUnit_ModifyTransforms(
    ExecutePin=()
    ItemToModify="((Item=(Type=Bone,Name=\"spline_ear_l_03\"),Transform=(Rotation=(X=0,Y=0,Z=0,W=1),Translation=(X=0,Y=0,Z=0),Scale3D=(X=1,Y=1,Z=1))),(Item=(Type=Bone,Name=\"spline_ear_l_04\"),Transform=(Rotation=(X=0,Y=0,Z=0,W=-1),Translation=(X=0,Y=-5.4966,Z=-1.5),Scale3D=(X= ...(truncated)"
    Weight=1
    WeightMinimum=0
   )
  ]}
 New Function_ContainedGraph {nodes [
   RigVMFunctionEntryNode()
   > RigVMFunction_Sequence = RigVMAggregateNode()
   &RigUnit_AimConstraintLocalSpaceOffset(
    Child=(Type=Bone,Name="dyn_ear_l_MSTR")
    Filter=(bX=True,bY=True,bZ=True)
    Parents=((Item=(Type=Bone,Name="spine_04"),Weight=1))
    Weight=1
    WorldUp=(Target=(X=0,Y=0,Z=1),Kind=Location,Space=(Type=Bone,Name="spine_04"))
    ExecutePin=@RigVMFunction_Sequence.A
   )
   > &RigUnit_AimConstraintLocalSpaceOffset(
    Child=(Type=Bone,Name="dyn_ear_r_MSTR")
    Filter=(bX=True,bY=True,bZ=True)
    Parents=((Item=(Type=Bone,Name="spine_04"),Weight=1))
    Weight=1
    WorldUp=(Target=(X=0,Y=0,Z=1),Kind=Location,Space=(Type=Bone,Name="spine_04"))
   )
   > n1 = &RigUnit_SphericalPoseReader(
    ActiveRegionScaleFactors=(PositiveWidth=3,NegativeWidth=1,PositiveHeight=2,NegativeHeight=2)
    ActiveRegionSize=0.1
    Debug=(bDrawDebug=true,bDraw2D=False,bDrawLocalAxes=False,DebugScale=25,DebugSegments=20,DebugThickness=0.25)
    DriverAxis=(X=1,Y=0,Z=0)
    DriverItem=(Type=Bone,Name="head")
    FalloffRegionScaleFactors=(PositiveWidth=1,NegativeWidth=1,PositiveHeight=1,NegativeHeight=1)
    FalloffSize=0.3
    FlipHeightScaling=False
    FlipWidthScaling=False
    OptionalParentItem=(Type=Bone,Name="spine_05")
    RotationOffset=(X=70,Y=0,Z=180)
   )
   > &RigUnit_ModifyTransforms(
    ItemToModify=((Item=(Type=Bone,Name="dyn_ear_r_MSTR"),Transform=(Rotation=(X=0,Y=0,Z=-0.1908,W=0.9816),Translation=(X=0,Y=0,Z=0),Scale3D=(X=1,Y=1,Z=1))))
    WeightMinimum=0
    Weight=@n1.OutputParam
   )
   > n2 = &RigUnit_SphericalPoseReader(
    ActiveRegionScaleFactors=(PositiveWidth=3,NegativeWidth=1,PositiveHeight=2,NegativeHeight=2)
    ActiveRegionSize=0.1
    Debug=(bDrawDebug=true,bDraw2D=False,bDrawLocalAxes=False,DebugScale=25,DebugSegments=20,DebugThickness=0.25)
    DriverAxis=(X=1,Y=0,Z=0)
    DriverItem=(Type=Bone,Name="head")
    FalloffRegionScaleFactors=(PositiveWidth=1,NegativeWidth=1,PositiveHeight=1,NegativeHeight=1)
    FalloffSize=0.3
    FlipHeightScaling=False
    FlipWidthScaling=False
    OptionalParentItem=(Type=Bone,Name="spine_05")
    RotationOffset=(X=110,Y=0,Z=180)
   )
   > &RigUnit_ModifyTransforms(
    ItemToModify=((Item=(Type=Bone,Name="dyn_ear_l_MSTR"),Transform=(Rotation=(X=0,Y=0,Z=-0.1908,W=0.9816),Translation=(X=0,Y=0,Z=0),Scale3D=(X=1,Y=1,Z=1))))
    WeightMinimum=0
    Weight=@n2.OutputParam
   )
   RigVMFunctionReturnNode(ExecuteContext=@RigVMFunction_Sequence.C)
   n3 = RigVMVariableNode() {variable=EarHarmonicsSpeed}
   Clamp = &RigVMFunction_MathDoubleClamp(Value=@n3.Value)
   n4 = /Script/RigVM.RigVMFunction_MathVectorMul(B=(X=1,Y=1,Z=1), B.X=@n3.Value, B.Y=@n3.Value, B.Z=@n3.Value, A=RigVMVariableNode() {variable=ItemHarmonics_WaveSpeed}.Value) {is=struct RigVMUnitNode, template="Multiply::Execute(in A,in B,out Result)"}
   /Script/ControlRig.RigUnit_ItemHarmonics(
    RotationOrder=YZX
    Targets=((Item=(Type=Bone,Name="dyn_ear_l_MSTR"),Ratio=20))
    WaveAmplitude=(X=0,Y=0,Z=20)
    WaveEase=Linear
    WaveFrequency=(X=0,Y=-5,Z=12)
    WaveMinimum=0
    WaveNoise=(X=0,Y=0,Z=0)
    WaveOffset=(X=0,Y=1,Z=3)
    WaveMaximum=@Clamp.Result
    WaveSpeed=@n4.Result
    ExecutePin=@RigVMFunction_Sequence.B
   ) {is=struct RigVMUnitNode}
   > &RigUnit_ItemHarmonics(
    RotationOrder=YZX
    Targets=((Item=(Type=Bone,Name="dyn_ear_r_MSTR"),Ratio=20))
    WaveAmplitude=(X=0,Y=2,Z=20)
    WaveEase=Linear
    WaveFrequency=(X=0,Y=5,Z=12)
    WaveMinimum=0
    WaveNoise=(X=0,Y=0,Z=0)
    WaveOffset=(X=0,Y=20,Z=1.5)
    WaveMaximum=@Clamp.Result
    WaveSpeed=@n4.Result
   )
  ]}
 RigVMFunctionLibrary {nodes [
   Ears Normal = RigVMCollapseNode(ExecuteContext=())
   Ears Spline = RigVMCollapseNode(ExecuteContext=())
   set_pose_func = RigVMCollapseNode(ExecuteContext=(), bind_pose=(HierarchyTopologyVersion=-1,PoseHash=-1), target_pose=(HierarchyTopologyVersion=-1,PoseHash=-1))
  ]}
 RigVMFunction_Sequence_1_ContainedGraph {nodes [
   n1 = RigVMFunctionEntryNode()
   RigVMFunction_Sequence = &RigVMFunction_Sequence(ExecuteContext=@n1.ExecuteContext)
   n2 = &RigVMFunction_Sequence(ExecuteContext=@RigVMFunction_Sequence.B)
   RigVMFunctionReturnNode(A=@RigVMFunction_Sequence.A, B=@n2.A, C=@n2.B)
  ]}
 RigVMFunction_Sequence_2_ContainedGraph {nodes [
   n1 = RigVMFunctionEntryNode()
   n2 = &RigVMFunction_Sequence(ExecuteContext=@n1.ExecuteContext)
   n3 = &RigVMFunction_Sequence(ExecuteContext=@n2.B)
   n4 = &RigVMFunction_Sequence(ExecuteContext=@n3.B)
   n5 = &RigVMFunction_Sequence(ExecuteContext=@n4.B)
   RigVMFunctionReturnNode(A=@n2.A, B=@n3.A, C=@n4.A, D=@n5.A, E=@n5.B)
  ]}
 RigVMModel {nodes [
   RigUnit_BeginExecution = &RigUnit_BeginExecution()
   RigVMFunction_ControlFlowBranch = &RigVMFunction_ControlFlowBranch(ExecuteContext=@RigUnit_BeginExecution.ExecutePin, Condition=RigVMVariableNode() {variable=UseSplineEars}.Value)
   Ears Normal = RigVMFunctionReferenceNode(ExecuteContext=@RigVMFunction_ControlFlowBranch.False)
   Ears Spline = RigVMFunctionReferenceNode(ExecuteContext=@RigVMFunction_ControlFlowBranch.True)
  ]}
 RigVMModel Rig {nodes [
   n1 = &RigUnit_PrepareForExecution()
   RigVMFunction_Sequence = &RigVMFunction_Sequence(ExecuteContext=@n1.ExecutePin)
   For_Each = RigVMDispatchNode(Array=RigVMVariableNode() {variable=EarBones_L}.Value, ExecuteContext=@RigVMFunction_Sequence.B) {template="DISPATCH_RigVMDispatch_ArrayIterator(in Array,out Element,out Index,out Count,out Ratio)"}
   > &RigUnit_SetTransform(
    Weight=1
    bInitial=true
    Item=RigVMDispatchNode(Index=@For_Each.Index, Array=RigVMVariableNode() {variable=EarControls_L}.Value) {template="DISPATCH_RigVMDispatch_ArrayGetAtIndex(in Array,in Index,out Element)"}.Element
    Value=&RigUnit_GetTransform(bInitial=true, Item=@For_Each.Element).Transform
   )
   For_Each_1 = RigVMDispatchNode(ExecuteContext=@For_Each.Completed, Array=RigVMVariableNode() {variable=EarBones_R}.Value) {template="DISPATCH_RigVMDispatch_ArrayIterator(in Array,out Element,out Index,out Count,out Ratio)"}
   > &RigUnit_SetTransform(
    Weight=1
    bInitial=true
    Item=RigVMDispatchNode(Index=@For_Each_1.Index, Array=RigVMVariableNode() {variable=EarControls_R}.Value) {template="DISPATCH_RigVMDispatch_ArrayGetAtIndex(in Array,in Index,out Element)"}.Element
    Value=&RigUnit_GetTransform(bInitial=true, Item=@For_Each_1.Element).Transform
   )
   n2 = &RigUnit_ItemArray(Items=((Type=Bone,Name="dyn_ear_l_01"),(Type=Bone,Name="dyn_ear_l_02"),(Type=Bone,Name="dyn_ear_l_03"),(Type=Bone,Name="dyn_ear_l_04"),(Type=Bone,Name="dyn_ear_l_05"),(Type=Bone,Name="dyn_ear_l_06"),(Type=Bone,Name="dyn_ear_l_07")))
   RigVMVariableNode(Value=@n2.Items, ExecuteContext=@RigVMFunction_Sequence.A) {variable=EarBones_L}
   n3 = &RigUnit_ItemArray(Items=((Type=Bone,Name="dyn_ear_r_01"),(Type=Bone,Name="dyn_ear_r_02"),(Type=Bone,Name="dyn_ear_r_03"),(Type=Bone,Name="dyn_ear_r_04"),(Type=Bone,Name="dyn_ear_r_05"),(Type=Bone,Name="dyn_ear_r_06"),(Type=Bone,Name="dyn_ear_r_07")))
   > RigVMVariableNode(Value=@n3.Items) {variable=EarBones_R}
   n4 = &RigUnit_ItemArray(Items="((Type=Control,Name=\"dyn_ear_l_01_ctrl\"),(Type=Control,Name=\"dyn_ear_l_02_ctrl\"),(Type=Control,Name=\"dyn_ear_l_03_ctrl\"),(Type=Control,Name=\"dyn_ear_l_04_ctrl\"),(Type=Control,Name=\"dyn_ear_l_05_ctrl\"),(Type=Control,Name=\"dyn_ear_l_06_ctrl\"),(Type=Control,N ...(truncated)")
   > RigVMVariableNode(Value=@n4.Items) {variable=EarControls_L}
   n5 = &RigUnit_ItemArray(Items="((Type=Control,Name=\"dyn_ear_r_01_ctrl\"),(Type=Control,Name=\"dyn_ear_r_02_ctrl\"),(Type=Control,Name=\"dyn_ear_r_03_ctrl\"),(Type=Control,Name=\"dyn_ear_r_04_ctrl\"),(Type=Control,Name=\"dyn_ear_r_05_ctrl\"),(Type=Control,Name=\"dyn_ear_r_06_ctrl\"),(Type=Control,N ...(truncated)")
   > RigVMVariableNode(Value=@n5.Items) {variable=EarControls_R}
   &RigUnit_ItemArray(Items=((Type=Bone,Name="spline_ear_l_01"),(Type=Bone,Name="spline_ear_l_02"),(Type=Bone,Name="spline_ear_l_03"),(Type=Bone,Name="spline_ear_l_04"),(Type=Bone,Name="spline_ear_l_05"),(Type=Bone,Name="spline_ear_l_06"),(Type=Bone,Name="spline_ear_l_07")))
   &RigUnit_ItemArray(Items=((Type=Bone,Name="spline_ear_r_01"),(Type=Bone,Name="spline_ear_r_02"),(Type=Bone,Name="spline_ear_r_03"),(Type=Bone,Name="spline_ear_r_04"),(Type=Bone,Name="spline_ear_r_05"),(Type=Bone,Name="spline_ear_r_06"),(Type=Bone,Name="spline_ear_r_07")))
  ]}
 set_pose_func_ContainedGraph {nodes [
   n1 = RigVMFunctionEntryNode()
   SequenceAggregate = &RigVMFunction_Sequence(ExecuteContext=@n1.ExecuteContext)
   Subtract = &RigVMFunction_MathDoubleSub(A=1, B=@n1.pose_alpha)
   /Script/ControlRig.RigUnit_HierarchySetPoseItemArray(ElementType=All, ItemsToSet=(), Pose=@n1.bind_pose, ExecutePin=@SequenceAggregate.A, Weight=@Subtract.Result) {is=struct RigVMUnitNode}
   &RigUnit_HierarchySetPoseItemArray(ElementType=All, ItemsToSet=(), Weight=@n1.pose_alpha, Pose=@n1.target_pose, ExecutePin=@SequenceAggregate.B)
   > RigVMFunctionReturnNode()
  ]}
]
hierarchy [
 root {loc=<0 0 0>, rot=<0 0 0 1>}
 pelvis {parent=root, loc=<0 -0.7071 93.1191>, rot=<0.0127 0.707 -0.0127 -0.707>}
 spine_01 {parent=pelvis, loc=<6.7494 -0.6388 0>, rot=<0 0 0.034 -0.9994>}
 spine_02 {parent=spine_01, loc=<7.2518 -0.035 0>, rot=<0 0 0.0238 -0.9997>}
 spine_03 {parent=spine_02, loc=<6.0104 0 0>, rot=<0 0 0.0149 -0.9999>}
 spine_04 {parent=spine_03, loc=<8.1024 0 0>, rot=<0 0 -0.0132 -0.9999>}
 spine_05 {parent=spine_04, loc=<10.0146 0 0>, rot=<0 0 -0.0015 -1>}
 neck_01 {parent=spine_05, loc=<11.3138 0.1529 0>, rot=<0 0 -0.2003 -0.9797>}
 neck_02 {parent=neck_01, loc=<4.5627 0 0>, rot=<0 0 0.0504 -0.9987>}
 head {parent=neck_02, loc=<5.3914 0 0>, rot=<0 0 0.115 -0.9934>}
 dyn_ear_l_MSTR {parent=head, loc=<9.7636 -6.5443 -4.5386>, rot=<0.1371 -0.8728 0.2473 -0.3978>}
 dyn_ear_l_01 {parent=dyn_ear_l_MSTR, loc=<0 0 0>, rot=<0 0 0 1>}
 dyn_ear_l_02 {parent=dyn_ear_l_01, loc=<6.9561 0 0>, rot=<0.3967 -0.2244 -0.0315 -0.8896>}
 dyn_ear_l_03 {parent=dyn_ear_l_02, loc=<7.8089 0 0>, rot=<-0.1103 -0.0644 -0.0348 -0.9912>}
 dyn_ear_l_04 {parent=dyn_ear_l_03, loc=<9.7101 -0.8258 -0.0136>, rot=<-0.0079 -0.0046 -0.0681 -0.9976>}
 dyn_ear_l_05 {parent=dyn_ear_l_04, loc=<11.7598 0 0>, rot=<-0.0001 0.0097 -0.0081 -0.9999>}
 dyn_ear_l_06 {parent=dyn_ear_l_05, loc=<12.0408 0 0>, rot=<0 -0.0125 0 -0.9999>}
 dyn_ear_l_07 {parent=dyn_ear_l_06, loc=<8.6814 0 0>, rot=<0 -0.0162 0 -0.9999>}
 dyn_ear_l_08 {parent=dyn_ear_l_07, loc=<7.3803 0 0>, rot=<0 0 0 1>}
 dyn_ear_l_06_spike_a {parent=dyn_ear_l_06, loc=<1.3115 0 7.5576>, rot=<0 -0.0607 0 -0.9982>}
 dyn_ear_l_06_spike_b {parent=dyn_ear_l_06, loc=<0.9063 0 -6.8275>, rot=<0 0 0 1>}
 dyn_ear_l_09 {parent=dyn_ear_l_03, loc=<8.9205 -0.3214 -6.1431>, rot=<-0.0001 -0.1174 -0.0686 -0.9907>}
 dyn_ear_l_010 {parent=dyn_ear_l_09, loc=<12 0 0>, rot=<0 0.0388 0 -0.9992>}
 dyn_ear_r_MSTR {parent=head, loc=<9.7632 -6.5443 4.5386>, rot=<0.8728 0.1371 -0.3978 -0.2473>}
 dyn_ear_r_01 {parent=dyn_ear_r_MSTR, loc=<0 0 0>, rot=<0 0 0 1>}
 dyn_ear_r_02 {parent=dyn_ear_r_01, loc=<-6.9561 0 0>, rot=<0.3967 -0.2244 -0.0315 -0.8896>}
 dyn_ear_r_03 {parent=dyn_ear_r_02, loc=<-7.8086 0.0001 -0.0001>, rot=<-0.1103 -0.0644 -0.0348 -0.9912>}
 dyn_ear_r_04 {parent=dyn_ear_r_03, loc=<-9.7096 0.826 0.0136>, rot=<-0.0079 -0.0046 -0.0681 -0.9976>}
 dyn_ear_r_05 {parent=dyn_ear_r_04, loc=<-11.76 -0.0001 0>, rot=<-0.0001 0.0097 -0.0081 -0.9999>}
 dyn_ear_r_06 {parent=dyn_ear_r_05, loc=<-12.0413 -0.0001 0.0001>, rot=<0 -0.0125 0 -0.9999>}
 dyn_ear_r_07 {parent=dyn_ear_r_06, loc=<-8.6817 -0.0001 0.0001>, rot=<0 -0.0162 0 -0.9999>}
 dyn_ear_r_08 {parent=dyn_ear_r_07, loc=<-7.38 0.0001 0>, rot=<0 0 0 1>}
 dyn_ear_r_06_spike_a {parent=dyn_ear_r_06, loc=<-1.311 0.0001 -7.5577>, rot=<0 -0.0607 0 -0.9982>}
 dyn_ear_r_06_spike_b {parent=dyn_ear_r_06, loc=<-0.9057 0.0001 6.8274>, rot=<0 0 0 1>}
 dyn_ear_r_09 {parent=dyn_ear_r_03, loc=<-8.9203 0.3215 6.1431>, rot=<-0.0001 -0.1174 -0.0686 -0.9907>}
 dyn_ear_r_010 {parent=dyn_ear_r_09, loc=<-12.0002 -0.0001 0>, rot=<0 0.0388 0 -0.9992>}
 spline_dyn_ear_l_MSTR {parent=head, loc=<9.7636 -6.5443 -4.5386>, rot=<0.1371 -0.8728 0.2473 -0.3978>}
 spline_ear_l_01 {parent=spline_dyn_ear_l_MSTR, loc=<0 0 0>, rot=<0 0 0 1>}
 dyn_ear_l_01_ctrl {controlType=EULER_TRANSFORM, parent=spline_ear_l_01, type=Control, loc=<0 0 0>, rot=<0 0 0 1>}
 spline_ear_l_02 {parent=spline_ear_l_01, loc=<6.9561 0 0>, rot=<0.3967 -0.2244 -0.0315 -0.8896>}
 dyn_ear_l_02_ctrl {controlType=EULER_TRANSFORM, parent=spline_ear_l_02, type=Control, loc=<0 0 0>, rot=<0 0 0 1>}
 spline_ear_l_03 {parent=spline_ear_l_02, loc=<7.8089 0 0>, rot=<-0.1103 -0.0644 -0.0348 -0.9912>}
 dyn_ear_l_03_ctrl {controlType=EULER_TRANSFORM, parent=spline_ear_l_03, type=Control, loc=<0 0 0>, rot=<0 0 0 -1>}
 spline_ear_l_04 {parent=spline_ear_l_03, loc=<9.7101 -0.8258 -0.0136>, rot=<-0.0079 -0.0046 -0.0681 -0.9976>}
 dyn_ear_l_04_ctrl {controlType=EULER_TRANSFORM, parent=spline_ear_l_04, type=Control, loc=<0 0 0>, rot=<0 0 0 1>}
 spline_ear_l_05 {parent=spline_ear_l_04, loc=<11.7598 0 0>, rot=<-0.0001 0.0097 -0.0081 -0.9999>}
 dyn_ear_l_05_ctrl {controlType=EULER_TRANSFORM, parent=spline_ear_l_05, type=Control, loc=<0 0 0>, rot=<0 0 0 1>}
 spline_ear_l_06 {parent=spline_ear_l_05, loc=<12.0408 0 0>, rot=<0 -0.0125 0 -0.9999>}
 dyn_ear_l_06_ctrl {controlType=EULER_TRANSFORM, parent=spline_ear_l_06, type=Control, loc=<0 0 0>, rot=<0 0 0 1>}
 spline_ear_l_07 {parent=spline_ear_l_06, loc=<8.6814 0 0>, rot=<0 -0.0162 0 -0.9999>}
 dyn_ear_l_07_ctrl {controlType=EULER_TRANSFORM, parent=spline_ear_l_07, type=Control, loc=<0 0 0>, rot=<0 0 0 1>}
 spline_ear_l_08 {parent=spline_ear_l_07, loc=<7.3803 0 0>, rot=<0 0 0 1>}
 dyn_ear_l_08_ctrl {controlType=EULER_TRANSFORM, parent=spline_ear_l_08, type=Control, loc=<0 0 0>, rot=<0 0 0 1>}
 spline_ear_r_MSTR {parent=head, loc=<9.7632 -6.5443 4.5386>, rot=<0.8728 0.1371 -0.3978 -0.2473>}
 spline_ear_r_01 {parent=spline_ear_r_MSTR, loc=<0 0 0>, rot=<0 0 0 1>}
 dyn_ear_r_01_ctrl {controlType=EULER_TRANSFORM, parent=spline_ear_r_01, type=Control, loc=<0 0 0>, rot=<0 0 0 1>}
 spline_ear_r_02 {parent=spline_ear_r_01, loc=<-6.9561 0 0>, rot=<0.3967 -0.2244 -0.0315 -0.8896>}
 dyn_ear_r_02_ctrl {controlType=EULER_TRANSFORM, parent=spline_ear_r_02, type=Control, loc=<0 0 0>, rot=<0 0 0 1>}
 spline_ear_r_03 {parent=spline_ear_r_02, loc=<-7.8086 0.0001 -0.0001>, rot=<-0.1103 -0.0644 -0.0348 -0.9912>}
 dyn_ear_r_03_ctrl {controlType=EULER_TRANSFORM, parent=spline_ear_r_03, type=Control, loc=<0 0 0>, rot=<0 0 0 1>}
 spline_ear_r_04 {parent=spline_ear_r_03, loc=<-9.7096 0.826 0.0136>, rot=<-0.0079 -0.0046 -0.0681 -0.9976>}
 dyn_ear_r_04_ctrl {controlType=EULER_TRANSFORM, parent=spline_ear_r_04, type=Control, loc=<0 0 0>, rot=<0 0 0 1>}
 spline_ear_r_05 {parent=spline_ear_r_04, loc=<-11.76 -0.0001 0>, rot=<-0.0001 0.0097 -0.0081 -0.9999>}
 dyn_ear_r_05_ctrl {controlType=EULER_TRANSFORM, parent=spline_ear_r_05, type=Control, loc=<0 0 0>, rot=<0 0 0 1>}
 spline_ear_r_06 {parent=spline_ear_r_05, loc=<-12.0413 -0.0001 0.0001>, rot=<0 -0.0125 0 -0.9999>}
 dyn_ear_r_06_ctrl {controlType=EULER_TRANSFORM, parent=spline_ear_r_06, type=Control, loc=<0 0 0>, rot=<0 0 0 1>}
 spline_ear_r_07 {parent=spline_ear_r_06, loc=<-8.6817 -0.0001 0.0001>, rot=<0 -0.0162 0 -0.9999>}
 dyn_ear_r_07_ctrl {controlType=EULER_TRANSFORM, parent=spline_ear_r_07, type=Control, loc=<0 0 0>, rot=<0 0 0 1>}
 spline_ear_r_08 {parent=spline_ear_r_07, loc=<-7.38 0.0001 0>, rot=<0 0 0 1>}
 dyn_ear_r_08_ctrl {controlType=EULER_TRANSFORM, parent=spline_ear_r_08, type=Control, loc=<0 0 0>, rot=<0 0 0 1>}
 clavicle_l {parent=spine_05, loc=<4.3575 -0.1466 -2.3223>, rot=<0.0465 -0.7289 -0.009 -0.6829>}
 upperarm_l {parent=clavicle_l, loc=<15.5111 0 0>, rot=<0.0226 -0.3234 -0.0088 -0.9459>}
 lowerarm_l {parent=upperarm_l, loc=<26.6514 0 0>, rot=<0.0391 -0.0099 -0.2453 -0.9686>}
 upperarm_twist_01_l {parent=upperarm_l, loc=<0.0041 0 0>, rot=<0 0 0 -1>}
 deform_upperarm_bk_base_l {parent=upperarm_twist_01_l, loc=<1.0813 -7.2193 -1.9453>, rot=<0 0 0 -1>}
 deform_upperarm_bk_l {parent=deform_upperarm_bk_base_l, loc=<0 0 0>, rot=<0 0 0 1>}
 deform_upperarm_up_base_l {parent=upperarm_twist_01_l, loc=<-3.2216 -1.1391 3.3311>, rot=<0 0 0 1>}
 deform_upperarm_up_l {parent=deform_upperarm_up_base_l, loc=<0 0 0>, rot=<0 0 0 1>}
 clavicle_r {parent=spine_05, loc=<4.3575 -0.1466 2.3223>, rot=<0.7289 0.0465 -0.6829 0.009>}
 upperarm_r {parent=clavicle_r, loc=<-15.5111 0 0>, rot=<0.0226 -0.3234 -0.0088 -0.9459>}
 lowerarm_r {parent=upperarm_r, loc=<-26.6514 0 0>, rot=<0.0391 -0.0099 -0.2453 -0.9686>}
 upperarm_twist_01_r {parent=upperarm_r, loc=<-0.0041 0 0>, rot=<0 0 0 -1>}
 deform_upperarm_bk_base_r {parent=upperarm_twist_01_r, loc=<-1.0814 7.2192 1.9455>, rot=<0 0 0 -1>}
 deform_upperarm_bk_r {parent=deform_upperarm_bk_base_r, loc=<0 0 0>, rot=<0 0 0 1>}
 deform_upperarm_up_base_r {parent=upperarm_twist_01_r, loc=<3.2214 1.1391 -3.3308>, rot=<0 0 0 1>}
 deform_upperarm_up_r {parent=deform_upperarm_up_base_r, loc=<0 0 0>, rot=<0 0 0 -1>}
 attach_cape {parent=spine_05, loc=<9.017 -11.714 0>, rot=<0.0616 -0.7044 -0.0616 -0.7044>}
 attach_backpack {parent=spine_05, loc=<-0.123 -13.473 0>, rot=<-0.0093 -0.707 0.0093 -0.707>}
 dyn_jacket_rear_l_MSTR {parent=spine_05, loc=<0.3369 -10.7837 -8.0912>, rot=<-0.0162 -0.0438 0.9982 -0.0378>}
 dyn_jacket_rear_l_01 {parent=dyn_jacket_rear_l_MSTR, loc=<0 0 0>, rot=<0 0 0 1>}
 dyn_jacket_rear_l_02 {parent=dyn_jacket_rear_l_01, loc=<7 0 0>, rot=<0 0 0 1>}
 dyn_jacket_rear_MSTR {parent=spine_05, loc=<0.0207 -10.3236 0.1169>, rot=<0.0017 -0.043 0.9983 -0.0401>}
 dyn_jacket_rear_01 {parent=dyn_jacket_rear_MSTR, loc=<0 0 0>, rot=<0 0 0 1>}
 dyn_jacket_rear_02 {parent=dyn_jacket_rear_01, loc=<7 0 0>, rot=<0 0 0 1>}
 dyn_jacket_rear_r_MSTR {parent=spine_05, loc=<0.3372 -10.7837 8.0912>, rot=<0.0438 -0.0162 -0.0378 -0.9982>}
 dyn_jacket_rear_r_01 {parent=dyn_jacket_rear_r_MSTR, loc=<0 0 0>, rot=<0 0 0 1>}
 dyn_jacket_rear_r_02 {parent=dyn_jacket_rear_r_01, loc=<-7.0006 0 0>, rot=<0 0 0 1>}
 dyn_ear_upper_cape_collision {parent=spine_05, loc=<-0.1011 -16.2153 0>, rot=<0.0401 -0.9992 0 0>}
 dyn_skirt_rear_l_MSTR {parent=spine_02, loc=<4.5746 -9.0862 -8.3748>, rot=<-0.0469 -0.0699 0.9814 -0.1723>}
 dyn_skirt_rear_l_01 {parent=dyn_skirt_rear_l_MSTR, loc=<0 0 0>, rot=<0 0 0 1>}
 dyn_skirt_rear_l_02 {parent=dyn_skirt_rear_l_01, loc=<10 0 0>, rot=<0 -0.0001 0.1598 -0.9872>}
 dyn_skirt_rear_l_03 {parent=dyn_skirt_rear_l_02, loc=<10 0 0>, rot=<0 0 -0.0059 -1>}
 dyn_skirt_rear_r_MSTR {parent=spine_02, loc=<4.5747 -9.0862 8.3748>, rot=<0.0699 -0.0469 -0.1723 -0.9814>}
 dyn_skirt_rear_r_01 {parent=dyn_skirt_rear_r_MSTR, loc=<0 0 0>, rot=<0 0 0 1>}
 dyn_skirt_rear_r_02 {parent=dyn_skirt_rear_r_01, loc=<-10.0001 0 0>, rot=<0 -0.0001 0.1598 -0.9872>}
 dyn_skirt_rear_r_03 {parent=dyn_skirt_rear_r_02, loc=<-10.0001 0 0>, rot=<0 0 -0.0059 -1>}
 dyn_ear_lower_cape_collision {parent=spine_02, loc=<4.0379 -21.2545 0>, rot=<0 -1 0 0>}
]
variables [
 EarBones_L {
  type="TArray<FRigElementKey>"
  default="((Type=Bone,Name=\"dyn_ear_l_01\"),(Type=Bone,Name=\"dyn_ear_l_02\"),(Type=Bone,Name=\"dyn_ear_l_03\"),(Type=Bone,Name=\"dyn_ear_l_04\"),(Type=Bone,Name=\"dyn_ear_l_05\"),(Type=Bone,Name=\"dyn_ear_l_06\"),(Type=Bone,Name=\"dyn_ear_l_07\"),(Type=Bone,Name=\"dyn_ear_l_08\") ...(truncated)"
 }
 EarBones_R {
  type="TArray<FRigElementKey>"
  default="((Type=Bone,Name=\"dyn_ear_r_01\"),(Type=Bone,Name=\"dyn_ear_r_02\"),(Type=Bone,Name=\"dyn_ear_r_03\"),(Type=Bone,Name=\"dyn_ear_r_04\"),(Type=Bone,Name=\"dyn_ear_r_05\"),(Type=Bone,Name=\"dyn_ear_r_06\"),(Type=Bone,Name=\"dyn_ear_r_07\"),(Type=Bone,Name=\"dyn_ear_r_08\") ...(truncated)"
 }
 EarControls_L {
  type="TArray<FRigElementKey>"
  default="((Type=Control,Name=\"dyn_ear_l_01_ctrl\"),(Type=Control,Name=\"dyn_ear_l_02_ctrl\"),(Type=Control,Name=\"dyn_ear_l_03_ctrl\"),(Type=Control,Name=\"dyn_ear_l_04_ctrl\"),(Type=Control,Name=\"dyn_ear_l_05_ctrl\"),(Type=Control,Name=\"dyn_ear_l_06_ctrl\"),(Type=Control,N ...(truncated)"
 }
 EarControls_R {
  type="TArray<FRigElementKey>"
  default="((Type=Control,Name=\"dyn_ear_r_01_ctrl\"),(Type=Control,Name=\"dyn_ear_r_02_ctrl\"),(Type=Control,Name=\"dyn_ear_r_03_ctrl\"),(Type=Control,Name=\"dyn_ear_r_04_ctrl\"),(Type=Control,Name=\"dyn_ear_r_05_ctrl\"),(Type=Control,Name=\"dyn_ear_r_06_ctrl\"),(Type=Control,N ...(truncated)"
 }
 EarHarmonicsSpeed {type=double, default=0}
 IsSkydiving {type=bool, default=False}
 UseSplineEars {type=bool, default=True}
]
