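# Character-part AnimBP that drives secondary ear motion for the MoleAcornGloam face accessory.
# Copied mesh poses, with Body arms layered in, feed a Control Rig, then cached RigidBody or AnimDynamics branches, with a separate chain for non-cape backblings.
# The selected pose gets bone offsets and cape-collision bones constrained between both ears.
# The EventGraph sets gravity, RBAN settings, ear harmonics speed, cape-cloth state and front-end simulation settings.
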AnimBlueprint /BRCosmetics/Characters/Player/Male/Medium/Bodies/M_MED_MoleAcornGloam/Meshes/Parts/M_MED_MoleAcornGloam_FaceAcc_AnimBP
target_skeleton=&/M_MED_MoleAcornGloam_FaceAcc_Skeleton
graphs [
 AnimGraph {nodes [
   "Transform (Modify) Bone - Bone: dyn_ear_lower_cape_collision"(Alpha=1, Translation="11.237,4.222,-16.329", Rotation="0.221,0.112,28.498", Scale="1,1,1") {
    rotationMode=BMM_Additive
    rotationSpace=BCS_BoneSpace
    translationMode=BMM_Additive
    translationSpace=BCS_BoneSpace
    alphaBoolBlend {blendOption=Linear}
    boneToModify {boneName=dyn_ear_lower_cape_collision}
    rotation=<rot 0.221 0.112 28.498>
    translation=<11.237 4.222 -16.329>
   }
   n1 = Blend Poses by bool(
    BlendPose_0=Use cached pose 'Post RBAN'()
    BlendPose_1=Use cached pose 'Post Anim Dynamics Ears'()
    BlendTime_0=0
    BlendTime_1=0
    bActiveValue=Get bRigidBodyEnabled()
   ) {blendType=Linear}
   n2 = Use cached pose 'Post Backbling Anim Dynamics Ears'()
   n3 = Get isWearingBackbling()
   n4 = "Transform (Modify) Bone - Bone: dyn_ear_l_MSTR"!(
    ComponentPose=Local To Component(LocalPose=Blend Poses by bool(BlendPose_0=@n2, BlendPose_1=@n1, BlendTime_0=0, BlendTime_1=0, bActiveValue=@n3) {blendType=Linear})
    bAlphaBoolEnabled:bool=True
    Translation="0,0,0"
    Rotation="10,-10,10"
    Scale="1,1,1"
   ) -> (Pose) {
    alphaInputType=Bool
    rotationMode=BMM_Additive
    rotationSpace=BCS_BoneSpace
    translationSpace=BCS_BoneSpace
    alphaBoolBlend {blendOption=Linear}
    boneToModify {boneName=dyn_ear_l_MSTR}
    rotation=<rot 10 -10 10>
   }
   n5 = "Transform (Modify) Bone - Bone: dyn_ear_r_MSTR"!(ComponentPose=@n4, bAlphaBoolEnabled:bool=True, Translation="0,0,0", Rotation="10,-10,10", Scale="1,1,1") -> (Pose) {
    alphaInputType=Bool
    rotationMode=BMM_Additive
    rotationSpace=BCS_BoneSpace
    translationSpace=BCS_BoneSpace
    alphaBoolBlend {blendOption=Linear}
    boneToModify {boneName=dyn_ear_r_MSTR}
    rotation=<rot 10 -10 10>
   }
   n6 = "Constraint - dyn_ear_upper_cape_collision (TT)"!(ComponentPose=@n5, Alpha=1, ConstraintWeights_0=0.5, ConstraintWeights_1:float=0.5) -> (Pose) {
    alphaBoolBlend {blendOption=Linear}
    boneToModify {boneName=dyn_ear_upper_cape_collision}
    constraintSetup [{transformType=Translation, targetBone {boneName=dyn_ear_l_04}}, {transformType=Translation, targetBone {boneName=dyn_ear_r_04}}]
    constraintWeights [1,1]
   }
   n7 = "Constraint - dyn_ear_lower_cape_collision (TT)"!(ComponentPose=@n6, Alpha=1, ConstraintWeights_0=0.5, ConstraintWeights_1:float=0.5) -> (Pose) {
    alphaBoolBlend {blendOption=Linear}
    boneToModify {boneName=dyn_ear_lower_cape_collision}
    constraintSetup [{transformType=Translation, targetBone {boneName=dyn_ear_l_07}}, {transformType=Translation, targetBone {boneName=dyn_ear_r_07}}]
    constraintWeights [1,1]
   }
   n8 = "Transform (Modify) Bone - Bone: dyn_ear_upper_cape_collision"(ComponentPose=@n7, Alpha=1, Translation="-7,-7,0", Rotation="0,0,0", Scale="1,1,1") {
    translationMode=BMM_Additive
    translationSpace=BCS_BoneSpace
    alphaBoolBlend {blendOption=Linear}
    boneToModify {boneName=dyn_ear_upper_cape_collision}
    translation=<-7 -7>
   }
   n9 = "Transform (Modify) Bone - Bone: dyn_ear_lower_cape_collision"!(ComponentPose=@n8, bAlphaBoolEnabled:bool=True, Translation="-10,-6,0", Rotation="0,0,0", Scale="1,1,1") -> (Pose) {
    alphaInputType=Bool
    rotationMode=BMM_Additive
    rotationSpace=BCS_BoneSpace
    translationMode=BMM_Additive
    translationSpace=BCS_BoneSpace
    alphaBoolBlend {blendOption=Linear}
    boneToModify {boneName=dyn_ear_lower_cape_collision}
    translation=<-10 -6>
   }
   n10 = "Transform (Modify) Bone - Bone: dyn_ear_l_03"(ComponentPose=@n9, Alpha=1, Translation="0,0,0", Rotation="0,5,0", Scale="1,1,1") {
    rotationMode=BMM_Additive
    rotationSpace=BCS_BoneSpace
    translationMode=BMM_Additive
    translationSpace=BCS_BoneSpace
    alphaBoolBlend {blendOption=Linear}
    boneToModify {boneName=dyn_ear_l_03}
    rotation {yaw=5}
   }
   n11 = "Transform (Modify) Bone - Bone: dyn_ear_r_03"(ComponentPose=@n10, Alpha=1, Translation="0,0,0", Rotation="0,5,0", Scale="1,1,1") {
    rotationMode=BMM_Additive
    rotationSpace=BCS_BoneSpace
    translationMode=BMM_Additive
    translationSpace=BCS_BoneSpace
    alphaBoolBlend {blendOption=Linear}
    boneToModify {boneName=dyn_ear_r_03}
    rotation {yaw=5}
   }
   Output Pose(Result=Component To Local(ComponentPose=@n11)) {layerGroup=None}
   n12 = Layered blend per bone(
    BasePose=Copy Pose From Mesh() {bUseAttachedParent=true}
    BlendPoses_0=Copy Pose From Mesh(SourceMeshComponent=Get Body() -> (Body:Skeletal Mesh Component Budgeted Object Reference)) {bUseAttachedParent=true}
   ) {layerSetup [{branchFilters [{boneName=upperarm_l}, {boneName=upperarm_r}, {boneName=lowerarm_l}, {boneName=lowerarm_r}]}]}
   n13 = Control Rig!(
    Source=@n12
    Alpha
    UseSplineEars:bool=Get isWearingBackbling()
    IsSkydiving:bool=Get bIsSkyDiving()
    EarHarmonicsSpeed:double=Get Ear Harmonics Speed() -> (Ear Harmonics Speed:double)
   ) -> (Pose) {
    alphaBoolBlend {blendOption=Linear}
    controlRigAssetReference {blueprintRigClass=&/M_MED_MoleAcornGloam_FaceAcc_CtrlRig.M_MED_MoleAcornGloam_FaceAcc_CtrlRig_C}
   }
   Save cached pose 'Post Deformation'(Pose=@n13) {cacheName=Post Deformation}
   n14 = RigidBody!(
    ComponentPose=Local To Component(LocalPose=Use cached pose 'Post Deformation'())
    Alpha
    OverrideWorldGravity:Vector="0,0,-980"
    ExternalForce
   ) -> (Pose) {
    bOverrideWorldGravity=true
    overridePhysicsAsset=&/M_MED_MoleAcornGloam_FaceAcc_Physics
    simulationSpace=BaseBoneSpace
    alphaBoolBlend {blendOption=Linear}
    baseBoneRef {boneName=neck_01}
    componentAppliedLinearAccClamp=<0 0 0>
    overrideWorldGravity {z=-980}
    simSpaceSettings {
     dampingAlpha=0
     maxAngularAcceleration=20
     maxAngularVelocity=6
     maxLinearAcceleration=1200
     maxLinearVelocity=650
     velocityScaleZ=0.5
     worldAlpha=1
     externalLinearDragV=<1 2 0.75>
    }
   }
   Save cached pose 'Post RBAN'(Pose=Component To Local(ComponentPose=@n14)) {cacheName=Post RBAN}
   n15 = Anim Dynamics - Bone: dyn_ear_l_03(ComponentPose=Local To Component(LocalPose=Use cached pose 'Post Deformation'()), Alpha=1, ExternalForce="0,0,0") {
    angularDampingOverride=0.98
    angularSpringConstant=400
    bAngularSpring=true
    bOverrideAngularDamping=true
    bUseGravityOverride=true
    gravityScale=0.35
    simulationSpace=RootRelative
    alphaBoolBlend {blendOption=Linear}
    boundBone {boneName=dyn_ear_l_03}
    componentAppliedLinearAccClamp=<1300 1300 1300>
    componentLinearAccScale=<1 1 1>
    componentLinearVelScale=<1 1 1>
    gravityOverride {z=-980}
    physicsBodyDefinitions [{
      boundBone {boneName=dyn_ear_l_03}
      boxExtents=<30 40 40>
      constraintSetup {angularLimitsMax=<5 5 0>, angularLimitsMin=<-5 -10 -2>, angularTarget=<1 0 0>}
      localJointOffset=<4 0 0>
     }]
   }
   n16 = Anim Dynamics - Bone: dyn_ear_l_04(ComponentPose=@n15, Alpha=1, ExternalForce="0,0,0") {
    angularDampingOverride=0.98
    angularSpringConstant=400
    bAngularSpring=true
    bOverrideAngularDamping=true
    bUseGravityOverride=true
    gravityScale=0.35
    simulationSpace=RootRelative
    alphaBoolBlend {blendOption=Linear}
    boundBone {boneName=dyn_ear_l_04}
    componentAppliedLinearAccClamp=<1300 1300 1300>
    componentLinearAccScale=<1 1 1>
    componentLinearVelScale=<1 1 1>
    gravityOverride {z=-980}
    physicsBodyDefinitions [{
      boundBone {boneName=dyn_ear_l_04}
      boxExtents=<30 40 40>
      constraintSetup {angularLimitsMax=<5 5 0>, angularLimitsMin=<-5 -10 -2>, angularTarget=<1 0 0>}
      localJointOffset=<4 0 0>
     }]
   }
   n17 = Anim Dynamics - Bone: dyn_ear_l_05(ComponentPose=@n16, Alpha=1, ExternalForce="0,0,0") {
    angularDampingOverride=0.98
    angularSpringConstant=400
    bAngularSpring=true
    bOverrideAngularDamping=true
    bUseGravityOverride=true
    gravityScale=0.35
    simulationSpace=RootRelative
    alphaBoolBlend {blendOption=Linear}
    boundBone {boneName=dyn_ear_l_05}
    componentAppliedLinearAccClamp=<1300 1300 1600>
    componentLinearAccScale=<1 1 1>
    componentLinearVelScale=<1 1 1>
    gravityOverride {z=-980}
    physicsBodyDefinitions [{
      boundBone {boneName=dyn_ear_l_05}
      boxExtents=<30 40 40>
      constraintSetup {angularLimitsMax=<5 5 0>, angularLimitsMin=<-5 -10 -5>, angularTarget=<1 0 0>}
      localJointOffset=<4 0 0>
     }]
   }
   n18 = Anim Dynamics - Bone: dyn_ear_l_06(ComponentPose=@n17, Alpha=1, ExternalForce="0,0,0") {
    angularDampingOverride=0.98
    angularSpringConstant=400
    bAngularSpring=true
    bOverrideAngularDamping=true
    bUseGravityOverride=true
    gravityScale=0.35
    simulationSpace=RootRelative
    alphaBoolBlend {blendOption=Linear}
    boundBone {boneName=dyn_ear_l_06}
    componentAppliedLinearAccClamp=<1300 1300 2400>
    componentLinearAccScale=<1 1 1>
    componentLinearVelScale=<1 1 1>
    gravityOverride {z=-980}
    physicsBodyDefinitions [{
      boundBone {boneName=dyn_ear_l_06}
      boxExtents=<30 40 40>
      constraintSetup {angularLimitsMax=<5 5 0>, angularLimitsMin=<-5 -5 -15>, angularTarget=<1 0 0>}
      localJointOffset=<4 0 0>
     }]
   }
   n19 = Anim Dynamics - Bone: dyn_ear_l_07(ComponentPose=@n18, Alpha=1, ExternalForce="0,0,0") {
    angularDampingOverride=0.98
    angularSpringConstant=400
    bAngularSpring=true
    bOverrideAngularDamping=true
    bUseGravityOverride=true
    gravityScale=0.35
    simulationSpace=RootRelative
    alphaBoolBlend {blendOption=Linear}
    boundBone {boneName=dyn_ear_l_07}
    componentAppliedLinearAccClamp=<1300 1300 2400>
    componentLinearAccScale=<1 1 1>
    componentLinearVelScale=<1 1 1>
    gravityOverride {z=-980}
    physicsBodyDefinitions [{
      boundBone {boneName=dyn_ear_l_07}
      boxExtents=<30 40 40>
      constraintSetup {angularLimitsMax=<5 5 0>, angularLimitsMin=<-5 -5 -15>, angularTarget=<1 0 0>}
      localJointOffset=<4 0 0>
     }]
   }
   n20 = Anim Dynamics - Bone: dyn_ear_l_08(ComponentPose=@n19, Alpha=1, ExternalForce="0,0,0") {
    angularDampingOverride=0.98
    angularSpringConstant=400
    bAngularSpring=true
    bOverrideAngularDamping=true
    bUseGravityOverride=true
    gravityScale=0.35
    simulationSpace=RootRelative
    alphaBoolBlend {blendOption=Linear}
    boundBone {boneName=dyn_ear_l_08}
    componentAppliedLinearAccClamp=<1300 1300 2400>
    componentLinearAccScale=<1 1 1>
    componentLinearVelScale=<1 1 1>
    gravityOverride {z=-980}
    physicsBodyDefinitions [{
      boundBone {boneName=dyn_ear_l_08}
      boxExtents=<30 40 40>
      constraintSetup {angularLimitsMax=<5 5 0>, angularLimitsMin=<-5 -5 -15>, angularTarget=<1 0 0>}
      localJointOffset=<4 0 0>
     }]
   }
   n21 = Anim Dynamics - Bone: dyn_ear_r_03(ComponentPose=@n20, Alpha=1, ExternalForce="0,0,0") {
    angularDampingOverride=0.98
    angularSpringConstant=400
    bAngularSpring=true
    bOverrideAngularDamping=true
    bUseGravityOverride=true
    gravityScale=0.35
    simulationSpace=RootRelative
    alphaBoolBlend {blendOption=Linear}
    boundBone {boneName=dyn_ear_r_03}
    componentAppliedLinearAccClamp=<1300 1300 1300>
    componentLinearAccScale=<1 1 1>
    componentLinearVelScale=<1 1 1>
    gravityOverride {z=-980}
    physicsBodyDefinitions [{
      boundBone {boneName=dyn_ear_r_03}
      boxExtents=<30 40 40>
      constraintSetup {angularLimitsMax=<5 5 0>, angularLimitsMin=<-5 -10 -2>, angularTarget=<1 0 0>}
      localJointOffset=<-4 0 0>
     }]
   }
   n22 = Anim Dynamics - Bone: dyn_ear_r_04(ComponentPose=@n21, Alpha=1, ExternalForce="0,0,0") {
    angularDampingOverride=0.98
    angularSpringConstant=400
    bAngularSpring=true
    bOverrideAngularDamping=true
    bUseGravityOverride=true
    gravityScale=0.35
    simulationSpace=RootRelative
    alphaBoolBlend {blendOption=Linear}
    boundBone {boneName=dyn_ear_r_04}
    componentAppliedLinearAccClamp=<1300 1300 1300>
    componentLinearAccScale=<1 1 1>
    componentLinearVelScale=<1 1 1>
    gravityOverride {z=-980}
    physicsBodyDefinitions [{
      boundBone {boneName=dyn_ear_r_04}
      boxExtents=<30 40 40>
      constraintSetup {angularLimitsMax=<5 5 0>, angularLimitsMin=<-5 -10 -2>, angularTarget=<1 0 0>}
      localJointOffset=<-4 0 0>
     }]
   }
   n23 = Anim Dynamics - Bone: dyn_ear_r_05(ComponentPose=@n22, Alpha=1, ExternalForce="0,0,0") {
    angularDampingOverride=0.98
    angularSpringConstant=400
    bAngularSpring=true
    bOverrideAngularDamping=true
    bUseGravityOverride=true
    gravityScale=0.35
    simulationSpace=RootRelative
    alphaBoolBlend {blendOption=Linear}
    boundBone {boneName=dyn_ear_r_05}
    componentAppliedLinearAccClamp=<1300 1300 1600>
    componentLinearAccScale=<1 1 1>
    componentLinearVelScale=<1 1 1>
    gravityOverride {z=-980}
    physicsBodyDefinitions [{
      boundBone {boneName=dyn_ear_r_05}
      boxExtents=<30 40 40>
      constraintSetup {angularLimitsMax=<5 5 0>, angularLimitsMin=<-5 -10 -5>, angularTarget=<1 0 0>}
      localJointOffset=<-4 0 0>
     }]
   }
   n24 = Anim Dynamics - Bone: dyn_ear_r_06(ComponentPose=@n23, Alpha=1, ExternalForce="0,0,0") {
    angularDampingOverride=0.98
    angularSpringConstant=400
    bAngularSpring=true
    bOverrideAngularDamping=true
    bUseGravityOverride=true
    gravityScale=0.35
    simulationSpace=RootRelative
    alphaBoolBlend {blendOption=Linear}
    boundBone {boneName=dyn_ear_r_06}
    componentAppliedLinearAccClamp=<1300 1300 2400>
    componentLinearAccScale=<1 1 1>
    componentLinearVelScale=<1 1 1>
    gravityOverride {z=-980}
    physicsBodyDefinitions [{
      boundBone {boneName=dyn_ear_r_06}
      boxExtents=<30 40 40>
      constraintSetup {angularLimitsMax=<5 5 0>, angularLimitsMin=<-5 -5 -15>, angularTarget=<1 0 0>}
      localJointOffset=<-4 0 0>
     }]
   }
   n25 = Anim Dynamics - Bone: dyn_ear_r_07(ComponentPose=@n24, Alpha=1, ExternalForce="0,0,0") {
    angularDampingOverride=0.98
    angularSpringConstant=400
    bAngularSpring=true
    bOverrideAngularDamping=true
    bUseGravityOverride=true
    gravityScale=0.35
    simulationSpace=RootRelative
    alphaBoolBlend {blendOption=Linear}
    boundBone {boneName=dyn_ear_r_07}
    componentAppliedLinearAccClamp=<1300 1300 2400>
    componentLinearAccScale=<1 1 1>
    componentLinearVelScale=<1 1 1>
    gravityOverride {z=-980}
    physicsBodyDefinitions [{
      boundBone {boneName=dyn_ear_r_07}
      boxExtents=<30 40 40>
      constraintSetup {angularLimitsMax=<5 5 0>, angularLimitsMin=<-5 -5 -15>, angularTarget=<1 0 0>}
      localJointOffset=<-4 0 0>
     }]
   }
   n26 = Anim Dynamics - Bone: dyn_ear_r_08(ComponentPose=@n25, Alpha=1, ExternalForce="0,0,0") {
    angularDampingOverride=0.98
    angularSpringConstant=400
    bAngularSpring=true
    bOverrideAngularDamping=true
    bUseGravityOverride=true
    gravityScale=0.35
    simulationSpace=RootRelative
    alphaBoolBlend {blendOption=Linear}
    boundBone {boneName=dyn_ear_r_08}
    componentAppliedLinearAccClamp=<1300 1300 2400>
    componentLinearAccScale=<1 1 1>
    componentLinearVelScale=<1 1 1>
    gravityOverride {z=-980}
    physicsBodyDefinitions [{
      boundBone {boneName=dyn_ear_r_08}
      boxExtents=<30 40 40>
      constraintSetup {angularLimitsMax=<5 5 0>, angularLimitsMin=<-5 -5 -15>, angularTarget=<1 0 0>}
      localJointOffset=<-4 0 0>
     }]
   }
   n27 = "Constraint - dyn_ear_l_09 (P)"(ComponentPose=@n26, Alpha=1, ConstraintWeights_0=1) {
    alphaBoolBlend {blendOption=Linear}
    boneToModify {boneName=dyn_ear_l_09}
    constraintSetup [{targetBone {boneName=dyn_ear_l_03}}]
    constraintWeights [1]
   }
   n28 = "Constraint - dyn_ear_l_010 (P)"(ComponentPose=@n27, Alpha=1, ConstraintWeights_0=1) {
    alphaBoolBlend {blendOption=Linear}
    boneToModify {boneName=dyn_ear_l_010}
    constraintSetup [{targetBone {boneName=dyn_ear_l_04}}]
    constraintWeights [1]
   }
   n29 = "Constraint - dyn_ear_r_09 (P)"(ComponentPose=@n28, Alpha=1, ConstraintWeights_0=1) {
    alphaBoolBlend {blendOption=Linear}
    boneToModify {boneName=dyn_ear_r_09}
    constraintSetup [{targetBone {boneName=dyn_ear_r_03}}]
    constraintWeights [1]
   }
   n30 = "Constraint - dyn_ear_r_010 (P)"(ComponentPose=@n29, Alpha=1, ConstraintWeights_0=1) {
    alphaBoolBlend {blendOption=Linear}
    boneToModify {boneName=dyn_ear_r_010}
    constraintSetup [{targetBone {boneName=dyn_ear_r_04}}]
    constraintWeights [1]
   }
   Save cached pose 'Post Backbling Anim Dynamics Ears'(Pose=Component To Local(ComponentPose=@n30)) {cacheName=Post Backbling Anim Dynamics Ears}
   n31 = Anim Dynamics - Bone: dyn_ear_l_01!(
    SimulationSpace:AnimPhysSimSpaceType Enum=Component
    ComponentPose=Local To Component(LocalPose=Use cached pose 'Post Deformation'())
    Alpha=1
    ExternalForce="0,0,0"
   ) -> (Pose) {
    angularDampingOverride=1
    angularSpringConstant=400
    bAngularSpring=true
    bOverrideAngularDamping=true
    bUseGravityOverride=true
    gravityScale=0.35
    alphaBoolBlend {blendOption=Linear}
    boundBone {boneName=dyn_ear_l_01}
    componentAppliedLinearAccClamp=<1300 1300 1300>
    componentLinearAccScale=<1 1 1>
    componentLinearVelScale=<1 1 1>
    gravityOverride {z=-980}
    physicsBodyDefinitions [{
      boundBone {boneName=dyn_ear_l_01}
      boxExtents=<15 20 20>
      constraintSetup {angularLimitsMax=<5 2 0>, angularLimitsMin=<-5 -7 -20>, angularTarget=<1 0 0>}
      localJointOffset=<4 0 0>
     }]
   }
   n32 = Anim Dynamics - Bone: dyn_ear_l_02!(SimulationSpace:AnimPhysSimSpaceType Enum=Component, ComponentPose=@n31, Alpha=1, ExternalForce="0,0,0") -> (Pose) {
    angularDampingOverride=1
    angularSpringConstant=400
    bAngularSpring=true
    bOverrideAngularDamping=true
    bUseGravityOverride=true
    gravityScale=0.35
    alphaBoolBlend {blendOption=Linear}
    boundBone {boneName=dyn_ear_l_02}
    componentAppliedLinearAccClamp=<1300 1300 1300>
    componentLinearAccScale=<1 1 1>
    componentLinearVelScale=<1 1 1>
    gravityOverride {z=-980}
    physicsBodyDefinitions [{
      boundBone {boneName=dyn_ear_l_02}
      boxExtents=<15 20 20>
      constraintSetup {angularLimitsMax=<5 2 0>, angularLimitsMin=<-5 -7 -20>, angularTarget=<1 0 0>}
      localJointOffset=<4 0 0>
     }]
   }
   n33 = Anim Dynamics - Bone: dyn_ear_l_03!(SimulationSpace:AnimPhysSimSpaceType Enum=Component, ComponentPose=@n32, Alpha=1, ExternalForce="0,0,0") -> (Pose) {
    angularDampingOverride=0.98
    angularSpringConstant=400
    bAngularSpring=true
    bOverrideAngularDamping=true
    bUseGravityOverride=true
    gravityScale=0.35
    alphaBoolBlend {blendOption=Linear}
    boundBone {boneName=dyn_ear_l_03}
    componentAppliedLinearAccClamp=<1300 1300 1300>
    componentLinearAccScale=<1 1 1>
    componentLinearVelScale=<1 1 1>
    gravityOverride {z=-980}
    physicsBodyDefinitions [{
      boundBone {boneName=dyn_ear_l_03}
      boxExtents=<15 20 20>
      constraintSetup {angularLimitsMax=<5 2 0>, angularLimitsMin=<-5 -7 -20>, angularTarget=<1 0 0>}
      localJointOffset=<4 0 0>
     }]
   }
   n34 = Anim Dynamics - Bone: dyn_ear_l_04!(SimulationSpace:AnimPhysSimSpaceType Enum, ComponentPose=@n33, Alpha=1, ExternalForce="0,0,0") -> (Pose) {
    angularDampingOverride=0.98
    angularSpringConstant=400
    bAngularSpring=true
    bOverrideAngularDamping=true
    bUseGravityOverride=true
    gravityScale=0.35
    simulationSpace=RootRelative
    alphaBoolBlend {blendOption=Linear}
    boundBone {boneName=dyn_ear_l_04}
    componentAppliedLinearAccClamp=<1300 1300 1300>
    componentLinearAccScale=<1 1 1>
    componentLinearVelScale=<1 1 1>
    gravityOverride {z=-980}
    physicsBodyDefinitions [{
      boundBone {boneName=dyn_ear_l_04}
      boxExtents=<15 20 20>
      constraintSetup {angularLimitsMax=<5 2 0>, angularLimitsMin=<-5 -7 -20>, angularTarget=<1 0 0>}
      localJointOffset=<4 0 0>
     }]
   }
   n35 = Anim Dynamics - Bone: dyn_ear_l_05!(SimulationSpace:AnimPhysSimSpaceType Enum, ComponentPose=@n34, Alpha=1, ExternalForce="0,0,0") -> (Pose) {
    angularDampingOverride=0.98
    angularSpringConstant=400
    bAngularSpring=true
    bOverrideAngularDamping=true
    bUseGravityOverride=true
    gravityScale=0.35
    simulationSpace=RootRelative
    alphaBoolBlend {blendOption=Linear}
    boundBone {boneName=dyn_ear_l_05}
    componentAppliedLinearAccClamp=<1300 1300 1300>
    componentLinearAccScale=<1 1 1>
    componentLinearVelScale=<1 1 1>
    gravityOverride {z=-980}
    physicsBodyDefinitions [{
      boundBone {boneName=dyn_ear_l_05}
      boxExtents=<15 20 20>
      constraintSetup {angularLimitsMax=<5 2 0>, angularLimitsMin=<-5 -7 -20>, angularTarget=<1 0 0>}
      localJointOffset=<4 0 0>
     }]
   }
   n36 = Anim Dynamics - Bone: dyn_ear_l_06!(SimulationSpace:AnimPhysSimSpaceType Enum, ComponentPose=@n35, Alpha=1, ExternalForce="0,0,0") -> (Pose) {
    angularDampingOverride=0.98
    angularSpringConstant=400
    bAngularSpring=true
    bOverrideAngularDamping=true
    bUseGravityOverride=true
    gravityScale=0.35
    simulationSpace=RootRelative
    alphaBoolBlend {blendOption=Linear}
    boundBone {boneName=dyn_ear_l_06}
    componentAppliedLinearAccClamp=<1300 1300 1300>
    componentLinearAccScale=<1 1 1>
    componentLinearVelScale=<1 1 1>
    gravityOverride {z=-980}
    physicsBodyDefinitions [{
      boundBone {boneName=dyn_ear_l_06}
      boxExtents=<15 20 20>
      constraintSetup {angularLimitsMax=<5 2 0>, angularLimitsMin=<-5 -7 -20>, angularTarget=<1 0 0>}
      localJointOffset=<4 0 0>
     }]
   }
   n37 = Anim Dynamics - Bone: dyn_ear_l_07!(SimulationSpace:AnimPhysSimSpaceType Enum=Component, ComponentPose=@n36, Alpha=1, ExternalForce="0,0,0") -> (Pose) {
    angularDampingOverride=0.98
    angularSpringConstant=400
    bAngularSpring=true
    bOverrideAngularDamping=true
    bUseGravityOverride=true
    gravityScale=0.35
    alphaBoolBlend {blendOption=Linear}
    boundBone {boneName=dyn_ear_l_07}
    componentAppliedLinearAccClamp=<1300 1300 1300>
    componentLinearAccScale=<1 1 1>
    componentLinearVelScale=<1 1 1>
    gravityOverride {z=-980}
    physicsBodyDefinitions [{
      boundBone {boneName=dyn_ear_l_07}
      boxExtents=<15 20 20>
      constraintSetup {angularLimitsMax=<5 2 0>, angularLimitsMin=<-5 -7 -20>, angularTarget=<1 0 0>}
      localJointOffset=<4 0 0>
     }]
   }
   n38 = Anim Dynamics - Bone: dyn_ear_l_08!(SimulationSpace:AnimPhysSimSpaceType Enum=Component, ComponentPose=@n37, Alpha=1, ExternalForce="0,0,0") -> (Pose) {
    angularDampingOverride=0.98
    angularSpringConstant=400
    bAngularSpring=true
    bOverrideAngularDamping=true
    bUseGravityOverride=true
    gravityScale=0.35
    alphaBoolBlend {blendOption=Linear}
    boundBone {boneName=dyn_ear_l_08}
    componentAppliedLinearAccClamp=<1300 1300 1300>
    componentLinearAccScale=<1 1 1>
    componentLinearVelScale=<1 1 1>
    gravityOverride {z=-980}
    physicsBodyDefinitions [{
      boundBone {boneName=dyn_ear_l_08}
      boxExtents=<15 20 20>
      constraintSetup {angularLimitsMax=<5 2 0>, angularLimitsMin=<-5 -7 -20>, angularTarget=<1 0 0>}
      localJointOffset=<4 0 0>
     }]
   }
   n39 = Anim Dynamics - Bone: dyn_ear_r_01!(SimulationSpace:AnimPhysSimSpaceType Enum=Component, ComponentPose=@n38, Alpha=1, ExternalForce="0,0,0") -> (Pose) {
    angularDampingOverride=1
    angularSpringConstant=400
    bAngularSpring=true
    bOverrideAngularDamping=true
    bUseGravityOverride=true
    gravityScale=0.35
    alphaBoolBlend {blendOption=Linear}
    boundBone {boneName=dyn_ear_r_01}
    componentAppliedLinearAccClamp=<1300 1300 1300>
    componentLinearAccScale=<1 1 1>
    componentLinearVelScale=<1 1 1>
    gravityOverride {z=-980}
    physicsBodyDefinitions [{
      boundBone {boneName=dyn_ear_r_01}
      boxExtents=<15 20 20>
      constraintSetup {angularLimitsMax=<5 2 0>, angularLimitsMin=<-5 -7 -20>, angularTarget=<1 0 0>}
      localJointOffset=<-4 0 0>
     }]
   }
   n40 = Anim Dynamics - Bone: dyn_ear_r_02!(SimulationSpace:AnimPhysSimSpaceType Enum=Component, ComponentPose=@n39, Alpha=1, ExternalForce="0,0,0") -> (Pose) {
    angularDampingOverride=1
    angularSpringConstant=400
    bAngularSpring=true
    bOverrideAngularDamping=true
    bUseGravityOverride=true
    gravityScale=0.35
    alphaBoolBlend {blendOption=Linear}
    boundBone {boneName=dyn_ear_r_02}
    componentAppliedLinearAccClamp=<1300 1300 1300>
    componentLinearAccScale=<1 1 1>
    componentLinearVelScale=<1 1 1>
    gravityOverride {z=-980}
    physicsBodyDefinitions [{
      boundBone {boneName=dyn_ear_r_02}
      boxExtents=<15 20 20>
      constraintSetup {angularLimitsMax=<5 2 0>, angularLimitsMin=<-5 -7 -20>, angularTarget=<1 0 0>}
      localJointOffset=<-4 0 0>
     }]
   }
   n41 = Anim Dynamics - Bone: dyn_ear_r_03!(SimulationSpace:AnimPhysSimSpaceType Enum=Component, ComponentPose=@n40, Alpha=1, ExternalForce="0,0,0") -> (Pose) {
    angularDampingOverride=0.98
    angularSpringConstant=400
    bAngularSpring=true
    bOverrideAngularDamping=true
    bUseGravityOverride=true
    gravityScale=0.35
    alphaBoolBlend {blendOption=Linear}
    boundBone {boneName=dyn_ear_r_03}
    componentAppliedLinearAccClamp=<1300 1300 1300>
    componentLinearAccScale=<1 1 1>
    componentLinearVelScale=<1 1 1>
    gravityOverride {z=-980}
    physicsBodyDefinitions [{
      boundBone {boneName=dyn_ear_r_03}
      boxExtents=<15 20 20>
      constraintSetup {angularLimitsMax=<5 2 0>, angularLimitsMin=<-5 -7 -20>, angularTarget=<1 0 0>}
      localJointOffset=<-4 0 0>
     }]
   }
   n42 = Anim Dynamics - Bone: dyn_ear_r_04!(SimulationSpace:AnimPhysSimSpaceType Enum, ComponentPose=@n41, Alpha=1, ExternalForce="0,0,0") -> (Pose) {
    angularDampingOverride=0.98
    angularSpringConstant=400
    bAngularSpring=true
    bOverrideAngularDamping=true
    bUseGravityOverride=true
    gravityScale=0.35
    simulationSpace=RootRelative
    alphaBoolBlend {blendOption=Linear}
    boundBone {boneName=dyn_ear_r_04}
    componentAppliedLinearAccClamp=<1300 1300 1300>
    componentLinearAccScale=<1 1 1>
    componentLinearVelScale=<1 1 1>
    gravityOverride {z=-980}
    physicsBodyDefinitions [{
      boundBone {boneName=dyn_ear_r_04}
      boxExtents=<15 20 20>
      constraintSetup {angularLimitsMax=<5 2 0>, angularLimitsMin=<-5 -7 -20>, angularTarget=<1 0 0>}
      localJointOffset=<-4 0 0>
     }]
   }
   n43 = Anim Dynamics - Bone: dyn_ear_r_05!(SimulationSpace:AnimPhysSimSpaceType Enum, ComponentPose=@n42, Alpha=1, ExternalForce="0,0,0") -> (Pose) {
    angularDampingOverride=0.98
    angularSpringConstant=400
    bAngularSpring=true
    bOverrideAngularDamping=true
    bUseGravityOverride=true
    gravityScale=0.35
    simulationSpace=RootRelative
    alphaBoolBlend {blendOption=Linear}
    boundBone {boneName=dyn_ear_r_05}
    componentAppliedLinearAccClamp=<1300 1300 1300>
    componentLinearAccScale=<1 1 1>
    componentLinearVelScale=<1 1 1>
    gravityOverride {z=-980}
    physicsBodyDefinitions [{
      boundBone {boneName=dyn_ear_r_05}
      boxExtents=<15 20 20>
      constraintSetup {angularLimitsMax=<5 2 0>, angularLimitsMin=<-5 -7 -20>, angularTarget=<1 0 0>}
      localJointOffset=<-4 0 0>
     }]
   }
   n44 = Anim Dynamics - Bone: dyn_ear_r_06!(SimulationSpace:AnimPhysSimSpaceType Enum, ComponentPose=@n43, Alpha=1, ExternalForce="0,0,0") -> (Pose) {
    angularDampingOverride=0.98
    angularSpringConstant=400
    bAngularSpring=true
    bOverrideAngularDamping=true
    bUseGravityOverride=true
    gravityScale=0.35
    simulationSpace=RootRelative
    alphaBoolBlend {blendOption=Linear}
    boundBone {boneName=dyn_ear_r_06}
    componentAppliedLinearAccClamp=<1300 1300 1300>
    componentLinearAccScale=<1 1 1>
    componentLinearVelScale=<1 1 1>
    gravityOverride {z=-980}
    physicsBodyDefinitions [{
      boundBone {boneName=dyn_ear_r_06}
      boxExtents=<15 20 20>
      constraintSetup {angularLimitsMax=<5 2 0>, angularLimitsMin=<-5 -7 -20>, angularTarget=<1 0 0>}
      localJointOffset=<-4 0 0>
     }]
   }
   n45 = Anim Dynamics - Bone: dyn_ear_r_07!(SimulationSpace:AnimPhysSimSpaceType Enum=Component, ComponentPose=@n44, Alpha=1, ExternalForce="0,0,0") -> (Pose) {
    angularDampingOverride=0.98
    angularSpringConstant=400
    bAngularSpring=true
    bOverrideAngularDamping=true
    bUseGravityOverride=true
    gravityScale=0.35
    alphaBoolBlend {blendOption=Linear}
    boundBone {boneName=dyn_ear_r_07}
    componentAppliedLinearAccClamp=<1300 1300 1300>
    componentLinearAccScale=<1 1 1>
    componentLinearVelScale=<1 1 1>
    gravityOverride {z=-980}
    physicsBodyDefinitions [{
      boundBone {boneName=dyn_ear_r_07}
      boxExtents=<15 20 20>
      constraintSetup {angularLimitsMax=<5 2 0>, angularLimitsMin=<-5 -7 -20>, angularTarget=<1 0 0>}
      localJointOffset=<-4 0 0>
     }]
   }
   n46 = Anim Dynamics - Bone: dyn_ear_r_08!(SimulationSpace:AnimPhysSimSpaceType Enum=Component, ComponentPose=@n45, Alpha=1, ExternalForce="0,0,0") -> (Pose) {
    angularDampingOverride=0.98
    angularSpringConstant=400
    bAngularSpring=true
    bOverrideAngularDamping=true
    bUseGravityOverride=true
    gravityScale=0.35
    alphaBoolBlend {blendOption=Linear}
    boundBone {boneName=dyn_ear_r_08}
    componentAppliedLinearAccClamp=<1300 1300 1300>
    componentLinearAccScale=<1 1 1>
    componentLinearVelScale=<1 1 1>
    gravityOverride {z=-980}
    physicsBodyDefinitions [{
      boundBone {boneName=dyn_ear_r_08}
      boxExtents=<15 20 20>
      constraintSetup {angularLimitsMax=<5 2 0>, angularLimitsMin=<-5 -7 -20>, angularTarget=<1 0 0>}
      localJointOffset=<-4 0 0>
     }]
   }
   n47 = "Constraint - dyn_ear_l_09 (P)"(ComponentPose=@n46, Alpha=1, ConstraintWeights_0=1) {
    alphaBoolBlend {blendOption=Linear}
    boneToModify {boneName=dyn_ear_l_09}
    constraintSetup [{targetBone {boneName=dyn_ear_l_04}}]
    constraintWeights [1]
   }
   n48 = "Constraint - dyn_ear_l_010 (P)"(ComponentPose=@n47, Alpha=1, ConstraintWeights_0=1) {
    alphaBoolBlend {blendOption=Linear}
    boneToModify {boneName=dyn_ear_l_010}
    constraintSetup [{targetBone {boneName=dyn_ear_l_05}}]
    constraintWeights [1]
   }
   n49 = "Constraint - dyn_ear_r_09 (P)"(ComponentPose=@n48, Alpha=1, ConstraintWeights_0=1) {
    alphaBoolBlend {blendOption=Linear}
    boneToModify {boneName=dyn_ear_r_09}
    constraintSetup [{targetBone {boneName=dyn_ear_r_04}}]
    constraintWeights [1]
   }
   n50 = "Constraint - dyn_ear_r_010 (P)"(ComponentPose=@n49, Alpha=1, ConstraintWeights_0=1) {
    alphaBoolBlend {blendOption=Linear}
    boneToModify {boneName=dyn_ear_r_010}
    constraintSetup [{targetBone {boneName=dyn_ear_r_05}}]
    constraintWeights [1]
   }
   Save cached pose 'Post Anim Dynamics Ears'(Pose=Component To Local(ComponentPose=@n50)) {cacheName=Post Anim Dynamics Ears}
  ]}
 EventGraph {nodes [
   Event BlueprintUpdateAnimation()
   > n1 = Sequence!(execute) -> (then_0, then_1, then_2:Exec, then_3:Exec)
   n2 = Branch(execute=@n1.then_1, Condition=Get bRigidBodyEnabled())
   > n3 = Branch(Condition=Get isWearingBackbling())
   > Set Ear Harmonics Speed(Ear Harmonics Speed:double=0) -> (Output_Get:double)
   n4 = Branch(execute=@n3.else, Condition=Get bIsSkyDiving())
   > n5 = Branch(Condition=Get bIsSkydiveFloating())
   > Set Ear Harmonics Speed(Ear Harmonics Speed=1)
   n6 = Branch(execute=@n5.else, Condition=Get bIsSkydiveDiveMode())
   > Set Ear Harmonics Speed(Ear Harmonics Speed=1.8)
   Set Ear Harmonics Speed(execute=@n6.else, Ear Harmonics Speed=1.2)
   Set Ear Harmonics Speed(execute=@n4.else, Ear Harmonics Speed=MapRangeClamped(Value=Get PawnSpeed2D(), InRangeA=600, InRangeB=1500))
   Set Ear Harmonics Speed(execute=@n2.else, Ear Harmonics Speed=0)
   n7 = Gravity Override State Function DEPRECATED(
    execute=@n1.then_0
    CharacterPartAnimInstance=Self-Reference()
    GameTimeInSeconds=0
    GravityOverrideParams=Get Gravity Override Params()
    RBAN Params=Get RBAN Params()
   )
   > Set Gravity Override(Gravity Override=@n7)
   > Set RBAN Settings Out(RBAN Settings Out:RBANSetting Struct Structure=@n7.RBAN SettingsOut) -> (Output_Get:RBANSetting Struct Structure)
   Branch(execute=@n1.then_2, Condition=Get Is Backbling A Cape() -> (Is Backbling A Cape:bool))
   > n8 = Cast To CustomCharacterPartAnimInstance!(execute, Object=GetAnimInstance(self=Get Backbling() -> (Backbling:Skeletal Mesh Component Budgeted Object Reference))) -> (then, CastFailed, AsCustom Character Part Anim Instance:Custom Character Part Anim Instance Object Reference) {is=title K2Node_DynamicCast}
   > Set Cape Cloth Disabled(Cape Cloth Disabled:bool=NOT Boolean(A=Get bClothEnabled!(self:Fort Anim Instance Object Reference=@n8) -> (bClothEnabled))) -> (Output_Get:bool)
   n9 = Branch(execute=@n1.then_3, Condition=Get bIsInFrontEnd())
   > Set AnimDynamics SimSpace(AnimDynamics SimSpace:AnimPhysSimSpaceType Enum=World) -> (Output_Get:AnimPhysSimSpaceType Enum)
   > Set AnimDynamics Alpha(AnimDynamics Alpha:double=0.3) -> (Output_Get:double)
   Set AnimDynamics SimSpace(execute=@n9.else, AnimDynamics SimSpace=Component)
   > Set AnimDynamics Alpha(AnimDynamics Alpha=1)
   n10 = Event OnCharacterCustomizationLoaded()
   n11 = CheckForBackbling()
   > Set isWearingBackbling(isWearingBackbling=AND Boolean(A=@n11, B=NOT Boolean(A=@n11.bIsBackblingACape)))
   > Set Is Backbling A Cape(Is Backbling A Cape:bool=@n11.bIsBackblingACape) -> (Output_Get:bool)
   > Set Body(Body:Skeletal Mesh Component Budgeted Object Reference=Get Array Item with EFort Custom Part Type(Array=@n10.CustomMeshParts)) -> (Output_Get:Skeletal Mesh Component Budgeted Object Reference)
   > Set Backbling(Backbling:Skeletal Mesh Component Budgeted Object Reference=Get Array Item with EFort Custom Part Type(Array=@n10.CustomMeshParts, PartType=Backpack)) -> (Output_Get:Skeletal Mesh Component Budgeted Object Reference)
  ]}
]
variables [
 AnimDynamics Alpha {type=number}
 AnimDynamics SimSpace {type=AnimPhysSimSpaceType}
 Backbling {type=SkeletalMeshComponentBudgeted}
 Body {type=SkeletalMeshComponentBudgeted}
 Cape Cloth Disabled {type=boolean}
 Ear Harmonics Speed {type=number}
 Gravity Override {type=Vector, default=<0 0 -980>}
 Gravity Override Params {
  type=GravityOverrideParamsStruct
  default {
   skydive_DivingDown {windAmplitude=<50 500 200>, windFrequency=<8 15 20>}
   skydive_FloatingDown {gravityOverride=<0 -700 200>, windAmplitude=<200 -500 600>, windFrequency=<1.5 2 15>}
   skydive_GoingUp {gravityOverride=<0 -700 200>, windAmplitude=<200 -500 800>, windFrequency=<1.5 7 12>, windOffset=<0 -600 400>}
   skydive_Parachuting {gravityOverride=<0 -700 200>, windAmplitude=<200 -500 800>, windFrequency=<1.5 7 12>, windOffset=<0 -600 400>}
  }
 }
 Is Backbling A Cape {type=boolean}
 isWearingBackbling {type=boolean}
 RBAN no Cloth {type=boolean}
 RBAN Params {
  type=RBANParamsStruct
  default {
   dBNO {simSpaceSettings {
     maxAngularAcceleration=50
     maxAngularVelocity=400
     maxLinearAcceleration=200
     maxLinearVelocity=400
     velocityScaleZ=1
     externalAngularVelocity=<0.5 0.5 0>
     externalLinearDragV=<1.5 2.5 1.5>
     externalLinearVelocity=<0 2 0>
    }}
   dBNO_Carried {simSpaceSettings {
     maxAngularAcceleration=50
     maxAngularVelocity=400
     maxLinearAcceleration=200
     maxLinearVelocity=400
     velocityScaleZ=1
     externalAngularVelocity=<0.5 0.5 0>
     externalLinearDragV=<1.5 2.5 1.5>
     externalLinearVelocity=<0 2 0>
    }}
   emoteOrMelee {simSpaceSettings {
     maxAngularAcceleration=50
     maxAngularVelocity=400
     maxLinearAcceleration=200
     maxLinearVelocity=400
     velocityScaleZ=1
     externalAngularVelocity=<0.5 0.5 0>
     externalLinearDragV=<1.5 2.5 1.5>
     externalLinearVelocity=<0 2 0>
    }}
   front_End {simSpaceSettings {
     maxAngularAcceleration=50
     maxAngularVelocity=400
     maxLinearAcceleration=200
     maxLinearVelocity=400
     velocityScaleZ=1
     externalAngularVelocity=<0.5 0.5 0>
     externalLinearDragV=<1.5 2.5 1.5>
     externalLinearVelocity=<0 2 0>
    }}
   onGround_Crouch_Moving {simSpaceSettings {
     maxAngularAcceleration=50
     maxAngularVelocity=400
     maxLinearAcceleration=200
     maxLinearVelocity=400
     velocityScaleZ=1
     externalAngularVelocity=<0.5 0.5 0>
     externalLinearDragV=<1.5 2.5 1.5>
     externalLinearVelocity=<0 2 0>
    }}
   onGround_Falling {simSpaceSettings {
     maxAngularAcceleration=50
     maxAngularVelocity=400
     maxLinearAcceleration=200
     maxLinearVelocity=400
     velocityScaleZ=1
     externalAngularVelocity=<0.5 0.5 0>
     externalLinearDragV=<1.5 2.5 1.5>
     externalLinearVelocity=<0 2 0>
    }}
   onGround_Moving {simSpaceSettings {
     maxAngularAcceleration=50
     maxAngularVelocity=400
     maxLinearAcceleration=200
     maxLinearVelocity=400
     velocityScaleZ=1
     externalAngularVelocity=<0.5 0.5 0>
     externalLinearDragV=<1.5 2.5 1.5>
     externalLinearVelocity=<0 2 0>
    }}
   onGround_Standing {simSpaceSettings {
     maxAngularAcceleration=50
     maxAngularVelocity=400
     maxLinearAcceleration=200
     maxLinearVelocity=400
     velocityScaleZ=1
     externalAngularVelocity=<0.5 0.5 0>
     externalLinearDragV=<1.5 2.5 1.5>
     externalLinearVelocity=<0 2 0>
    }}
   rail_Grinding {simSpaceSettings {
     maxAngularAcceleration=50
     maxAngularVelocity=400
     maxLinearAcceleration=200
     maxLinearVelocity=400
     velocityScaleZ=1
     worldAlpha=1
     externalAngularVelocity=<0.5 0.5 0>
     externalLinearDragV=<1.5 2.5 1.5>
     externalLinearVelocity=<0 2 0>
    }}
   riding_Moving {simSpaceSettings {
     maxAngularAcceleration=50
     maxAngularVelocity=400
     maxLinearAcceleration=200
     maxLinearVelocity=400
     velocityScaleZ=1
     externalAngularVelocity=<0.5 0.5 0>
     externalLinearDragV=<1.5 2.5 1.5>
     externalLinearVelocity=<0 2 0>
    }}
   riding_Moving_Targeting {simSpaceSettings {
     maxAngularAcceleration=50
     maxAngularVelocity=400
     maxLinearAcceleration=200
     maxLinearVelocity=400
     velocityScaleZ=1
     externalAngularVelocity=<0.5 0.5 0>
     externalLinearDragV=<1.5 2.5 1.5>
     externalLinearVelocity=<0 2 0>
    }}
   riding_Standing {simSpaceSettings {
     maxAngularAcceleration=50
     maxAngularVelocity=400
     maxLinearAcceleration=200
     maxLinearVelocity=400
     velocityScaleZ=1
     externalAngularVelocity=<0.5 0.5 0>
     externalLinearDragV=<1.5 2.5 1.5>
     externalLinearVelocity=<0 2 0>
    }}
   skydive_DivingDown {simSpaceSettings {
     maxAngularAcceleration=50
     maxAngularVelocity=400
     maxLinearAcceleration=200
     maxLinearVelocity=400
     velocityScaleZ=1
     externalAngularVelocity=<0.5 0.5 0>
     externalLinearDragV=<1.5 2.5 1.5>
     externalLinearVelocity=<0 2 0>
    }}
   skydive_FloatingDown {
    componentLinearAccScale=<0 0.5 0>
    componentLinearVelScale=<0 0.5 0>
    simSpaceSettings {
     maxAngularAcceleration=25
     maxAngularVelocity=100
     maxLinearAcceleration=50
     maxLinearVelocity=100
     velocityScaleZ=0.5
     worldAlpha=0.5
     externalAngularVelocity=<0.5 0.5 0>
     externalLinearDragV=<1.5 2.5 1.5>
     externalLinearVelocity=<0 2 0>
    }
   }
   skydive_GoingUp {simSpaceSettings {
     maxAngularAcceleration=50
     maxAngularVelocity=400
     maxLinearAcceleration=200
     maxLinearVelocity=400
     velocityScaleZ=1
     externalAngularVelocity=<0.5 0.5 0>
     externalLinearDragV=<1.5 2.5 1.5>
     externalLinearVelocity=<0 2 0>
    }}
   skydive_Parachuting {
    componentLinearAccScale=<0 0.1 0.3>
    componentLinearVelScale=<0 0.1 0.3>
    simSpaceSettings {
     maxAngularAcceleration=50
     maxAngularVelocity=400
     maxLinearAcceleration=200
     maxLinearVelocity=400
     velocityScaleZ=1
     externalAngularVelocity=<0.5 0.5 0>
     externalLinearDragV=<1.5 2.5 1.5>
     externalLinearVelocity=<0 2 0>
    }
   }
   skydive_Surfing {simSpaceSettings {
     maxAngularAcceleration=50
     maxAngularVelocity=400
     maxLinearAcceleration=200
     maxLinearVelocity=400
     velocityScaleZ=1
     externalAngularVelocity=<0.5 0.5 0>
     externalLinearDragV=<1.5 2.5 1.5>
     externalLinearVelocity=<0 2 0>
    }}
   swimming {simSpaceSettings {
     maxAngularAcceleration=50
     maxAngularVelocity=400
     maxLinearAcceleration=200
     maxLinearVelocity=400
     velocityScaleZ=1
     externalAngularVelocity=<0.5 0.5 0>
     externalLinearDragV=<1.5 2.5 1.5>
     externalLinearVelocity=<0 2 0>
    }}
   tactical_Sprint {simSpaceSettings {
     maxAngularAcceleration=50
     maxAngularVelocity=400
     maxLinearAcceleration=200
     maxLinearVelocity=400
     velocityScaleZ=1
     externalAngularVelocity=<0.5 0.5 0>
     externalLinearDragV=<1.5 2.5 1.5>
     externalLinearVelocity=<0 2 0>
    }}
   ziplining {simSpaceSettings {
     maxAngularAcceleration=50
     maxAngularVelocity=400
     maxLinearAcceleration=200
     maxLinearVelocity=400
     velocityScaleZ=1
     externalAngularVelocity=<0.5 0.5 0>
     externalLinearDragV=<1.5 2.5 1.5>
     externalLinearVelocity=<0 2 0>
    }}
  }
 }
 RBAN Settings Out {
  type=RBANSettingStruct
  default {
   componentLinearAccScale=<0 0 0>
   componentLinearVelScale=<0 0 0>
   simSpaceSettings {
    dampingAlpha=1
    maxAngularAcceleration=50
    maxAngularVelocity=400
    maxLinearAcceleration=200
    maxLinearVelocity=400
    velocityScaleZ=1
    worldAlpha=1
    externalAngularVelocity=<0.5 0.5 0>
    externalLinearDragV=<1.5 2.5 1.5>
    externalLinearVelocity=<0 2 0>
   }
  }
 }
 Wind Gust Local Settings {type=WindGustLocalSettingStruct, default {elapsedTime=0, finalGustStrengthLocal=0, gustDuration=0, gustStrength=0, timeBetweenGusts=0}}
 Wind Gust Settings {
  type=WindGustSettingStruct
  default {
   maxDownInterpSpeed=5
   maxGustDuration=2
   maxGustStrength=7
   maxTimeBetweenGusts=7
   maxUpInterpSpeed=5
   minDownInterpSpeed=2
   minGustDuration=0.5
   minGustStrength=3
   minTimeBetweenGusts=0
   minUpInterpSpeed=2
   windTurbulenceScalarA=0.1
   windTurbulenceScalarB=1
   windAmplitude=<50 100 50>
   windFrequency=<0 2 2>
   windOffset=<0 0 0>
  }
 }
]
