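# Physics asset for the MoleAcornGloam male character's face accessory, providing cape collider shapes.
# It has two capsule bodies: a large, offset and angled capsule on the dyn_ear_lower_cape_collision bone, and a tiny one on the head bone.
# No constraints link the bodies, so the asset defines collision volumes only, with no jointed chain.
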PhysicsAsset /BRCosmetics/Characters/Player/Male/Medium/Bodies/M_MED_MoleAcornGloam/Meshes/Parts/M_MED_MoleAcornGloam_FaceAcc_CapeColliders_New_Physics
bodies {
 dyn_ear_lower_cape_collision [{
   shapeType=Capsule
   length=36.255
   radius=16.255
   shapeName=dyn_ear_lower_cape_collision_capsule
   center=<-6.4266 -16.2248 0>
   rotation=<rot 0 -110.1696 0>
  }]
 head [{shapeType=Capsule, length=1, radius=1, rotation=<rot -0.0515 -90 -90.2064>}]
}
constraints []
